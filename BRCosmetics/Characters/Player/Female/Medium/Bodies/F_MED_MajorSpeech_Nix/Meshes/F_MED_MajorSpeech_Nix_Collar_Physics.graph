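# Physics asset for the collar of the female medium MajorSpeech_Nix cosmetic body, setting up capsule bodies for its secondary motion.
# Seven two-link dyn_collar chains (middle plus left/right a, b, c) hang from small base bodies, which join spine_05 with free angular constraints.
# Chain joints lock swing1 and limit swing2, with twist locked on a and b but narrowly limited on c and middle.
# Two head capsules add head volume.
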PhysicsAsset /BRCosmetics/Characters/Player/Female/Medium/Bodies/F_MED_MajorSpeech_Nix/Meshes/F_MED_MajorSpeech_Nix_Collar_Physics
bodies {
 base_collar_a_l [{shapeType=Capsule, length=0.505, radius=0.505, shapeName=base_collar_a_l_capsule, rotation=<rot 0 -90 0>}]
 base_collar_a_r [{shapeType=Capsule, length=0.505, radius=0.505, shapeName=base_collar_a_r_capsule, rotation=<rot 0 -90 0>}]
 base_collar_b_l [{shapeType=Capsule, length=0.505, radius=0.505, shapeName=base_collar_b_l_capsule, rotation=<rot 0 -90 0>}]
 base_collar_b_r [{shapeType=Capsule, length=0.505, radius=0.505, shapeName=base_collar_b_r_capsule, rotation=<rot 0 -90 0>}]
 base_collar_c_l [{shapeType=Capsule, length=0.505, radius=0.505, shapeName=base_collar_c_l_capsule, rotation=<rot 0 -90 0>}]
 base_collar_c_r [{shapeType=Capsule, length=0.505, radius=0.505, shapeName=base_collar_c_r_capsule, rotation=<rot 0 -90 0>}]
 base_collar_m [{shapeType=Capsule, length=0.505, radius=0.505, shapeName=base_collar_m_capsule, rotation=<rot 0 -90 0>}]
 dyn_collar_01_m [{shapeType=Capsule, length=7.1515, radius=2.0897, shapeName=dyn_collar_01_m_capsule, center=<3.0635 0 -1.9813>, rotation=<rot 0 0 90>}]
 dyn_collar_02_m [{shapeType=Capsule, length=6.7929, radius=2.7, shapeName=dyn_collar_02_m_capsule, center=<3.6415 0 -1.6542>, rotation=<rot 0 0 90>}]
 dyn_collar_a_01_l [{shapeType=Capsule, length=5, radius=2, shapeName=dyn_collar_a_01_l_capsule, center=<0.948 0 -1.0705>, rotation=<rot 0 0 90>}]
 dyn_collar_a_01_r [{shapeType=Capsule, length=5, radius=2, shapeName=dyn_collar_a_01_r_capsule, center=<-0.948 0 1.0705>, rotation=<rot 0 0 -90>}]
 dyn_collar_a_02_l [{shapeType=Capsule, length=5, radius=2, shapeName=dyn_collar_a_02_l_capsule, center=<2.6261 0 -1.0705>, rotation=<rot 0 12.8705 90>}]
 dyn_collar_a_02_r [{shapeType=Capsule, length=5, radius=2, shapeName=dyn_collar_a_02_r_capsule, center=<-2.6261 0 1.0705>, rotation=<rot 0 12.8705 -90>}]
 dyn_collar_b_01_l [{
   shapeType=Capsule
   length=5
   radius=2
   shapeName=dyn_collar_b_01_l_capsule
   center=<1.2812 0.3281 -1.3793>
   rotation=<rot -0.0082 -0.0104 83.2618>
  }]
 dyn_collar_b_01_r [{
   shapeType=Capsule
   length=5
   radius=2
   shapeName=dyn_collar_b_01_r_capsule
   center=<-1.2812 -0.2865 1.0275>
   rotation=<rot -0.0082 -0.0104 -105.6004>
  }]
 dyn_collar_b_02_l [{
   shapeType=Capsule
   length=4.7136
   radius=2.7
   shapeName=dyn_collar_b_02_l_capsule
   center=<2.224 0.269 -0.794>
   rotation=<rot 5.6651 17.0351 72.135>
  }]
 dyn_collar_b_02_r [{
   shapeType=Capsule
   length=4.7136
   radius=3
   shapeName=dyn_collar_b_02_r_capsule
   center=<-2.224 -0.269 0.794>
   rotation=<rot 5.6651 17.0351 -107.865>
  }]
 dyn_collar_c_01_l [{shapeType=Capsule, length=5, radius=2, shapeName=dyn_collar_c_01_l_capsule, center=<1.5044 0.0128 -1.5001>, rotation=<rot -0.0001 0 86.4156>}]
 dyn_collar_c_01_r [{shapeType=Capsule, length=5, radius=2, shapeName=dyn_collar_c_01_r_capsule, center=<-1.5044 0 1.4751>, rotation=<rot 0 0 -90>}]
 dyn_collar_c_02_l [{
   shapeType=Capsule
   length=4.6101
   radius=2.7
   shapeName=dyn_collar_c_02_l_capsule
   center=<2.9461 -0.3665 -1.5101>
   rotation=<rot 1.7754 12.2068 85.2277>
  }]
 dyn_collar_c_02_r [{
   shapeType=Capsule
   length=4.6101
   radius=3
   shapeName=dyn_collar_c_02_r_capsule
   center=<-2.9426 0.3907 1.2355>
   rotation=<rot 1.7755 12.2068 -98.1508>
  }]
 head [
  {shapeType=Capsule, length=3.9344, radius=8.43, shapeName=head_capsule, center=<-1.5324 -0.3627 0>, rotation=<rot 90 -90 0>}
  {shapeType=Capsule, length=3.1068, radius=9.7125, shapeName=head_capsule, center=<5.5184 -2.5847 0>}
 ]
 spine_05 [{shapeType=Capsule, length=2, radius=2, shapeName=spine_05_capsule}]
}
constraints [
 {
  bone1Name=head
  bone2Name=spine_05
  swing1LimitDegrees=45
  swing1Motion=Free
  swing2LimitDegrees=45
  swing2Motion=Free
  twistLimitDegrees=45
  twistMotion=Free
 }
 {
  bone1Name=dyn_collar_a_01_r
  bone2Name=base_collar_a_r
  swing1LimitDegrees=45
  swing1Motion=Locked
  swing2LimitDegrees=45
  twistLimitDegrees=45
  twistMotion=Locked
 }
 {
  bone1Name=dyn_collar_a_02_r
  bone2Name=dyn_collar_a_01_r
  swing1LimitDegrees=45
  swing1Motion=Locked
  swing2LimitDegrees=45
  twistLimitDegrees=45
  twistMotion=Locked
 }
 {
  bone1Name=dyn_collar_b_01_l
  bone2Name=base_collar_b_l
  swing1LimitDegrees=45
  swing1Motion=Locked
  swing2LimitDegrees=45
  twistLimitDegrees=45
  twistMotion=Locked
 }
 {
  bone1Name=dyn_collar_b_02_l
  bone2Name=dyn_collar_b_01_l
  swing1LimitDegrees=45
  swing1Motion=Locked
  swing2LimitDegrees=45
  twistLimitDegrees=45
  twistMotion=Locked
 }
 {
  bone1Name=dyn_collar_b_01_r
  bone2Name=base_collar_b_r
  swing1LimitDegrees=45
  swing1Motion=Locked
  swing2LimitDegrees=45
  twistLimitDegrees=45
  twistMotion=Locked
 }
 {
  bone1Name=dyn_collar_b_02_r
  bone2Name=dyn_collar_b_01_r
  swing1LimitDegrees=45
  swing1Motion=Locked
  swing2LimitDegrees=45
  twistLimitDegrees=45
  twistMotion=Locked
 }
 {
  bone1Name=dyn_collar_c_01_l
  bone2Name=base_collar_c_l
  swing1LimitDegrees=45
  swing1Motion=Locked
  swing2LimitDegrees=45
  twistLimitDegrees=3
 }
 {
  bone1Name=dyn_collar_c_02_l
  bone2Name=dyn_collar_c_01_l
  swing1LimitDegrees=45
  swing1Motion=Locked
  swing2LimitDegrees=45
  twistLimitDegrees=3
 }
 {
  bone1Name=dyn_collar_c_01_r
  bone2Name=base_collar_c_r
  swing1LimitDegrees=45
  swing1Motion=Locked
  swing2LimitDegrees=45
  twistLimitDegrees=3
 }
 {
  bone1Name=dyn_collar_c_02_r
  bone2Name=dyn_collar_c_01_r
  swing1LimitDegrees=45
  swing1Motion=Locked
  swing2LimitDegrees=45
  twistLimitDegrees=3
 }
 {bone1Name=dyn_collar_01_m, bone2Name=base_collar_m, swing1LimitDegrees=45, swing1Motion=Locked, swing2LimitDegrees=45, twistLimitDegrees=7}
 {bone1Name=dyn_collar_02_m, bone2Name=dyn_collar_01_m, swing1LimitDegrees=45, swing1Motion=Locked, swing2LimitDegrees=45, twistLimitDegrees=7}
 {
  bone1Name=dyn_collar_a_01_l
  bone2Name=base_collar_a_l
  swing1LimitDegrees=45
  swing1Motion=Locked
  swing2LimitDegrees=45
  twistLimitDegrees=45
  twistMotion=Locked
 }
 {
  bone1Name=dyn_collar_a_02_l
  bone2Name=dyn_collar_a_01_l
  swing1LimitDegrees=45
  swing1Motion=Locked
  swing2LimitDegrees=45
  twistLimitDegrees=45
  twistMotion=Locked
 }
 {
  bone1Name=base_collar_a_l
  bone2Name=spine_05
  swing1LimitDegrees=45
  swing1Motion=Free
  swing2LimitDegrees=45
  swing2Motion=Free
  twistLimitDegrees=45
  twistMotion=Free
 }
 {
  bone1Name=base_collar_a_r
  bone2Name=spine_05
  swing1LimitDegrees=45
  swing1Motion=Free
  swing2LimitDegrees=45
  swing2Motion=Free
  twistLimitDegrees=45
  twistMotion=Free
 }
 {
  bone1Name=base_collar_b_l
  bone2Name=spine_05
  swing1LimitDegrees=45
  swing1Motion=Free
  swing2LimitDegrees=45
  swing2Motion=Free
  twistLimitDegrees=45
  twistMotion=Free
 }
 {
  bone1Name=base_collar_b_r
  bone2Name=spine_05
  swing1LimitDegrees=45
  swing1Motion=Free
  swing2LimitDegrees=45
  swing2Motion=Free
  twistLimitDegrees=45
  twistMotion=Free
 }
 {
  bone1Name=base_collar_c_l
  bone2Name=spine_05
  swing1LimitDegrees=45
  swing1Motion=Free
  swing2LimitDegrees=45
  swing2Motion=Free
  twistLimitDegrees=45
  twistMotion=Free
 }
 {
  bone1Name=base_collar_c_r
  bone2Name=spine_05
  swing1LimitDegrees=45
  swing1Motion=Free
  swing2LimitDegrees=45
  swing2Motion=Free
  twistLimitDegrees=45
  twistMotion=Free
 }
 {
  bone1Name=base_collar_m
  bone2Name=spine_05
  swing1LimitDegrees=45
  swing1Motion=Free
  swing2LimitDegrees=45
  swing2Motion=Free
  twistLimitDegrees=45
  twistMotion=Free
 }
]
