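# Control Rig for the MajorSpeech_Nix female body that applies corrective motion to skirt, pant cloth, sleeve and shoulder pad bones.
# Leg-direction weights, multiplied by standing, crouching or DBNO state, drive layered skirt transform offsets.
# Spherical pose readers on the upper arms, gated by ground, weapon and emote states, drive sleeve and shoulder corrections.
# Constraints attach cloth to limbs and aim sleeves floorward in emotes, while interpolated alphas smooth several state changes.
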ControlRigBlueprint /BRCosmetics/Characters/Player/Female/Medium/Bodies/F_MED_MajorSpeech_Nix/Meshes/F_MED_MajorSpeech_Nix_CtrlRig
graphs [
 Multiply_14_ContainedGraph {nodes [
   n1 = RigVMFunctionEntryNode()
   RigVMFunctionReturnNode(Result=&RigVMFunction_MathDoubleMul(B=@n1.C, A=&RigVMFunction_MathDoubleMul(A=@n1.A, B=@n1.B).Result).Result)
  ]}
 Multiply_14_ContainedGraph {nodes [
   n1 = RigVMFunctionEntryNode()
   RigVMFunctionReturnNode(Result=&RigVMFunction_MathDoubleMul(B=@n1.C, A=&RigVMFunction_MathDoubleMul(A=@n1.A, B=@n1.B).Result).Result)
  ]}
 Multiply_14_ContainedGraph {nodes [
   n1 = RigVMFunctionEntryNode()
   RigVMFunctionReturnNode(Result=&RigVMFunction_MathDoubleMul(B=@n1.C, A=&RigVMFunction_MathDoubleMul(A=@n1.A, B=@n1.B).Result).Result)
  ]}
 Multiply_14_ContainedGraph {nodes [
   n1 = RigVMFunctionEntryNode()
   RigVMFunctionReturnNode(Result=&RigVMFunction_MathDoubleMul(B=@n1.C, A=&RigVMFunction_MathDoubleMul(A=@n1.A, B=@n1.B).Result).Result)
  ]}
 Multiply_15_ContainedGraph {nodes [
   n1 = RigVMFunctionEntryNode()
   Multiply = &RigVMFunction_MathDoubleMul(A=@n1.A, B=@n1.B)
   RigVMFunctionReturnNode(Result=&RigVMFunction_MathDoubleMul(B=@n1.C, A=@Multiply.Result).Result)
  ]}
 Multiply_15_ContainedGraph {nodes [
   n1 = RigVMFunctionEntryNode()
   Multiply = &RigVMFunction_MathDoubleMul(A=@n1.A, B=@n1.B)
   RigVMFunctionReturnNode(Result=&RigVMFunction_MathDoubleMul(B=@n1.C, A=@Multiply.Result).Result)
  ]}
 Multiply_1_ContainedGraph {nodes [
   n1 = RigVMFunctionEntryNode()
   Multiply = &RigVMFunction_MathDoubleMul(A=@n1.A, B=@n1.B)
   RigVMFunctionReturnNode(Result=&RigVMFunction_MathDoubleMul(B=@n1.C, A=@Multiply.Result).Result)
  ]}
 Multiply_31_ContainedGraph {nodes [
   n1 = RigVMFunctionEntryNode()
   n2 = &RigVMFunction_MathDoubleMul(A=@n1.A, B=@n1.B)
   RigVMFunctionReturnNode(Result=&RigVMFunction_MathDoubleMul(B=@n1.D, A=&RigVMFunction_MathDoubleMul(B=@n1.C, A=@n2.Result).Result).Result)
  ]}
 Multiply_31_ContainedGraph {nodes [
   n1 = RigVMFunctionEntryNode()
   n2 = &RigVMFunction_MathDoubleMul(A=@n1.A, B=@n1.B)
   RigVMFunctionReturnNode(Result=&RigVMFunction_MathDoubleMul(B=@n1.D, A=&RigVMFunction_MathDoubleMul(B=@n1.C, A=@n2.Result).Result).Result)
  ]}
 Multiply_32_ContainedGraph {nodes [
   n1 = RigVMFunctionEntryNode()
   RigVMFunctionReturnNode(Result=&RigVMFunction_MathDoubleMul(B=@n1.C, A=&RigVMFunction_MathDoubleMul(A=@n1.A, B=@n1.B).Result).Result)
  ]}
 Multiply_4_ContainedGraph {nodes [
   n1 = RigVMFunctionEntryNode()
   RigVMFunctionReturnNode(Result=&RigVMFunction_MathDoubleMul(B=@n1.C, A=&RigVMFunction_MathDoubleMul(A=@n1.A, B=@n1.B).Result).Result)
  ]}
 New Function_ContainedGraph {nodes [
   RigVMFunctionEntryNode()
   > RigVMFunction_Sequence = RigVMAggregateNode()
   &RigUnit_ParentConstraint(
    AdvancedSettings=(InterpolationType=Average,RotationOrderForFilter=XZY)
    Child=(Type=Bone,Name="base_ClothPants_out_l")
    Parents=((Item=(Type=Bone,Name="calf_l"),Weight=1))
    Weight=1
    ExecutePin=@RigVMFunction_Sequence.A
   )
   > &RigUnit_ParentConstraint(
    AdvancedSettings=(InterpolationType=Average,RotationOrderForFilter=XZY)
    Child=(Type=Bone,Name="base_ClothPants_bk_l")
    Parents=((Item=(Type=Bone,Name="calf_l"),Weight=1))
    Weight=1
   )
   > &RigUnit_ParentConstraint(
    AdvancedSettings=(InterpolationType=Average,RotationOrderForFilter=XZY)
    Child=(Type=Bone,Name="base_ClothPants_in_l")
    Parents=((Item=(Type=Bone,Name="calf_l"),Weight=1))
    Weight=1
   )
   > &RigUnit_ParentConstraint(
    AdvancedSettings=(InterpolationType=Average,RotationOrderForFilter=XZY)
    Child=(Type=Bone,Name="base_ClothPants_fr_l")
    Parents=((Item=(Type=Bone,Name="calf_l"),Weight=1))
    Weight=1
   )
   > &RigUnit_ParentConstraint(
    AdvancedSettings=(InterpolationType=Average,RotationOrderForFilter=XZY)
    Child=(Type=Bone,Name="base_ClothPants_out_r")
    Parents=((Item=(Type=Bone,Name="calf_r"),Weight=1))
    Weight=1
   )
   > &RigUnit_ParentConstraint(
    AdvancedSettings=(InterpolationType=Average,RotationOrderForFilter=XZY)
    Child=(Type=Bone,Name="base_ClothPants_bk_r")
    Parents=((Item=(Type=Bone,Name="calf_r"),Weight=1))
    Weight=1
   )
   > &RigUnit_ParentConstraint(
    AdvancedSettings=(InterpolationType=Average,RotationOrderForFilter=XZY)
    Child=(Type=Bone,Name="base_ClothPants_in_r")
    Parents=((Item=(Type=Bone,Name="calf_r"),Weight=1))
    Weight=1
   )
   > n1 = &RigUnit_ParentConstraint(
    AdvancedSettings=(InterpolationType=Average,RotationOrderForFilter=XZY)
    Child=(Type=Bone,Name="base_ClothPants_fr_r")
    Parents=((Item=(Type=Bone,Name="calf_r"),Weight=1))
    Weight=1
   )
   RigVMFunction_ControlFlowBranch = &RigVMFunction_ControlFlowBranch(ExecuteContext=@n1.ExecutePin, Condition=RigVMVariableNode() {variable=Skydiving}.Value)
   Delta Rotation = RigVMFunctionReferenceNode(
    Rotation Order=ZYX
    child=calf_l
    debugRotation=False
    max=90
    onlyTwist=False
    parent=calf_l_null
    parentType=Null
    premult=1
    primAxis=2
    twistAxis=(X=1,Y=0,Z=0)
    ExecuteContext=@RigVMFunction_ControlFlowBranch.False
   )
   n2 = &RigVMFunction_AlphaInterp(
    InterpSpeedDecreasing=10
    InterpSpeedIncreasing=10
    Scale=1
    bClampResult=true
    Value=&RigVMFunction_MathBoolToFloat(Value=RigVMVariableNode() {variable=Rail}.Value).Result
   )
   n3 = &RigVMFunction_MathDoubleRemap(SourceMaximum=1, SourceMinimum=0, TargetMaximum=0, TargetMinimum=1, Value=@n2.Result)
   Multiply_25 = RigVMAggregateNode(
    B=@"Delta Rotation".ouputWeight
    A=&RigVMFunction_MathDoubleRemap(SourceMaximum=0.1, SourceMinimum=0, TargetMaximum=1, TargetMinimum=0, Value=RigVMVariableNode() {variable=Leg Back L}.Value).Result
    C=RigVMVariableNode() {variable=Standing}.Value
    D=@n3.Result
   )
   > &RigUnit_ModifyTransforms(
    ItemToModify="((Item=(Type=Bone,Name=\"base_skirt_sd_l\"),Transform=(Rotation=(X=0,Y=-0.2625,Z=0,W=0.9649),Translation=(X=0,Y=0,Z=0),Scale3D=(X=1,Y=1,Z=1))),(Item=(Type=Bone,Name=\"base_skirt_bk_c\"),Transform=(Rotation=(X=0,Y=-0.322,Z=0,W=0.9467),Translation=(X=0,Y=0,Z=0), ...(truncated)"
    WeightMinimum=0
    Weight=@Multiply_25.Result
   )
   > Delta Rotation_1 = RigVMFunctionReferenceNode(
    Rotation Order=ZYX
    child=calf_r
    debugRotation=False
    max=90
    onlyTwist=False
    parent=calf_r_null
    parentType=Null
    premult=1
    primAxis=2
    twistAxis=(X=1,Y=0,Z=0)
   )
   Multiply_26 = RigVMAggregateNode(
    B=@"Delta Rotation_1".ouputWeight
    A=&RigVMFunction_MathDoubleRemap(SourceMaximum=0.1, SourceMinimum=0, TargetMaximum=1, TargetMinimum=0, Value=RigVMVariableNode() {variable=Leg Back R}.Value).Result
    C=RigVMVariableNode() {variable=Standing}.Value
    D=@n3.Result
   )
   > &RigUnit_ModifyTransforms(
    ItemToModify="((Item=(Type=Bone,Name=\"base_skirt_sd_r\"),Transform=(Rotation=(X=0,Y=-0.3225,Z=0,W=0.9466),Translation=(X=0,Y=0,Z=0),Scale3D=(X=1,Y=1,Z=1))),(Item=(Type=Bone,Name=\"base_skirt_bk_c\"),Transform=(Rotation=(X=0,Y=-0.3539,Z=0,W=0.9353),Translation=(X=0,Y=0,Z=0) ...(truncated)"
    WeightMinimum=0
    Weight=@Multiply_26.Result
   )
   &RigUnit_ModifyTransforms(
    ItemToModify="((Item=(Type=Bone,Name=\"base_skirt_fr_l\"),Transform=(Rotation=(X=-0.0021,Y=-0.001,Z=-0.5018,W=-0.865),Translation=(X=0,Y=0,Z=0),Scale3D=(X=1,Y=1,Z=1))),(Item=(Type=Bone,Name=\"base_skirt_sd_l\"),Transform=(Rotation=(X=0.1577,Y=0.0751,Z=0.0385,W=0.9839),Trans ...(truncated)"
    WeightMinimum=0
    Weight=&RigVMFunction_MathDoubleMul(B=RigVMVariableNode() {variable=Standing}.Value, A=RigVMVariableNode() {variable=Leg Front L}.Value).Result
    ExecutePin=@RigVMFunction_Sequence.B
   )
   > &RigUnit_ModifyTransforms(
    ItemToModify="((Item=(Type=Bone,Name=\"base_skirt_fr_l\"),Transform=(Rotation=(X=0.0765,Y=0.0279,Z=0.0235,W=0.9964),Translation=(X=0,Y=0,Z=0),Scale3D=(X=1,Y=1,Z=1))),(Item=(Type=Bone,Name=\"base_skirt_sd_l\"),Transform=(Rotation=(X=0.0973,Y=-0.0333,Z=-0.2915,W=0.951),Transl ...(truncated)"
    WeightMinimum=0
    Weight=&RigVMFunction_MathDoubleMul(B=RigVMVariableNode() {variable=Standing}.Value, A=RigVMVariableNode() {variable=Leg FrontSide L}.Value).Result
   )
   n4 = &RigVMFunction_MathDoubleMul(B=RigVMVariableNode() {variable=Standing}.Value, A=RigVMVariableNode() {variable=Leg Side L}.Value)
   > &RigUnit_ModifyTransforms(
    ItemToModify="((Item=(Type=Bone,Name=\"base_skirt_fr_l\"),Transform=(Rotation=(X=-0.3744,Y=-0.6297,Z=0.0285,W=0.6801),Translation=(X=1.9249,Y=3.194,Z=-2.4247),Scale3D=(X=1,Y=1,Z=1))),(Item=(Type=Bone,Name=\"base_skirt_sd_l\"),Transform=(Rotation=(X=0.4069,Y=-0.4031,Z=0.3188 ...(truncated)"
    WeightMinimum=0
    Weight=@n4.Result
   )
   Remap = &RigVMFunction_MathDoubleRemap(SourceMaximum=1, SourceMinimum=0.7, TargetMaximum=1, TargetMinimum=0, Value=@n4.Result)
   > &RigUnit_ModifyTransforms(
    ItemToModify="((Item=(Type=Bone,Name=\"base_skirt_fr_l\"),Transform=(Rotation=(X=0.0341,Y=0.1684,Z=-0.1455,W=0.9743),Translation=(X=0,Y=0,Z=0),Scale3D=(X=1,Y=1,Z=1))),(Item=(Type=Bone,Name=\"base_skirt_sd_l\"),Transform=(Rotation=(X=-0.008,Y=0.0404,Z=-0.3531,W=0.9347),Trans ...(truncated)"
    WeightMinimum=0
    Weight=@Remap.Result
   )
   > &RigUnit_ModifyTransforms(
    ItemToModify="((Item=(Type=Bone,Name=\"base_skirt_fr_l\"),Transform=(Rotation=(X=-0.1934,Y=-0.0462,Z=-0.0557,W=0.9784),Translation=(X=0,Y=0,Z=0),Scale3D=(X=1,Y=1,Z=1))),(Item=(Type=Bone,Name=\"base_skirt_sd_l\"),Transform=(Rotation=(X=-0.1606,Y=-0.1275,Z=0.3642,W=0.9084),Tr ...(truncated)"
    WeightMinimum=0
    Weight=&RigVMFunction_MathDoubleMul(B=RigVMVariableNode() {variable=Standing}.Value, A=RigVMVariableNode() {variable=Leg BackSide L}.Value).Result
   )
   > &RigUnit_ModifyTransforms(
    ItemToModify="((Item=(Type=Bone,Name=\"base_skirt_fr_l\"),Transform=(Rotation=(X=0.0015,Y=-0.0602,Z=-0.2587,W=0.9641),Translation=(X=2.158,Y=3.7493,Z=0.1471),Scale3D=(X=1,Y=1,Z=1))),(Item=(Type=Bone,Name=\"base_skirt_sd_l\"),Transform=(Rotation=(X=-0.4339,Y=-0.4318,Z=-0.441 ...(truncated)"
    WeightMinimum=0
    Weight=&RigVMFunction_MathDoubleMul(B=RigVMVariableNode() {variable=Standing}.Value, A=RigVMVariableNode() {variable=Leg Back L}.Value).Result
   )
   > &RigUnit_ModifyTransforms(
    ItemToModify="((Item=(Type=Bone,Name=\"base_skirt_fr_l\"),Transform=(Rotation=(X=0,Y=0,Z=0,W=1),Translation=(X=0,Y=0,Z=0),Scale3D=(X=1,Y=1,Z=1))),(Item=(Type=Bone,Name=\"dyn_skirt_fr_l_02\"),Transform=(Rotation=(X=0,Y=0,Z=0,W=1),Translation=(X=-7.4287,Y=4.011,Z=-2.5606),Sca ...(truncated)"
    WeightMinimum=0
    Weight=&RigVMFunction_MathDoubleMul(A=RigVMVariableNode() {variable=Leg FrontInside L}.Value, B=RigVMVariableNode() {variable=Standing}.Value).Result
   )
   > &RigUnit_ModifyTransforms(
    ItemToModify="((Item=(Type=Bone,Name=\"base_skirt_fr_r\"),Transform=(Rotation=(X=0.2689,Y=0.0535,Z=0.6602,W=-0.6993),Translation=(X=-3.7412,Y=0.7142,Z=0.3702),Scale3D=(X=1,Y=1,Z=1))),(Item=(Type=Bone,Name=\"base_skirt_fr_l\"),Transform=(Rotation=(X=0,Y=0,Z=0.0748,W=0.9972), ...(truncated)"
    WeightMinimum=0
    Weight=&RigVMFunction_MathDoubleMul(A=RigVMVariableNode() {variable=Leg Front R}.Value, B=RigVMVariableNode() {variable=Standing}.Value).Result
   )
   > &RigUnit_ModifyTransforms(
    ItemToModify="((Item=(Type=Bone,Name=\"base_skirt_fr_r\"),Transform=(Rotation=(X=-0.3939,Y=-0.508,Z=-0.5546,W=0.5284),Translation=(X=0,Y=0,Z=0),Scale3D=(X=1,Y=1,Z=1))),(Item=(Type=Bone,Name=\"base_skirt_sd_r\"),Transform=(Rotation=(X=-0.3934,Y=-0.242,Z=0.0396,W=0.8861),Tran ...(truncated)"
    WeightMinimum=0
    Weight=&RigVMFunction_MathDoubleMul(A=RigVMVariableNode() {variable=Leg Side R}.Value, B=RigVMVariableNode() {variable=Standing}.Value).Result
   )
   > &RigUnit_ModifyTransforms(
    ItemToModify="((Item=(Type=Bone,Name=\"base_skirt_fr_r\"),Transform=(Rotation=(X=0.0888,Y=-0.1438,Z=-0.373,W=0.9123),Translation=(X=0,Y=0,Z=0),Scale3D=(X=1,Y=1,Z=1))),(Item=(Type=Bone,Name=\"base_skirt_sd_r\"),Transform=(Rotation=(X=0.0085,Y=-0.2884,Z=-0.591,W=0.7533),Trans ...(truncated)"
    WeightMinimum=0
    Weight=&RigVMFunction_MathDoubleMul(A=RigVMVariableNode() {variable=Leg BackSide R}.Value, B=RigVMVariableNode() {variable=Standing}.Value).Result
   )
   > &RigUnit_ModifyTransforms(
    ItemToModify="((Item=(Type=Bone,Name=\"base_skirt_fr_r\"),Transform=(Rotation=(X=0.0357,Y=-0.1143,Z=0.2961,W=0.9476),Translation=(X=0,Y=0,Z=0),Scale3D=(X=1,Y=1,Z=1))),(Item=(Type=Bone,Name=\"base_skirt_sd_r\"),Transform=(Rotation=(X=0.1817,Y=-0.5491,Z=0.2734,W=0.7686),Trans ...(truncated)"
    WeightMinimum=0
    Weight=&RigVMFunction_MathDoubleMul(A=RigVMVariableNode() {variable=Leg Back R}.Value, B=RigVMVariableNode() {variable=Standing}.Value).Result
   )
   > &RigUnit_ModifyTransforms(
    ItemToModify="((Item=(Type=Bone,Name=\"base_skirt_fr_l\"),Transform=(Rotation=(X=0,Y=0,Z=0,W=1),Translation=(X=-2.0365,Y=1.1489,Z=2.0956),Scale3D=(X=1,Y=1,Z=1))),(Item=(Type=Bone,Name=\"dyn_skirt_fr_l_02\"),Transform=(Rotation=(X=0,Y=0,Z=0,W=1),Translation=(X=-1.0912,Y=1.41 ...(truncated)"
    WeightMinimum=0
    Weight=&RigVMFunction_MathDoubleMul(A=RigVMVariableNode() {variable=Leg FrontInside R}.Value, B=RigVMVariableNode() {variable=Standing}.Value).Result
   )
   Multiply = RigVMAggregateNode(
    A=RigVMVariableNode() {variable=Leg Back L}.Value
    B=RigVMVariableNode() {variable=Leg Back R}.Value
    C=RigVMVariableNode() {variable=Standing}.Value
   )
   RigVMFunction_AnimEvalRichCurve = &RigVMFunction_AnimEvalRichCurve(
    Curve=(EditorCurveData=(Keys=((),(InterpMode=RCIM_Cubic,Time=0.2988,Value=0.428,ArriveTangent=1,LeaveTangent=1),(Time=1,Value=1))))
    SourceMinimum=0
    TargetMinimum=0
    Value=@Multiply.Result
   )
   > &RigUnit_ModifyTransforms(
    ItemToModify=((Item=(Type=Bone,Name="base_skirt_bk_c"),Transform=(Rotation=(X=0,Y=0.8552,Z=0,W=0.5183),Translation=(X=6.2454,Y=0,Z=-11.7605),Scale3D=(X=1,Y=1,Z=1))))
    WeightMinimum=0
    Weight=@RigVMFunction_AnimEvalRichCurve.Result
   )
   Multiply_1 = RigVMAggregateNode(
    A=RigVMVariableNode() {variable=Leg Front L}.Value
    B=RigVMVariableNode() {variable=Leg Front R}.Value
    C=RigVMVariableNode() {variable=Standing}.Value
   )
   > &RigUnit_ModifyTransforms(
    ItemToModify=((Item=(Type=Bone,Name="base_skirt_fr_l"),Transform=(Rotation=(X=0,Y=0,Z=-0.1902,W=0.9817),Translation=(X=0,Y=0,Z=0),Scale3D=(X=1,Y=1,Z=1))))
    WeightMinimum=0
    Weight=@Multiply_1.Result
   )
   &RigUnit_ModifyTransforms(
    ItemToModify="((Item=(Type=Bone,Name=\"base_skirt_fr_l\"),Transform=(Rotation=(X=0,Y=0,Z=0,W=1),Translation=(X=0,Y=0,Z=0),Scale3D=(X=1,Y=1,Z=1))),(Item=(Type=Bone,Name=\"base_skirt_sd_l\"),Transform=(Rotation=(X=-0.0952,Y=0.2106,Z=0.0755,W=-0.97),Translation=(X=0,Y=0,Z=0),S ...(truncated)"
    WeightMinimum=0
    ExecutePin=@RigVMFunction_Sequence.F
    Weight=RigVMVariableNode() {variable=Crouching}.Value
   )
   > &RigUnit_ModifyTransforms(
    ItemToModify="((Item=(Type=Bone,Name=\"base_skirt_fr_l\"),Transform=(Rotation=(X=0,Y=0,Z=-0.4775,W=-0.8786),Translation=(X=3.3605,Y=-1.6514,Z=-0.9204),Scale3D=(X=1,Y=1,Z=1))),(Item=(Type=Bone,Name=\"base_skirt_sd_l\"),Transform=(Rotation=(X=0.1483,Y=-0.2091,Z=-0.0374,W=-0.9 ...(truncated)"
    WeightMinimum=0
    Weight=&RigVMFunction_MathDoubleMul(A=RigVMVariableNode() {variable=Leg Front L}.Value, B=RigVMVariableNode() {variable=Crouching}.Value).Result
   )
   > &RigUnit_ModifyTransforms(
    ItemToModify="((Item=(Type=Bone,Name=\"base_skirt_fr_l\"),Transform=(Rotation=(X=0.078,Y=-0.0472,Z=0.0177,W=0.9957),Translation=(X=0,Y=0,Z=0),Scale3D=(X=1,Y=1,Z=1))),(Item=(Type=Bone,Name=\"base_skirt_sd_l\"),Transform=(Rotation=(X=0.0973,Y=-0.0333,Z=-0.2915,W=0.951),Transl ...(truncated)"
    WeightMinimum=0
    Weight=&RigVMFunction_MathDoubleMul(A=RigVMVariableNode() {variable=Leg FrontSide L}.Value, B=RigVMVariableNode() {variable=Crouching}.Value).Result
   )
   n5 = &RigVMFunction_MathDoubleMul(A=RigVMVariableNode() {variable=Leg Side L}.Value, B=RigVMVariableNode() {variable=Crouching}.Value)
   > &RigUnit_ModifyTransforms(
    ItemToModify="((Item=(Type=Bone,Name=\"base_skirt_fr_l\"),Transform=(Rotation=(X=-0.3744,Y=-0.6297,Z=0.0285,W=0.6801),Translation=(X=1.9249,Y=3.194,Z=-2.4247),Scale3D=(X=1,Y=1,Z=1))),(Item=(Type=Bone,Name=\"base_skirt_sd_l\"),Transform=(Rotation=(X=0.4069,Y=-0.4031,Z=0.3188 ...(truncated)"
    WeightMinimum=0
    Weight=@n5.Result
   )
   > &RigUnit_ModifyTransforms(
    ItemToModify="((Item=(Type=Bone,Name=\"base_skirt_fr_l\"),Transform=(Rotation=(X=0.0153,Y=0.0432,Z=-0.1487,W=0.9878),Translation=(X=0,Y=0,Z=0),Scale3D=(X=1,Y=1,Z=1))),(Item=(Type=Bone,Name=\"base_skirt_sd_l\"),Transform=(Rotation=(X=-0.008,Y=0.0404,Z=-0.3531,W=0.9347),Trans ...(truncated)"
    WeightMinimum=0
    Weight=&RigVMFunction_MathDoubleRemap(SourceMaximum=1, SourceMinimum=0.6, TargetMaximum=1, TargetMinimum=0, Value=@n5.Result).Result
   )
   > &RigUnit_ModifyTransforms(
    ItemToModify="((Item=(Type=Bone,Name=\"base_skirt_fr_l\"),Transform=(Rotation=(X=0.0015,Y=-0.0602,Z=-0.2587,W=0.9641),Translation=(X=2.158,Y=3.7493,Z=0.1471),Scale3D=(X=1,Y=1,Z=1))),(Item=(Type=Bone,Name=\"base_skirt_sd_l\"),Transform=(Rotation=(X=-0.4339,Y=-0.4318,Z=-0.441 ...(truncated)"
    WeightMinimum=0
    Weight=&RigVMFunction_MathDoubleMul(A=RigVMVariableNode() {variable=Leg Back L}.Value, B=RigVMVariableNode() {variable=Crouching}.Value).Result
   )
   > &RigUnit_ModifyTransforms(
    ItemToModify="((Item=(Type=Bone,Name=\"base_skirt_fr_l\"),Transform=(Rotation=(X=0,Y=0,Z=0,W=1),Translation=(X=0,Y=0,Z=0),Scale3D=(X=1,Y=1,Z=1))),(Item=(Type=Bone,Name=\"dyn_skirt_fr_l_02\"),Transform=(Rotation=(X=0,Y=0,Z=0,W=1),Translation=(X=-9.2389,Y=5.8047,Z=2.1814),Sca ...(truncated)"
    WeightMinimum=0
    Weight=&RigVMFunction_MathDoubleMul(A=RigVMVariableNode() {variable=Leg FrontInside L}.Value, B=RigVMVariableNode() {variable=Crouching}.Value).Result
   )
   > &RigUnit_ModifyTransforms(
    ItemToModify="((Item=(Type=Bone,Name=\"base_skirt_fr_r\"),Transform=(Rotation=(X=0.2689,Y=0.0535,Z=0.6602,W=-0.6993),Translation=(X=-3.7412,Y=0.7142,Z=0.3702),Scale3D=(X=1,Y=1,Z=1))),(Item=(Type=Bone,Name=\"base_skirt_fr_l\"),Transform=(Rotation=(X=0,Y=0,Z=0.0748,W=0.9972), ...(truncated)"
    WeightMinimum=0
    Weight=&RigVMFunction_MathDoubleMul(A=RigVMVariableNode() {variable=Leg Front R}.Value, B=RigVMVariableNode() {variable=Crouching}.Value).Result
   )
   > &RigUnit_ModifyTransforms(
    ItemToModify="((Item=(Type=Bone,Name=\"base_skirt_fr_r\"),Transform=(Rotation=(X=-0.3939,Y=-0.508,Z=-0.5546,W=0.5284),Translation=(X=0,Y=0,Z=0),Scale3D=(X=1,Y=1,Z=1))),(Item=(Type=Bone,Name=\"base_skirt_sd_r\"),Transform=(Rotation=(X=-0.3934,Y=-0.242,Z=0.0396,W=0.8861),Tran ...(truncated)"
    WeightMinimum=0
    Weight=&RigVMFunction_MathDoubleMul(A=RigVMVariableNode() {variable=Leg Side R}.Value, B=RigVMVariableNode() {variable=Crouching}.Value).Result
   )
   > &RigUnit_ModifyTransforms(
    ItemToModify="((Item=(Type=Bone,Name=\"base_skirt_fr_r\"),Transform=(Rotation=(X=0.0888,Y=-0.1438,Z=-0.373,W=0.9123),Translation=(X=0,Y=0,Z=0),Scale3D=(X=1,Y=1,Z=1))),(Item=(Type=Bone,Name=\"base_skirt_sd_r\"),Transform=(Rotation=(X=0.0085,Y=-0.2884,Z=-0.591,W=0.7533),Trans ...(truncated)"
    WeightMinimum=0
    Weight=&RigVMFunction_MathDoubleMul(A=RigVMVariableNode() {variable=Leg BackSide R}.Value, B=RigVMVariableNode() {variable=Crouching}.Value).Result
   )
   > &RigUnit_ModifyTransforms(
    ItemToModify="((Item=(Type=Bone,Name=\"base_skirt_fr_r\"),Transform=(Rotation=(X=0.0357,Y=-0.1143,Z=0.2961,W=0.9476),Translation=(X=0,Y=0,Z=0),Scale3D=(X=1,Y=1,Z=1))),(Item=(Type=Bone,Name=\"base_skirt_sd_r\"),Transform=(Rotation=(X=0.1817,Y=-0.5491,Z=0.2734,W=0.7686),Trans ...(truncated)"
    WeightMinimum=0
    Weight=&RigVMFunction_MathDoubleMul(A=RigVMVariableNode() {variable=Leg Back R}.Value, B=RigVMVariableNode() {variable=Crouching}.Value).Result
   )
   > &RigUnit_ModifyTransforms(
    ItemToModify="((Item=(Type=Bone,Name=\"base_skirt_fr_l\"),Transform=(Rotation=(X=0,Y=0,Z=0,W=1),Translation=(X=0,Y=0,Z=0),Scale3D=(X=1,Y=1,Z=1))),(Item=(Type=Bone,Name=\"dyn_skirt_fr_l_02\"),Transform=(Rotation=(X=0,Y=0,Z=0,W=1),Translation=(X=0,Y=0,Z=2.6234),Scale3D=(X=1,Y ...(truncated)"
    WeightMinimum=0
    Weight=&RigVMFunction_MathDoubleMul(A=RigVMVariableNode() {variable=Leg FrontInside R}.Value, B=RigVMVariableNode() {variable=Crouching}.Value).Result
   )
   n6 = &RigVMFunction_AlphaInterp(
    InterpSpeedDecreasing=10
    InterpSpeedIncreasing=10
    Scale=1
    bClampResult=true
    Value=&RigVMFunction_MathDoubleMul(A=RigVMVariableNode() {variable=Idle}.Value, B=RigVMVariableNode() {variable=Crouching}.Value).Result
   )
   > &RigUnit_ModifyTransforms(
    ItemToModify="((Item=(Type=Bone,Name=\"base_skirt_fr_l\"),Transform=(Rotation=(X=-0.0007,Y=-0.0069,Z=0.0999,W=-0.995),Translation=(X=-0.0283,Y=0.1367,Z=0.2468),Scale3D=(X=1,Y=1,Z=1))),(Item=(Type=Bone,Name=\"base_skirt_sd_l\"),Transform=(Rotation=(X=0,Y=0,Z=0,W=1),Translati ...(truncated)"
    WeightMinimum=0
    Weight=@n6.Result
   )
   Multiply_21 = RigVMAggregateNode(
    A=RigVMVariableNode() {variable=Leg Front L}.Value
    B=RigVMVariableNode() {variable=Leg Front R}.Value
    C=RigVMVariableNode() {variable=Crouching}.Value
   )
   > &RigUnit_ModifyTransforms(
    ItemToModify=((Item=(Type=Bone,Name="base_skirt_fr_l"),Transform=(Rotation=(X=0,Y=0,Z=-0.1902,W=0.9817),Translation=(X=0,Y=0,Z=0),Scale3D=(X=1,Y=1,Z=1))))
    WeightMinimum=0
    Weight=@Multiply_21.Result
   )
   RigVMFunctionReturnNode(ExecuteContext=@RigVMFunction_Sequence.H)
   RigVMFunction_MathBoolToFloat = &RigVMFunction_MathBoolToFloat(Value=RigVMVariableNode() {variable=DBNO}.Value)
   AlphaInterp = &RigVMFunction_AlphaInterp(InterpSpeedDecreasing=10, InterpSpeedIncreasing=10, Scale=1, bClampResult=true, Value=@RigVMFunction_MathBoolToFloat.Result)
   &RigUnit_ModifyTransforms(
    ItemToModify="((Item=(Type=Bone,Name=\"base_skirt_fr_l\"),Transform=(Rotation=(X=0,Y=0,Z=0,W=1),Translation=(X=0,Y=0,Z=0),Scale3D=(X=1,Y=1,Z=1))),(Item=(Type=Bone,Name=\"base_skirt_sd_l\"),Transform=(Rotation=(X=-0.1166,Y=-0.0433,Z=0.0549,W=0.9907),Translation=(X=0,Y=0,Z=0) ...(truncated)"
    WeightMinimum=0
    Weight=&RigVMFunction_MathDoubleMul(B=@AlphaInterp.Result, A=RigVMVariableNode() {variable=Leg Front L}.Value).Result
    ExecutePin=@RigVMFunction_Sequence.G
   )
   > &RigUnit_ModifyTransforms(
    ItemToModify="((Item=(Type=Bone,Name=\"base_skirt_fr_r\"),Transform=(Rotation=(X=0,Y=0,Z=0,W=1),Translation=(X=0,Y=0,Z=0),Scale3D=(X=1,Y=1,Z=1))),(Item=(Type=Bone,Name=\"base_skirt_fr_l\"),Transform=(Rotation=(X=0,Y=0,Z=0,W=1),Translation=(X=0,Y=0,Z=0),Scale3D=(X=1,Y=1,Z=1) ...(truncated)"
    WeightMinimum=0
    Weight=&RigVMFunction_MathDoubleMul(B=@AlphaInterp.Result, A=RigVMVariableNode() {variable=Leg Front R}.Value).Result
   )
   > &RigUnit_ModifyTransforms(
    ItemToModify="((Item=(Type=Bone,Name=\"base_skirt_fr_l\"),Transform=(Rotation=(X=0,Y=0,Z=0,W=1),Translation=(X=0,Y=0,Z=0),Scale3D=(X=1,Y=1,Z=1))),(Item=(Type=Bone,Name=\"base_skirt_sd_l\"),Transform=(Rotation=(X=0.0908,Y=-0.3898,Z=0.0337,W=-0.9158),Translation=(X=0,Y=0,Z=0) ...(truncated)"
    WeightMinimum=0
    Weight=&RigVMFunction_MathDoubleMul(B=@AlphaInterp.Result, A=RigVMVariableNode() {variable=Leg Side L}.Value).Result
   )
   > &RigUnit_ModifyTransforms(
    ItemToModify="((Item=(Type=Bone,Name=\"base_skirt_fr_r\"),Transform=(Rotation=(X=0,Y=0,Z=0,W=1),Translation=(X=0,Y=0,Z=0),Scale3D=(X=1,Y=1,Z=1))),(Item=(Type=Bone,Name=\"base_skirt_sd_r\"),Transform=(Rotation=(X=0,Y=0.1816,Z=0,W=0.9834),Translation=(X=0,Y=0,Z=0),Scale3D=(X= ...(truncated)"
    WeightMinimum=0
    Weight=&RigVMFunction_MathDoubleMul(B=@AlphaInterp.Result, A=RigVMVariableNode() {variable=Leg Side R}.Value).Result
   )
   > &RigUnit_ModifyTransforms(
    ItemToModify="((Item=(Type=Bone,Name=\"base_ClothPants_bk_l\"),Transform=(Rotation=(X=0,Y=0.0936,Z=0,W=0.9956),Translation=(X=0,Y=0,Z=0),Scale3D=(X=1,Y=1,Z=1))),(Item=(Type=Bone,Name=\"base_ClothPants_out_l\"),Transform=(Rotation=(X=0,Y=0,Z=0.0678,W=0.9977),Translation=(X=0 ...(truncated)"
    WeightMinimum=0
    Weight=@AlphaInterp.Result
   )
  ]}
 New Function_ContainedGraph {nodes [
   RigVMFunctionEntryNode()
   > RigVMFunction_Sequence = RigVMAggregateNode()
   &RigUnit_ParentConstraint(
    AdvancedSettings=(InterpolationType=Average,RotationOrderForFilter=XZY)
    Child=(Type=Bone,Name="base_sleeve_r")
    Parents=((Item=(Type=Bone,Name="lowerarm_r"),Weight=1))
    Weight=1
    ExecutePin=@RigVMFunction_Sequence.A
   )
   > &RigUnit_ParentConstraint(
    AdvancedSettings=(InterpolationType=Average,RotationOrderForFilter=XZY)
    Child=(Type=Bone,Name="base_sleeve_l")
    Parents=((Item=(Type=Bone,Name="lowerarm_l"),Weight=1))
    Weight=1
   )
   n1 = RigVMVariableNode() {variable=HangGliding}
   > &RigUnit_ModifyTransforms(
    ItemToModify="((Item=(Type=Bone,Name=\"base_sleeves_twist_b_l\"),Transform=(Rotation=(X=0,Y=0,Z=0.2216,W=-0.9751),Translation=(X=-2.1215,Y=-3.8892,Z=0.0001),Scale3D=(X=1,Y=1,Z=1))),(Item=(Type=Bone,Name=\"base_sleeves_twist_a_l\"),Transform=(Rotation=(X=0,Y=0,Z=0.3462,W=-0. ...(truncated)"
    WeightMinimum=0
    Weight=@n1.Value
   )
   > &RigUnit_ModifyTransforms(
    ItemToModify="((Item=(Type=Bone,Name=\"base_sleeves_twist_b_r\"),Transform=(Rotation=(X=0,Y=0,Z=0.1451,W=-0.9894),Translation=(X=1.6891,Y=3.6206,Z=0),Scale3D=(X=1,Y=1,Z=1))),(Item=(Type=Bone,Name=\"base_sleeves_twist_a_r\"),Transform=(Rotation=(X=0,Y=0,Z=0.3017,W=-0.9534),T ...(truncated)"
    WeightMinimum=0
    Weight=@n1.Value
   )
   > &RigUnit_ModifyTransforms(
    ItemToModify=((Item=(Type=Bone,Name="floor_aim"),Transform=(Rotation=(X=0,Y=0,Z=0,W=1),Translation=(X=0,Y=20,Z=0),Scale3D=(X=1,Y=1,Z=1))))
    Weight=1
    WeightMinimum=0
   )
   > n2 = &RigUnit_SphericalPoseReader(
    ActiveRegionScaleFactors=(PositiveWidth=1,NegativeWidth=1,PositiveHeight=1,NegativeHeight=1)
    ActiveRegionSize=0.1
    Debug=(bDrawDebug=false,bDraw2D=False,bDrawLocalAxes=False,DebugScale=25,DebugSegments=20,DebugThickness=0.25)
    DriverAxis=(X=1,Y=0,Z=0)
    DriverItem=(Type=Bone,Name="upperarm_l")
    FalloffRegionScaleFactors=(PositiveWidth=1,NegativeWidth=1,PositiveHeight=1,NegativeHeight=1)
    FalloffSize=0.3
    FlipHeightScaling=False
    FlipWidthScaling=False
    OptionalParentItem=(Type=Bone,Name="None")
    RotationOffset=(X=180,Y=40,Z=0)
   )
   AlphaInterp = &RigVMFunction_AlphaInterp(InterpSpeedDecreasing=10, InterpSpeedIncreasing=10, Scale=1, bClampResult=true, Value=RigVMVariableNode() {variable=OnGround}.Value)
   Multiply = &RigVMFunction_MathDoubleMul(A=@n2.OutputParam, B=@AlphaInterp.Result)
   > &RigUnit_ModifyTransforms(
    ItemToModify="((Item=(Type=Bone,Name=\"base_sleeves_twist_b_l\"),Transform=(Rotation=(X=0,Y=-0.3171,Z=0,W=0.9484),Translation=(X=0,Y=0,Z=0),Scale3D=(X=1,Y=1,Z=1))),(Item=(Type=Bone,Name=\"base_sleeves_twist_a_l\"),Transform=(Rotation=(X=0,Y=-0.5189,Z=0,W=0.8549),Translation ...(truncated)"
    WeightMinimum=0
    Weight=@Multiply.Result
   )
   > n3 = &RigUnit_SphericalPoseReader(
    ActiveRegionScaleFactors=(PositiveWidth=1,NegativeWidth=1,PositiveHeight=1,NegativeHeight=1)
    ActiveRegionSize=0.1
    Debug=(bDrawDebug=false,bDraw2D=False,bDrawLocalAxes=False,DebugScale=25,DebugSegments=20,DebugThickness=0.25)
    DriverAxis=(X=1,Y=0,Z=0)
    DriverItem=(Type=Bone,Name="upperarm_l")
    FalloffRegionScaleFactors=(PositiveWidth=1,NegativeWidth=1,PositiveHeight=1,NegativeHeight=1)
    FalloffSize=0.4
    FlipHeightScaling=False
    FlipWidthScaling=False
    OptionalParentItem=(Type=Bone,Name="spine_05")
    RotationOffset=(X=8,Y=-3,Z=0)
   )
   > &RigUnit_ModifyTransforms(
    ItemToModify="((Item=(Type=Bone,Name=\"base_sleeves_twist_b_l\"),Transform=(Rotation=(X=0.2169,Y=0.2961,Z=-0.1575,W=0.9168),Translation=(X=-2.4751,Y=-4.227,Z=-2.0972),Scale3D=(X=1,Y=1,Z=1))),(Item=(Type=Bone,Name=\"base_sleeves_twist_a_l\"),Transform=(Rotation=(X=-0.041,Y=- ...(truncated)"
    WeightMinimum=0
    Weight=&RigVMFunction_MathDoubleMul(A=@n3.OutputParam, B=@AlphaInterp.Result).Result
   )
   > n4 = &RigUnit_SphericalPoseReader(
    ActiveRegionScaleFactors=(PositiveWidth=1,NegativeWidth=1,PositiveHeight=1,NegativeHeight=1)
    ActiveRegionSize=0.1
    Debug=(bDrawDebug=false,bDraw2D=False,bDrawLocalAxes=False,DebugScale=25,DebugSegments=20,DebugThickness=0.25)
    DriverAxis=(X=-1,Y=0,Z=0)
    DriverItem=(Type=Bone,Name="upperarm_r")
    FalloffRegionScaleFactors=(PositiveWidth=1,NegativeWidth=1,PositiveHeight=1,NegativeHeight=1)
    FalloffSize=0.3
    FlipHeightScaling=False
    FlipWidthScaling=False
    OptionalParentItem=(Type=Bone,Name="None")
    RotationOffset=(X=0,Y=40,Z=0)
   )
   > &RigUnit_ModifyTransforms(
    ItemToModify="((Item=(Type=Bone,Name=\"base_sleeves_twist_b_r\"),Transform=(Rotation=(X=-0.0084,Y=0.2455,Z=0.0331,W=-0.9688),Translation=(X=-0.0859,Y=4.2814,Z=-0.0465),Scale3D=(X=1,Y=1,Z=1))),(Item=(Type=Bone,Name=\"base_sleeves_twist_a_r\"),Transform=(Rotation=(X=0.0721,Y= ...(truncated)"
    WeightMinimum=0
    Weight=&RigVMFunction_MathDoubleMul(A=@n4.OutputParam, B=@AlphaInterp.Result).Result
   )
   > n5 = &RigUnit_SphericalPoseReader(
    ActiveRegionScaleFactors=(PositiveWidth=1,NegativeWidth=1,PositiveHeight=1,NegativeHeight=1)
    ActiveRegionSize=0.1
    Debug=(bDrawDebug=false,bDraw2D=False,bDrawLocalAxes=False,DebugScale=25,DebugSegments=20,DebugThickness=0.25)
    DriverAxis=(X=-1,Y=0,Z=0)
    DriverItem=(Type=Bone,Name="upperarm_r")
    FalloffRegionScaleFactors=(PositiveWidth=1,NegativeWidth=1,PositiveHeight=1,NegativeHeight=1)
    FalloffSize=0.7
    FlipHeightScaling=false
    FlipWidthScaling=false
    OptionalParentItem=(Type=Bone,Name="spine_05")
    RotationOffset=(X=-140,Y=50,Z=0)
   )
   > &RigUnit_ModifyTransforms(
    ItemToModify="((Item=(Type=Bone,Name=\"base_sleeves_twist_b_r\"),Transform=(Rotation=(X=0.0011,Y=0.3379,Z=0.4432,W=-0.8303),Translation=(X=-0.9266,Y=3.2288,Z=1.2693),Scale3D=(X=1,Y=1,Z=1))),(Item=(Type=Bone,Name=\"base_sleeves_twist_a_r\"),Transform=(Rotation=(X=0.0328,Y=-0 ...(truncated)"
    WeightMinimum=0
    Weight=&RigVMFunction_MathDoubleMul(A=@n5.OutputParam, B=@AlphaInterp.Result).Result
   )
   > n6 = &RigUnit_SphericalPoseReader(
    ActiveRegionScaleFactors=(PositiveWidth=1,NegativeWidth=1,PositiveHeight=1,NegativeHeight=1)
    ActiveRegionSize=0.15
    Debug=(bDrawDebug=false,bDraw2D=False,bDrawLocalAxes=False,DebugScale=25,DebugSegments=20,DebugThickness=0.25)
    DriverAxis=(X=1,Y=0,Z=0)
    DriverItem=(Type=Bone,Name="upperarm_l")
    FalloffRegionScaleFactors=(PositiveWidth=1,NegativeWidth=1,PositiveHeight=1,NegativeHeight=1)
    FalloffSize=0.5
    FlipHeightScaling=False
    FlipWidthScaling=False
    OptionalParentItem=(Type=Bone,Name="spine_05")
    RotationOffset=(X=0,Y=40,Z=0)
   )
   > &RigUnit_ModifyTransforms(
    ItemToModify="((Item=(Type=Bone,Name=\"base_sleeves_twist_b_l\"),Transform=(Rotation=(X=0,Y=0,Z=0.458,W=-0.8889),Translation=(X=1.411,Y=-2.0942,Z=0),Scale3D=(X=1,Y=1,Z=1))),(Item=(Type=Bone,Name=\"base_sleeves_twist_a_l\"),Transform=(Rotation=(X=0,Y=0,Z=0.3613,W=-0.9325),Tr ...(truncated)"
    WeightMinimum=0
    Weight=&RigVMFunction_MathDoubleMul(A=@n6.OutputParam, B=RigVMVariableNode() {variable=AlphaEmotes}.Value).Result
   )
   > n7 = &RigUnit_SphericalPoseReader(
    ActiveRegionScaleFactors=(PositiveWidth=1,NegativeWidth=1,PositiveHeight=1,NegativeHeight=1)
    ActiveRegionSize=0.15
    Debug=(bDrawDebug=false,bDraw2D=False,bDrawLocalAxes=False,DebugScale=25,DebugSegments=20,DebugThickness=0.25)
    DriverAxis=(X=-1,Y=0,Z=0)
    DriverItem=(Type=Bone,Name="upperarm_r")
    FalloffRegionScaleFactors=(PositiveWidth=1,NegativeWidth=1,PositiveHeight=1,NegativeHeight=1)
    FalloffSize=0.5
    FlipHeightScaling=false
    FlipWidthScaling=false
    OptionalParentItem=(Type=Bone,Name="spine_05")
    RotationOffset=(X=180,Y=40,Z=0)
   )
   > &RigUnit_ModifyTransforms(
    ItemToModify="((Item=(Type=Bone,Name=\"base_sleeves_twist_b_r\"),Transform=(Rotation=(X=0,Y=0,Z=0.448,W=-0.894),Translation=(X=-1.3448,Y=1.7994,Z=0),Scale3D=(X=1,Y=1,Z=1))),(Item=(Type=Bone,Name=\"base_sleeves_twist_a_r\"),Transform=(Rotation=(X=0,Y=0,Z=0.3501,W=-0.9367),Tr ...(truncated)"
    WeightMinimum=0
    Weight=&RigVMFunction_MathDoubleMul(A=@n7.OutputParam, B=RigVMVariableNode() {variable=AlphaEmotes}.Value).Result
   )
   Remap = &RigVMFunction_MathDoubleRemap(SourceMaximum=1, SourceMinimum=0, TargetMaximum=0, TargetMinimum=1, Value=@n6.OutputParam)
   n8 = &RigVMFunction_MathDoubleMul(A=@Remap.Result, B=RigVMVariableNode() {variable=AlphaEmotes}.Value)
   > &RigUnit_AimConstraintLocalSpaceOffset(
    AdvancedSettings=(DebugSettings=(bEnabled=true,Scale=10,WorldOffset=(Rotation=(X=0,Y=0,Z=0,W=1),Translation=(X=0,Y=0,Z=0),Scale3D=(X=1,Y=1,Z=1))),RotationOrderForFilter=XZY)
    AimAxis=(X=60,Y=30,Z=0)
    Child=(Type=Bone,Name="base_sleeves_twist_b_l")
    Filter=(bX=True,bY=True,bZ=True)
    Parents=((Item=(Type=Bone,Name="floor_aim"),Weight=1))
    WorldUp=(Target=(X=0,Y=0,Z=1),Kind=Location,Space=(Type=Bone,Name="lowerarm_l"))
    Weight=@n8.Result
   )
   > &RigUnit_AimConstraintLocalSpaceOffset(
    AdvancedSettings=(DebugSettings=(bEnabled=true,Scale=10,WorldOffset=(Rotation=(X=0,Y=0,Z=0,W=1),Translation=(X=0,Y=0,Z=0),Scale3D=(X=1,Y=1,Z=1))),RotationOrderForFilter=XZY)
    AimAxis=(X=60,Y=30,Z=0)
    Child=(Type=Bone,Name="base_sleeves_twist_a_l")
    Filter=(bX=True,bY=True,bZ=True)
    Parents=((Item=(Type=Bone,Name="floor_aim"),Weight=1))
    WorldUp=(Target=(X=0,Y=0,Z=1),Kind=Location,Space=(Type=Bone,Name="lowerarm_l"))
    Weight=@n8.Result
   )
   n9 = &RigVMFunction_MathDoubleMul(
    A=&RigVMFunction_MathDoubleRemap(SourceMaximum=1, SourceMinimum=0, TargetMaximum=0, TargetMinimum=1, Value=@n7.OutputParam).Result
    B=RigVMVariableNode() {variable=AlphaEmotes}.Value
   )
   > &RigUnit_AimConstraintLocalSpaceOffset(
    AdvancedSettings=(DebugSettings=(bEnabled=true,Scale=10,WorldOffset=(Rotation=(X=0,Y=0,Z=0,W=1),Translation=(X=0,Y=0,Z=0),Scale3D=(X=1,Y=1,Z=1))),RotationOrderForFilter=XZY)
    AimAxis=(X=-60,Y=30,Z=0)
    Child=(Type=Bone,Name="base_sleeves_twist_b_r")
    Filter=(bX=True,bY=True,bZ=True)
    Parents=((Item=(Type=Bone,Name="floor_aim"),Weight=1))
    WorldUp=(Target=(X=0,Y=0,Z=1),Kind=Location,Space=(Type=Bone,Name="lowerarm_r"))
    Weight=@n9.Result
   )
   > &RigUnit_AimConstraintLocalSpaceOffset(
    AdvancedSettings=(DebugSettings=(bEnabled=true,Scale=10,WorldOffset=(Rotation=(X=0,Y=0,Z=0,W=1),Translation=(X=0,Y=0,Z=0),Scale3D=(X=1,Y=1,Z=1))),RotationOrderForFilter=XZY)
    AimAxis=(X=-60,Y=30,Z=0)
    Child=(Type=Bone,Name="base_sleeves_twist_a_r")
    Filter=(bX=True,bY=True,bZ=True)
    Parents=((Item=(Type=Bone,Name="floor_aim"),Weight=1))
    WorldUp=(Target=(X=0,Y=0,Z=1),Kind=Location,Space=(Type=Bone,Name="lowerarm_r"))
    Weight=@n9.Result
   )
   > Delta Rotation X = RigVMFunctionReferenceNode(Child=upperarm_l, Parent=lowerarm_l, SourceMax=-150)
   > &RigUnit_ModifyTransforms(
    ItemToModify=((Item=(Type=Bone,Name="lowerarm_twist_extra_l"),Transform=(Rotation=(X=0.8287,Y=0,Z=0,W=-0.5597),Translation=(X=0.0162,Y=0.2657,Z=-0.4181),Scale3D=(X=1,Y=1,Z=1))))
    WeightMinimum=0
    Weight=@"Delta Rotation X".Delta X
   )
   > Delta Rotation X_1 = RigVMFunctionReferenceNode(Child=upperarm_l, Parent=lowerarm_l, SourceMax=150)
   > &RigUnit_ModifyTransforms(
    ItemToModify=((Item=(Type=Bone,Name="lowerarm_twist_extra_l"),Transform=(Rotation=(X=-0.8609,Y=0,Z=0,W=-0.5087),Translation=(X=-0.5945,Y=-0.6572,Z=-0.5245),Scale3D=(X=1,Y=1,Z=1))))
    WeightMinimum=0
    Weight=@"Delta Rotation X_1".Delta X
   )
   > Delta Rotation X_2 = RigVMFunctionReferenceNode(Child=upperarm_r, Parent=lowerarm_r, SourceMax=-150)
   > &RigUnit_ModifyTransforms(
    ItemToModify=((Item=(Type=Bone,Name="lowerarm_twist_extra_r"),Transform=(Rotation=(X=0.8042,Y=0,Z=0,W=-0.5943),Translation=(X=0,Y=-0.654,Z=0.6398),Scale3D=(X=1,Y=1,Z=1))))
    WeightMinimum=0
    Weight=@"Delta Rotation X_2".Delta X
   )
   > Delta Rotation X_3 = RigVMFunctionReferenceNode(Child=upperarm_r, Parent=lowerarm_r, SourceMax=150)
   > &RigUnit_ModifyTransforms(
    ItemToModify=((Item=(Type=Bone,Name="lowerarm_twist_extra_r"),Transform=(Rotation=(X=0.829,Y=0,Z=0,W=0.5592),Translation=(X=0,Y=0.8058,Z=0.0749),Scale3D=(X=1,Y=1,Z=1))))
    WeightMinimum=0
    Weight=@"Delta Rotation X_3".Delta X
   )
   RigVMFunctionReturnNode(ExecuteContext=@RigVMFunction_Sequence.C)
   &RigUnit_ModifyTransforms(
    ItemToModify="((Item=(Type=Bone,Name=\"base_sleeves_twist_b_l\"),Transform=(Rotation=(X=0,Y=-0.0039,Z=-0.0048,W=-1),Translation=(X=0,Y=0,Z=0),Scale3D=(X=1,Y=1,Z=1))),(Item=(Type=Bone,Name=\"lowerarm_twist_extra_l\"),Transform=(Rotation=(X=-0.7303,Y=0,Z=0,W=-0.6831),Translat ...(truncated)"
    WeightMinimum=0
    Weight=RigVMVariableNode() {variable=FEIdle}.Value
    ExecutePin=@RigVMFunction_Sequence.B
   )
  ]}
 New Function_ContainedGraph {nodes [
   n1 = RigVMFunctionEntryNode()
   RigVMFunction_MathQuaternionToEuler = &RigVMFunction_MathQuaternionToEuler(
    RotationOrder=XYZ
    Value=&RigUnit_GetRelativeTransformForItem(
     Child=(Type=Bone,Name="None")
     Parent=(Type=Bone,Name="None")
     bChildInitial=False
     bParentInitial=False
     Child.Name=@n1.Child
     Parent.Name=@n1.Parent
    ).RelativeTransform.Rotation
   )
   Remap = &RigVMFunction_MathDoubleRemap(
    TargetMaximum=1
    TargetMinimum=0
    Value=@RigVMFunction_MathQuaternionToEuler.Result.X
    SourceMaximum=@n1.SourceMax
    SourceMinimum=@n1.SourceMin
   )
   > RigVMFunctionReturnNode(Delta X=@Remap.Result)
  ]}
 New Function_ContainedGraph {nodes [
   n1 = RigVMFunctionEntryNode()
   RigVMFunction_Sequence = &RigVMFunction_Sequence(ExecuteContext=@n1.ExecuteContext)
   n2 = &RigVMFunction_ControlFlowBranch(ExecuteContext=@RigVMFunction_Sequence.A, Condition=RigVMVariableNode() {variable=Emotes}.Value)
   n3 = &RigUnit_SphericalPoseReader(
    ActiveRegionScaleFactors=(PositiveWidth=1,NegativeWidth=1,PositiveHeight=1,NegativeHeight=1)
    ActiveRegionSize=0.1
    Debug=(bDrawDebug=false,bDraw2D=False,bDrawLocalAxes=False,DebugScale=25,DebugSegments=20,DebugThickness=0.25)
    DriverAxis=(X=1,Y=0,Z=0)
    DriverItem=(Type=Bone,Name="upperarm_l")
    FalloffRegionScaleFactors=(PositiveWidth=1,NegativeWidth=1,PositiveHeight=1,NegativeHeight=1)
    FalloffSize=0.3
    FlipHeightScaling=False
    FlipWidthScaling=False
    OptionalParentItem=(Type=Bone,Name="spine_05")
    RotationOffset=(X=8,Y=-3,Z=0)
    ExecutePin=@n2.False
   )
   n4 = RigVMVariableNode() {variable=OnGround}
   > &RigUnit_ModifyTransforms(
    ItemToModify="((Item=(Type=Bone,Name=\"base_ShoulderPad_general_l\"),Transform=(Rotation=(X=0,Y=-0.355,Z=0,W=-0.9349),Translation=(X=0,Y=0,Z=0),Scale3D=(X=0.8,Y=0.8,Z=0.8))),(Item=(Type=Bone,Name=\"base_ShoulderPad_side_l\"),Transform=(Rotation=(X=0,Y=0,Z=0,W=1),Translation ...(truncated)"
    WeightMinimum=0
    Weight=&RigVMFunction_MathDoubleMul(A=@n3.OutputParam, B=@n4.Value).Result
   )
   Multiply = RigVMAggregateNode(A=@n3.OutputParam, B=RigVMVariableNode() {variable=Pistol}.Value, C=@n4.Value)
   Remap = &RigVMFunction_MathDoubleRemap(SourceMaximum=0.5, SourceMinimum=0, TargetMaximum=1, TargetMinimum=0, Value=@Multiply.Result)
   > &RigUnit_ModifyTransforms(
    ItemToModify="((Item=(Type=Bone,Name=\"base_ShoulderPad_general_r\"),Transform=(Rotation=(X=-0.358,Y=0.0202,Z=-0.0526,W=-0.932),Translation=(X=0,Y=0,Z=0),Scale3D=(X=1,Y=1,Z=1))),(Item=(Type=Bone,Name=\"base_ShoulderPad_side_l\"),Transform=(Rotation=(X=0,Y=0,Z=0,W=1),Transla ...(truncated)"
    WeightMinimum=0
    Weight=@Remap.Result
   )
   RigVMFunction_MathBoolToFloat = &RigVMFunction_MathBoolToFloat(Value=RigVMVariableNode() {variable=Skydiving}.Value)
   n5 = &RigVMFunction_AlphaInterp(InterpSpeedDecreasing=10, InterpSpeedIncreasing=10, Scale=1, bClampResult=true, Value=@RigVMFunction_MathBoolToFloat.Result)
   > &RigUnit_ModifyTransforms(
    ItemToModify="((Item=(Type=Bone,Name=\"base_ShoulderPad_front_l\"),Transform=(Rotation=(X=0,Y=0,Z=0,W=-1),Translation=(X=0,Y=0,Z=0),Scale3D=(X=1,Y=1,Z=1))),(Item=(Type=Bone,Name=\"base_ShoulderPad_side_l\"),Transform=(Rotation=(X=0,Y=0,Z=0,W=1),Translation=(X=0,Y=0,Z=0),Sca ...(truncated)"
    WeightMinimum=0
    Weight=@n5.Result
   )
   AlphaInterp = &RigVMFunction_AlphaInterp(InterpSpeedDecreasing=5, InterpSpeedIncreasing=5, Scale=1, bClampResult=true, Value=RigVMVariableNode() {variable=HangGliding}.Value)
   > &RigUnit_ModifyTransforms(
    ItemToModify="((Item=(Type=Bone,Name=\"base_ShoulderPad_general_l\"),Transform=(Rotation=(X=0.0343,Y=-0.2541,Z=0.0613,W=-0.9646),Translation=(X=4,Y=0,Z=0),Scale3D=(X=0.8,Y=0.8,Z=0.8))),(Item=(Type=Bone,Name=\"base_ShoulderPad_general_r\"),Transform=(Rotation=(X=0,Y=0.2588,Z ...(truncated)"
    WeightMinimum=0
    Weight=@AlphaInterp.Result
   )
   n6 = &RigUnit_SphericalPoseReader(
    ActiveRegionScaleFactors=(PositiveWidth=1,NegativeWidth=1,PositiveHeight=1,NegativeHeight=1)
    ActiveRegionSize=0.15
    Debug=(bDrawDebug=false,bDraw2D=False,bDrawLocalAxes=false,DebugScale=25,DebugSegments=20,DebugThickness=0.25)
    DriverAxis=(X=1,Y=0,Z=0)
    DriverItem=(Type=Bone,Name="upperarm_l")
    FalloffRegionScaleFactors=(PositiveWidth=1,NegativeWidth=1,PositiveHeight=1,NegativeHeight=1)
    FalloffSize=0.5
    FlipHeightScaling=False
    FlipWidthScaling=False
    OptionalParentItem=(Type=Bone,Name="spine_05")
    RotationOffset=(X=0,Y=40,Z=0)
    ExecutePin=@n2.True
   )
   > &RigUnit_ModifyTransforms(
    ItemToModify="((Item=(Type=Bone,Name=\"base_ShoulderPad_front_l\"),Transform=(Rotation=(X=-0.1856,Y=0.5969,Z=0.1444,W=0.7671),Translation=(X=0,Y=0,Z=0),Scale3D=(X=1,Y=1,Z=0.7))),(Item=(Type=Bone,Name=\"base_ShoulderPad_side_l\"),Transform=(Rotation=(X=0,Y=0.8024,Z=0,W=0.596 ...(truncated)"
    WeightMinimum=0
    Weight=@n6.OutputParam
   )
   > n7 = &RigUnit_SphericalPoseReader(
    ActiveRegionScaleFactors=(PositiveWidth=1,NegativeWidth=1,PositiveHeight=1,NegativeHeight=1)
    ActiveRegionSize=0.15
    Debug=(bDrawDebug=false,bDraw2D=False,bDrawLocalAxes=False,DebugScale=25,DebugSegments=20,DebugThickness=0.25)
    DriverAxis=(X=-1,Y=0,Z=0)
    DriverItem=(Type=Bone,Name="upperarm_r")
    FalloffRegionScaleFactors=(PositiveWidth=1,NegativeWidth=1,PositiveHeight=1,NegativeHeight=1)
    FalloffSize=0.5
    FlipHeightScaling=false
    FlipWidthScaling=false
    OptionalParentItem=(Type=Bone,Name="spine_05")
    RotationOffset=(X=180,Y=40,Z=0)
   )
   > &RigUnit_ModifyTransforms(
    ItemToModify="((Item=(Type=Bone,Name=\"base_ShoulderPad_front_r\"),Transform=(Rotation=(X=-0.3257,Y=0.6387,Z=0.1573,W=0.6791),Translation=(X=0,Y=0,Z=0),Scale3D=(X=1,Y=1,Z=0.7))),(Item=(Type=Bone,Name=\"base_ShoulderPad_side_r\"),Transform=(Rotation=(X=0,Y=0.6267,Z=0,W=0.779 ...(truncated)"
    WeightMinimum=0
    Weight=@n7.OutputParam
   )
   RigVMFunctionReturnNode(ExecuteContext=@RigVMFunction_Sequence.B)
   &RigUnit_ModifyTransforms(
    ExecutePin=()
    ItemToModify="((Item=(Type=Bone,Name=\"base_ShoulderPad_front_r\"),Transform=(Rotation=(X=-0.0141,Y=-0.1053,Z=-0.1321,W=-0.9855),Translation=(X=0,Y=0,Z=0),Scale3D=(X=0.1615,Y=1.0383,Z=1))),(Item=(Type=Bone,Name=\"base_ShoulderPad_side_r\"),Transform=(Rotation=(X=0.1101,Y=0. ...(truncated)"
    WeightMinimum=0
    Weight=@AlphaInterp.Result
   )
  ]}
 New Function_ContainedGraph {nodes [
   n1 = RigVMFunctionEntryNode()
   RigVMFunction_MathBoolToInteger = &RigVMFunction_MathBoolToInteger(Value=@n1.onlyTwist)
   RigVMFunction_MathQuaternionSwingTwist = &RigVMFunction_MathQuaternionSwingTwist(
    TwistAxis=@n1.twistAxis
    Input=&RigUnit_GetRelativeTransformForItem(
     Child=(Type=Bone,Name="None")
     Parent=(Type=Bone,Name="None")
     bChildInitial=False
     bParentInitial=False
     Child.Name=@n1.child
     Parent.Name=@n1.parent
     Parent.Type=@n1.parentType
    ).RelativeTransform.Rotation
   )
   RigVMFunction_MathQuaternionToEuler = &RigVMFunction_MathQuaternionToEuler(
    Value=RigVMDispatchNode(
     Values=((),())
     Index=@RigVMFunction_MathBoolToInteger.Result
     Values.0=@RigVMFunction_MathQuaternionSwingTwist.Swing
     Values.1=@RigVMFunction_MathQuaternionSwingTwist.Twist
    ) {template="DISPATCH_RigVMDispatch_SelectInt32(in Index,in Values,out Result)"}.Result
    RotationOrder=@n1.Rotation Order
   )
   > DISPATCH_RigVMDispatch_Print = RigVMDispatchNode(ScreenColor=(R=1,G=1,B=1,A=1), ScreenDuration=1.5, Enabled=@n1.debugRotation, Value=@RigVMFunction_MathQuaternionToEuler.Result) {template="DISPATCH_RigVMDispatch_Print(in Prefix,in Value,in Enabled,in ScreenDuration,in ScreenColor)"}
   Clamp = &RigVMFunction_MathIntClamp(Value=@n1.primAxis)
   DISPATCH_RigVMDispatch_SelectInt32 = RigVMDispatchNode(
    Values=(0,0,0)
    Index=@Clamp.Result
    Values.0=@RigVMFunction_MathQuaternionToEuler.Result.X
    Values.1=@RigVMFunction_MathQuaternionToEuler.Result.Y
    Values.2=@RigVMFunction_MathQuaternionToEuler.Result.Z
   ) {template="DISPATCH_RigVMDispatch_SelectInt32(in Index,in Values,out Result)"}
   Remap = &RigVMFunction_MathDoubleRemap(
    TargetMaximum=1
    TargetMinimum=0
    bClamp=False
    Value=@DISPATCH_RigVMDispatch_SelectInt32.Result
    SourceMaximum=@n1.max
    SourceMinimum=@n1.min
   )
   Multiply = &RigVMFunction_MathDoubleMul(A=&RigVMFunction_MathDoubleClamp(Value=@Remap.Result).Result, B=@n1.premult)
   > RigVMFunctionReturnNode(ouputWeight=@Multiply.Result)
  ]}
 New Function_ContainedGraph {nodes [
   RigVMFunctionEntryNode()
   > RigVMFunction_Sequence = RigVMAggregateNode()
   Delta Rotation_1 = RigVMFunctionReferenceNode(
    Rotation Order=ZYX
    child=head
    debugRotation=False
    max=-20
    min=-90
    onlyTwist=true
    parent=head_null
    parentType=Null
    premult=1
    primAxis=0
    twistAxis=(X=1,Y=0,Z=0)
    ExecuteContext=@RigVMFunction_Sequence.A
   )
   Delta Rotation_2 = RigVMFunctionReferenceNode(
    Rotation Order=ZYX
    child=head
    debugRotation=False
    max=20
    min=90
    onlyTwist=true
    parent=head_null
    parentType=Null
    premult=1
    primAxis=0
    twistAxis=(X=1,Y=0,Z=0)
    ExecuteContext=@RigVMFunction_Sequence.B
   )
   Delta Rotation = RigVMFunctionReferenceNode(
    Rotation Order=ZYX
    child=head
    debugRotation=False
    max=-40
    min=10
    onlyTwist=False
    parent=head_null
    parentType=Null
    premult=1
    primAxis=2
    twistAxis=(X=1,Y=0,Z=0)
    ExecuteContext=@RigVMFunction_Sequence.C
   )
   Minimum = /Script/RigVM.RigVMFunction_MathDoubleMin(A=@"Delta Rotation_1".ouputWeight, B=@"Delta Rotation_2".ouputWeight) {is=struct RigVMUnitNode, template="Minimum::Execute(in A,in B,out Result)"}
   Multiply = &RigVMFunction_MathDoubleMul(B=@"Delta Rotation".ouputWeight, A=@Minimum.Result)
   > &RigUnit_ModifyTransforms(
    ItemToModify="((Item=(Type=Bone,Name=\"base_collar_a_l\"),Transform=(Rotation=(X=0,Y=0,Z=0,W=1),Translation=(X=0,Y=0,Z=0),Scale3D=(X=1,Y=1,Z=1))),(Item=(Type=Bone,Name=\"base_collar_b_l\"),Transform=(Rotation=(X=0,Y=0,Z=0,W=1),Translation=(X=0,Y=0,Z=0),Scale3D=(X=1,Y=1,Z=1) ...(truncated)"
    WeightMinimum=0
    Weight=@Multiply.Result
   )
   n1 = &RigVMFunction_AlphaInterp(
    InterpSpeedDecreasing=10
    InterpSpeedIncreasing=10
    Scale=1
    bClampResult=true
    Value=&RigVMFunction_MathBoolToFloat(Value=RigVMVariableNode() {variable=Skydiving}.Value).Result
   )
   > &RigUnit_ModifyTransforms(
    ItemToModify="((Item=(Type=Bone,Name=\"base_collar_a_l\"),Transform=(Rotation=(X=0,Y=0,Z=0,W=1),Translation=(X=0,Y=0,Z=0),Scale3D=(X=1,Y=1,Z=1))),(Item=(Type=Bone,Name=\"base_collar_b_l\"),Transform=(Rotation=(X=0,Y=0,Z=0,W=1),Translation=(X=0,Y=0,Z=0),Scale3D=(X=1,Y=1,Z=1) ...(truncated)"
    WeightMinimum=0
    Weight=@n1.Result
   )
   > n2 = &RigUnit_SphericalPoseReader(
    ActiveRegionScaleFactors=(PositiveWidth=1,NegativeWidth=1,PositiveHeight=1,NegativeHeight=1)
    ActiveRegionSize=0.05
    Debug=(bDrawDebug=false,bDraw2D=False,bDrawLocalAxes=False,DebugScale=25,DebugSegments=20,DebugThickness=0.25)
    DriverAxis=(X=1,Y=0,Z=0)
    DriverItem=(Type=Bone,Name="head")
    FalloffRegionScaleFactors=(PositiveWidth=1,NegativeWidth=1,PositiveHeight=1,NegativeHeight=1)
    FalloffSize=0.15
    FlipHeightScaling=False
    FlipWidthScaling=False
    OptionalParentItem=(Type=Bone,Name="spine_05")
    RotationOffset=(X=-45,Y=0,Z=70)
   )
   RigVMFunction_MathBoolToFloat = &RigVMFunction_MathBoolToFloat(Value=RigVMVariableNode() {variable=Targeting}.Value)
   AlphaInterp = &RigVMFunction_AlphaInterp(InterpSpeedDecreasing=20, InterpSpeedIncreasing=20, Scale=1, bClampResult=true, Value=@RigVMFunction_MathBoolToFloat.Result)
   Remap = &RigVMFunction_MathDoubleRemap(SourceMaximum=0, SourceMinimum=1, TargetMaximum=1, TargetMinimum=0, Value=@AlphaInterp.Result)
   Multiply_3 = RigVMAggregateNode(C=2.3, A=@n2.OutputParam, B=@Remap.Result)
   > &RigUnit_ModifyTransforms(
    ItemToModify="((Item=(Type=Bone,Name=\"base_collar_a_l\"),Transform=(Rotation=(X=0,Y=0,Z=0,W=1),Translation=(X=0,Y=0,Z=0),Scale3D=(X=1,Y=1,Z=1))),(Item=(Type=Bone,Name=\"base_collar_b_l\"),Transform=(Rotation=(X=0,Y=0,Z=0,W=1),Translation=(X=0,Y=0,Z=0),Scale3D=(X=1,Y=1,Z=1) ...(truncated)"
    WeightMinimum=0
    Weight=@Multiply_3.Result
   )
   > &RigUnit_ModifyTransforms(
    ItemToModify="((Item=(Type=Bone,Name=\"base_collar_a_l\"),Transform=(Rotation=(X=0,Y=0.0463,Z=0,W=-0.9989),Translation=(X=0,Y=0,Z=0),Scale3D=(X=1,Y=1,Z=1))),(Item=(Type=Bone,Name=\"base_collar_b_l\"),Transform=(Rotation=(X=0,Y=0.1642,Z=0,W=-0.9864),Translation=(X=0,Y=0,Z=0) ...(truncated)"
    WeightMinimum=0
    Weight=&RigVMFunction_MathDoubleMul(B=@AlphaInterp.Result, A=@n2.OutputParam).Result
   )
   RigVMFunctionReturnNode(ExecuteContext=@RigVMFunction_Sequence.F)
   n3 = &RigVMFunction_AlphaInterp(
    InterpSpeedDecreasing=10
    InterpSpeedIncreasing=10
    Scale=1
    bClampResult=true
    Value=RigVMVariableNode() {variable=HeadRotateExtremCrouch}.Value
   )
   &RigUnit_ModifyTransforms(
    ItemToModify="((Item=(Type=Bone,Name=\"base_collar_a_l\"),Transform=(Rotation=(X=0,Y=0,Z=0,W=1),Translation=(X=0,Y=0,Z=0),Scale3D=(X=1,Y=1,Z=1))),(Item=(Type=Bone,Name=\"base_collar_b_l\"),Transform=(Rotation=(X=0,Y=0,Z=0,W=1),Translation=(X=0,Y=0,Z=0),Scale3D=(X=1,Y=1,Z=1) ...(truncated)"
    WeightMinimum=0
    Weight=@n3.Result
    ExecutePin=@RigVMFunction_Sequence.E
   )
   &RigUnit_ModifyTransforms(
    ItemToModify="((Item=(Type=Bone,Name=\"base_pendants_bk\"),Transform=(Rotation=(X=0,Y=0,Z=0,W=-1),Translation=(X=0,Y=0,Z=0),Scale3D=(X=1,Y=1,Z=1))),(Item=(Type=Bone,Name=\"base_collar_b_l\"),Transform=(Rotation=(X=0,Y=0,Z=0,W=1),Translation=(X=0,Y=0,Z=0),Scale3D=(X=1,Y=1,Z= ...(truncated)"
    WeightMinimum=0
    Weight=&RigVMFunction_MathDoubleMul(A=@"Delta Rotation".ouputWeight, B=RigVMVariableNode() {variable=Crouching}.Value).Result
    ExecutePin=@RigVMFunction_Sequence.D
   )
   > &RigUnit_ModifyTransforms(
    ItemToModify="((Item=(Type=Bone,Name=\"base_collar_a_l\"),Transform=(Rotation=(X=0,Y=0,Z=0,W=1),Translation=(X=0,Y=0,Z=0),Scale3D=(X=1,Y=1,Z=1))),(Item=(Type=Bone,Name=\"base_collar_b_l\"),Transform=(Rotation=(X=0.0005,Y=-0.0484,Z=0.0095,W=-0.9988),Translation=(X=0,Y=0,Z=0) ...(truncated)"
    WeightMinimum=0
    Weight=&RigVMFunction_MathDoubleMul(A=@Multiply.Result, B=RigVMVariableNode() {variable=Pistol}.Value).Result
   )
  ]}
 New Function_ContainedGraph {nodes [
   n1 = RigVMFunctionEntryNode()
   RigVMFunction_Sequence = &RigVMFunction_Sequence(ExecuteContext=@n1.ExecuteContext)
   RigVMFunction_MathBoolToFloat = &RigVMFunction_MathBoolToFloat(Value=RigVMVariableNode() {variable=CrouchingBool}.Value)
   AlphaInterp = &RigVMFunction_AlphaInterp(InterpSpeedDecreasing=10, InterpSpeedIncreasing=10, Scale=1, bClampResult=true, Value=@RigVMFunction_MathBoolToFloat.Result)
   RigVMVariableNode(Value=@AlphaInterp.Result, ExecuteContext=@RigVMFunction_Sequence.A) {variable=Crouching}
   > RigVMVariableNode(Value=&RigVMFunction_MathDoubleRemap(SourceMaximum=1, SourceMinimum=0, TargetMaximum=0, TargetMinimum=1, Value=@AlphaInterp.Result).Result) {variable=Standing}
   n2 = &RigVMFunction_AlphaInterp(
    InterpSpeedDecreasing=10
    InterpSpeedIncreasing=10
    Scale=1
    bClampResult=true
    Value=&RigVMFunction_MathBoolToFloat(Value=RigVMVariableNode() {variable=Emotes}.Value).Result
   )
   > RigVMVariableNode(Value=@n2.Result) {variable=AlphaEmotes}
   RigVMFunctionReturnNode(ExecuteContext=@RigVMFunction_Sequence.B)
  ]}
 New Function_ContainedGraph {nodes [
   RigVMFunctionEntryNode()
   > RigVMFunction_Sequence = RigVMAggregateNode()
   &RigUnit_ModifyTransforms(
    ItemToModify="((Item=(Type=Bone,Name=\"base_skirt_fr_l\"),Transform=(Rotation=(X=-0.0021,Y=-0.001,Z=-0.5018,W=-0.865),Translation=(X=0,Y=0,Z=0),Scale3D=(X=1,Y=1,Z=1))),(Item=(Type=Bone,Name=\"base_skirt_sd_l\"),Transform=(Rotation=(X=0.1577,Y=0.0751,Z=0.0385,W=0.9839),Trans ...(truncated)"
    WeightMinimum=0
    Weight=&RigVMFunction_MathDoubleMul(B=RigVMVariableNode() {variable=Standing}.Value, A=RigVMVariableNode() {variable=Leg Front L}.Value).Result
    ExecutePin=@RigVMFunction_Sequence.C
   )
   > &RigUnit_ModifyTransforms(
    ItemToModify="((Item=(Type=Bone,Name=\"base_skirt_fr_l\"),Transform=(Rotation=(X=0.0765,Y=0.0279,Z=0.0235,W=0.9964),Translation=(X=0,Y=0,Z=0),Scale3D=(X=1,Y=1,Z=1))),(Item=(Type=Bone,Name=\"base_skirt_sd_l\"),Transform=(Rotation=(X=0.0973,Y=-0.0333,Z=-0.2915,W=0.951),Transl ...(truncated)"
    WeightMinimum=0
    Weight=&RigVMFunction_MathDoubleMul(B=RigVMVariableNode() {variable=Standing}.Value, A=RigVMVariableNode() {variable=Leg FrontSide L}.Value).Result
   )
   n1 = &RigVMFunction_MathDoubleMul(B=RigVMVariableNode() {variable=Standing}.Value, A=RigVMVariableNode() {variable=Leg Side L}.Value)
   > &RigUnit_ModifyTransforms(
    ItemToModify="((Item=(Type=Bone,Name=\"base_skirt_fr_l\"),Transform=(Rotation=(X=-0.3744,Y=-0.6297,Z=0.0285,W=0.6801),Translation=(X=1.9249,Y=3.194,Z=-2.4247),Scale3D=(X=1,Y=1,Z=1))),(Item=(Type=Bone,Name=\"base_skirt_sd_l\"),Transform=(Rotation=(X=0.4069,Y=-0.4031,Z=0.3188 ...(truncated)"
    WeightMinimum=0
    Weight=@n1.Result
   )
   Remap = &RigVMFunction_MathDoubleRemap(SourceMaximum=1, SourceMinimum=0.7, TargetMaximum=1, TargetMinimum=0, Value=@n1.Result)
   > &RigUnit_ModifyTransforms(
    ItemToModify="((Item=(Type=Bone,Name=\"base_skirt_fr_l\"),Transform=(Rotation=(X=0.0341,Y=0.1684,Z=-0.1455,W=0.9743),Translation=(X=0,Y=0,Z=0),Scale3D=(X=1,Y=1,Z=1))),(Item=(Type=Bone,Name=\"base_skirt_sd_l\"),Transform=(Rotation=(X=-0.008,Y=0.0404,Z=-0.3531,W=0.9347),Trans ...(truncated)"
    WeightMinimum=0
    Weight=@Remap.Result
   )
   > &RigUnit_ModifyTransforms(
    ItemToModify="((Item=(Type=Bone,Name=\"base_skirt_fr_l\"),Transform=(Rotation=(X=-0.1934,Y=-0.0462,Z=-0.0557,W=0.9784),Translation=(X=0,Y=0,Z=0),Scale3D=(X=1,Y=1,Z=1))),(Item=(Type=Bone,Name=\"base_skirt_sd_l\"),Transform=(Rotation=(X=-0.1606,Y=-0.1275,Z=0.3642,W=0.9084),Tr ...(truncated)"
    WeightMinimum=0
    Weight=&RigVMFunction_MathDoubleMul(B=RigVMVariableNode() {variable=Standing}.Value, A=RigVMVariableNode() {variable=Leg BackSide L}.Value).Result
   )
   > &RigUnit_ModifyTransforms(
    ItemToModify="((Item=(Type=Bone,Name=\"base_skirt_fr_l\"),Transform=(Rotation=(X=0.0015,Y=-0.0602,Z=-0.2587,W=0.9641),Translation=(X=2.158,Y=3.7493,Z=0.1471),Scale3D=(X=1,Y=1,Z=1))),(Item=(Type=Bone,Name=\"base_skirt_sd_l\"),Transform=(Rotation=(X=-0.4339,Y=-0.4318,Z=-0.441 ...(truncated)"
    WeightMinimum=0
    Weight=&RigVMFunction_MathDoubleMul(B=RigVMVariableNode() {variable=Standing}.Value, A=RigVMVariableNode() {variable=Leg Back L}.Value).Result
   )
   > &RigUnit_ModifyTransforms(
    ItemToModify="((Item=(Type=Bone,Name=\"base_skirt_fr_l\"),Transform=(Rotation=(X=0,Y=0,Z=0,W=1),Translation=(X=0,Y=0,Z=0),Scale3D=(X=1,Y=1,Z=1))),(Item=(Type=Bone,Name=\"dyn_skirt_fr_l_02\"),Transform=(Rotation=(X=0,Y=0,Z=0,W=1),Translation=(X=-7.4287,Y=4.011,Z=-2.5606),Sca ...(truncated)"
    WeightMinimum=0
    Weight=&RigVMFunction_MathDoubleMul(A=RigVMVariableNode() {variable=Leg FrontInside L}.Value, B=RigVMVariableNode() {variable=Standing}.Value).Result
   )
   > &RigUnit_ModifyTransforms(
    ItemToModify="((Item=(Type=Bone,Name=\"base_skirt_fr_r\"),Transform=(Rotation=(X=0.2689,Y=0.0535,Z=0.6602,W=-0.6993),Translation=(X=-3.7412,Y=0.7142,Z=0.3702),Scale3D=(X=1,Y=1,Z=1))),(Item=(Type=Bone,Name=\"base_skirt_fr_l\"),Transform=(Rotation=(X=0,Y=0,Z=0.0748,W=0.9972), ...(truncated)"
    WeightMinimum=0
    Weight=&RigVMFunction_MathDoubleMul(A=RigVMVariableNode() {variable=Leg Front R}.Value, B=RigVMVariableNode() {variable=Standing}.Value).Result
   )
   > &RigUnit_ModifyTransforms(
    ItemToModify="((Item=(Type=Bone,Name=\"base_skirt_fr_r\"),Transform=(Rotation=(X=-0.3939,Y=-0.508,Z=-0.5546,W=0.5284),Translation=(X=0,Y=0,Z=0),Scale3D=(X=1,Y=1,Z=1))),(Item=(Type=Bone,Name=\"base_skirt_sd_r\"),Transform=(Rotation=(X=-0.3934,Y=-0.242,Z=0.0396,W=0.8861),Tran ...(truncated)"
    WeightMinimum=0
    Weight=&RigVMFunction_MathDoubleMul(A=RigVMVariableNode() {variable=Leg Side R}.Value, B=RigVMVariableNode() {variable=Standing}.Value).Result
   )
   > &RigUnit_ModifyTransforms(
    ItemToModify="((Item=(Type=Bone,Name=\"base_skirt_fr_r\"),Transform=(Rotation=(X=0.0888,Y=-0.1438,Z=-0.373,W=0.9123),Translation=(X=0,Y=0,Z=0),Scale3D=(X=1,Y=1,Z=1))),(Item=(Type=Bone,Name=\"base_skirt_sd_r\"),Transform=(Rotation=(X=0.0085,Y=-0.2884,Z=-0.591,W=0.7533),Trans ...(truncated)"
    WeightMinimum=0
    Weight=&RigVMFunction_MathDoubleMul(A=RigVMVariableNode() {variable=Leg BackSide R}.Value, B=RigVMVariableNode() {variable=Standing}.Value).Result
   )
   > &RigUnit_ModifyTransforms(
    ItemToModify="((Item=(Type=Bone,Name=\"base_skirt_fr_r\"),Transform=(Rotation=(X=0.0357,Y=-0.1143,Z=0.2961,W=0.9476),Translation=(X=0,Y=0,Z=0),Scale3D=(X=1,Y=1,Z=1))),(Item=(Type=Bone,Name=\"base_skirt_sd_r\"),Transform=(Rotation=(X=0.1817,Y=-0.5491,Z=0.2734,W=0.7686),Trans ...(truncated)"
    WeightMinimum=0
    Weight=&RigVMFunction_MathDoubleMul(A=RigVMVariableNode() {variable=Leg Back R}.Value, B=RigVMVariableNode() {variable=Standing}.Value).Result
   )
   > &RigUnit_ModifyTransforms(
    ItemToModify="((Item=(Type=Bone,Name=\"base_skirt_fr_l\"),Transform=(Rotation=(X=0,Y=0,Z=0,W=1),Translation=(X=-2.0365,Y=1.1489,Z=2.0956),Scale3D=(X=1,Y=1,Z=1))),(Item=(Type=Bone,Name=\"dyn_skirt_fr_l_02\"),Transform=(Rotation=(X=0,Y=0,Z=0,W=1),Translation=(X=-1.0912,Y=1.41 ...(truncated)"
    WeightMinimum=0
    Weight=&RigVMFunction_MathDoubleMul(A=RigVMVariableNode() {variable=Leg FrontInside R}.Value, B=RigVMVariableNode() {variable=Standing}.Value).Result
   )
   Multiply = RigVMAggregateNode(
    A=RigVMVariableNode() {variable=Leg Back L}.Value
    B=RigVMVariableNode() {variable=Leg Back R}.Value
    C=RigVMVariableNode() {variable=Standing}.Value
   )
   RigVMFunction_AnimEvalRichCurve = &RigVMFunction_AnimEvalRichCurve(
    Curve=(EditorCurveData=(Keys=((),(InterpMode=RCIM_Cubic,Time=0.2988,Value=0.428,ArriveTangent=1,LeaveTangent=1),(Time=1,Value=1))))
    SourceMinimum=0
    TargetMinimum=0
    Value=@Multiply.Result
   )
   > &RigUnit_ModifyTransforms(
    ItemToModify=((Item=(Type=Bone,Name="base_skirt_bk_c"),Transform=(Rotation=(X=0,Y=0.8552,Z=0,W=0.5183),Translation=(X=6.2454,Y=0,Z=-11.7605),Scale3D=(X=1,Y=1,Z=1))))
    WeightMinimum=0
    Weight=@RigVMFunction_AnimEvalRichCurve.Result
   )
   Multiply_1 = RigVMAggregateNode(
    A=RigVMVariableNode() {variable=Leg Front L}.Value
    B=RigVMVariableNode() {variable=Leg Front R}.Value
    C=RigVMVariableNode() {variable=Standing}.Value
   )
   > &RigUnit_ModifyTransforms(
    ItemToModify=((Item=(Type=Bone,Name="base_skirt_fr_l"),Transform=(Rotation=(X=0,Y=0,Z=-0.1902,W=0.9817),Translation=(X=0,Y=0,Z=0),Scale3D=(X=1,Y=1,Z=1))))
    WeightMinimum=0
    Weight=@Multiply_1.Result
   )
   RigVMFunctionReturnNode(ExecuteContext=@RigVMFunction_Sequence.F)
   Delta Rotation = RigVMFunctionReferenceNode(
    Rotation Order=ZYX
    child=calf_l
    debugRotation=False
    max=130
    onlyTwist=False
    parent=calf_l_null
    parentType=Null
    premult=1
    primAxis=2
    twistAxis=(X=1,Y=0,Z=0)
    ExecuteContext=@RigVMFunction_Sequence.A
   )
   n2 = &RigVMFunction_MathDoubleMul(
    A=@"Delta Rotation".ouputWeight
    B=&RigVMFunction_MathDoubleRemap(SourceMaximum=0.5, SourceMinimum=0, TargetMaximum=0, TargetMinimum=1, Value=RigVMVariableNode() {variable=Leg Front L}.Value).Result
   )
   > &RigUnit_ModifyTransforms(
    ItemToModify="((Item=(Type=Bone,Name=\"dyn_skirt_bk_02_c\"),Transform=(Rotation=(X=0,Y=-0.8431,Z=0,W=0.5377),Translation=(X=-11.3773,Y=0,Z=-0.9073),Scale3D=(X=1,Y=1,Z=1))),(Item=(Type=Bone,Name=\"dyn_skirt_bk_03_c\"),Transform=(Rotation=(X=0.004,Y=-0.2585,Z=0.0149,W=-0.9659 ...(truncated)"
    WeightMinimum=0
    Weight=@n2.Result
   )
   > &RigUnit_ModifyTransforms(
    ItemToModify="((Item=(Type=Bone,Name=\"dyn_skirt_bk_02_c\"),Transform=(Rotation=(X=0,Y=0,Z=0,W=1),Translation=(X=0,Y=0,Z=0),Scale3D=(X=1,Y=1,Z=1))),(Item=(Type=Bone,Name=\"dyn_skirt_bk_03_c\"),Transform=(Rotation=(X=0,Y=0,Z=0,W=1),Translation=(X=0,Y=0,Z=0),Scale3D=(X=1,Y=1, ...(truncated)"
    WeightMinimum=0
    Weight=&RigVMFunction_MathDoubleMul(A=@n2.Result, B=RigVMVariableNode() {variable=Crouching}.Value).Result
   )
   > Delta Rotation_1 = RigVMFunctionReferenceNode(
    Rotation Order=ZYX
    child=calf_r
    debugRotation=False
    max=130
    onlyTwist=False
    parent=calf_r_null
    parentType=Null
    premult=1
    primAxis=2
    twistAxis=(X=1,Y=0,Z=0)
   )
   n3 = &RigVMFunction_MathDoubleRemap(SourceMaximum=0.5, SourceMinimum=0, TargetMaximum=0, TargetMinimum=1, Value=RigVMVariableNode() {variable=Leg Front R}.Value)
   > &RigUnit_ModifyTransforms(
    ItemToModify="((Item=(Type=Bone,Name=\"dyn_skirt_bk_02_c\"),Transform=(Rotation=(X=-0.0802,Y=-0.6395,Z=0.0951,W=0.7587),Translation=(X=-17.5416,Y=-0.0004,Z=4.1833),Scale3D=(X=1,Y=1,Z=1))),(Item=(Type=Bone,Name=\"dyn_skirt_bk_03_c\"),Transform=(Rotation=(X=0,Y=0.3575,Z=0,W=- ...(truncated)"
    WeightMinimum=0
    Weight=&RigVMFunction_MathDoubleMul(A=@"Delta Rotation_1".ouputWeight, B=@n3.Result).Result
   )
   > &RigUnit_ModifyTransforms(
    ItemToModify=((Item=(Type=Bone,Name="dyn_skirt_sd_r_03"),Transform=(Rotation=(X=0,Y=0.416,Z=0,W=0.9094),Translation=(X=0,Y=0,Z=0),Scale3D=(X=1,Y=1,Z=1))))
    WeightMinimum=0
    Weight=&RigVMFunction_MathDoubleMul(A=@n3.Result, B=RigVMVariableNode() {variable=Crouching}.Value).Result
   )
   Multiply_26 = RigVMAggregateNode(
    B=@"Delta Rotation".ouputWeight
    A=@"Delta Rotation_1".ouputWeight
    C=&RigVMFunction_MathDoubleRemap(SourceMaximum=0.5, SourceMinimum=0, TargetMaximum=0, TargetMinimum=1, Value=RigVMVariableNode() {variable=Leg Front R}.Value).Result
   )
   > &RigUnit_ModifyTransforms(
    ItemToModify="((Item=(Type=Bone,Name=\"dyn_skirt_bk_02_c\"),Transform=(Rotation=(X=0,Y=0.6803,Z=0,W=0.7329),Translation=(X=0,Y=0,Z=0),Scale3D=(X=1,Y=1,Z=1))),(Item=(Type=Bone,Name=\"dyn_skirt_bk_03_c\"),Transform=(Rotation=(X=0,Y=0.0958,Z=0,W=-0.9954),Translation=(X=11.1758 ...(truncated)"
    WeightMinimum=0
    Weight=@Multiply_26.Result
   )
   n4 = RigVMVariableNode() {variable=Crouching}
   &RigUnit_ModifyTransforms(
    ItemToModify="((Item=(Type=Bone,Name=\"base_skirt_fr_l\"),Transform=(Rotation=(X=0,Y=0,Z=0,W=1),Translation=(X=0,Y=0,Z=0),Scale3D=(X=1,Y=1,Z=1))),(Item=(Type=Bone,Name=\"base_skirt_sd_l\"),Transform=(Rotation=(X=0.0072,Y=0.2503,Z=-0.0774,W=-0.965),Translation=(X=0,Y=0,Z=0), ...(truncated)"
    WeightMinimum=0
    Weight=@n4.Value
    ExecutePin=@RigVMFunction_Sequence.D
   )
   n5 = &RigVMFunction_AlphaInterp(
    InterpSpeedDecreasing=10
    InterpSpeedIncreasing=10
    Scale=1
    bClampResult=true
    Value=&RigVMFunction_MathBoolToFloat(Value=RigVMVariableNode() {variable=Moving2D}.Value).Result
   )
   > &RigUnit_ModifyTransforms(
    ItemToModify="((Item=(Type=Bone,Name=\"base_skirt_fr_l\"),Transform=(Rotation=(X=0,Y=0,Z=0,W=1),Translation=(X=0,Y=0,Z=0),Scale3D=(X=1,Y=1,Z=1))),(Item=(Type=Bone,Name=\"base_skirt_sd_l\"),Transform=(Rotation=(X=0.0132,Y=-0.0113,Z=-0.1541,W=-0.9879),Translation=(X=0,Y=0,Z=0 ...(truncated)"
    WeightMinimum=0
    Weight=&RigVMFunction_MathDoubleMul(B=@n5.Result, A=@n4.Value).Result
   )
   > &RigUnit_ModifyTransforms(
    ItemToModify="((Item=(Type=Bone,Name=\"base_skirt_fr_l\"),Transform=(Rotation=(X=0,Y=0,Z=-0.4775,W=-0.8786),Translation=(X=3.3605,Y=-1.6514,Z=-0.9204),Scale3D=(X=1,Y=1,Z=1))),(Item=(Type=Bone,Name=\"base_skirt_sd_l\"),Transform=(Rotation=(X=-0.0469,Y=-0.1146,Z=0.2801,W=-0.9 ...(truncated)"
    WeightMinimum=0
    Weight=&RigVMFunction_MathDoubleMul(A=RigVMVariableNode() {variable=Leg Front L}.Value, B=RigVMVariableNode() {variable=Crouching}.Value).Result
   )
   > &RigUnit_ModifyTransforms(
    ItemToModify="((Item=(Type=Bone,Name=\"base_skirt_fr_l\"),Transform=(Rotation=(X=0.078,Y=-0.0472,Z=0.0177,W=0.9957),Translation=(X=0,Y=0,Z=0),Scale3D=(X=1,Y=1,Z=1))),(Item=(Type=Bone,Name=\"base_skirt_sd_l\"),Transform=(Rotation=(X=0.0973,Y=-0.0333,Z=-0.2915,W=0.951),Transl ...(truncated)"
    WeightMinimum=0
    Weight=&RigVMFunction_MathDoubleMul(A=RigVMVariableNode() {variable=Leg FrontSide L}.Value, B=RigVMVariableNode() {variable=Crouching}.Value).Result
   )
   n6 = &RigVMFunction_MathDoubleMul(A=RigVMVariableNode() {variable=Leg Side L}.Value, B=RigVMVariableNode() {variable=Crouching}.Value)
   > &RigUnit_ModifyTransforms(
    ItemToModify="((Item=(Type=Bone,Name=\"base_skirt_fr_l\"),Transform=(Rotation=(X=-0.3744,Y=-0.6297,Z=0.0285,W=0.6801),Translation=(X=1.9249,Y=3.194,Z=-2.4247),Scale3D=(X=1,Y=1,Z=1))),(Item=(Type=Bone,Name=\"base_skirt_sd_l\"),Transform=(Rotation=(X=0.4069,Y=-0.4031,Z=0.3188 ...(truncated)"
    WeightMinimum=0
    Weight=@n6.Result
   )
   > &RigUnit_ModifyTransforms(
    ItemToModify="((Item=(Type=Bone,Name=\"base_skirt_fr_l\"),Transform=(Rotation=(X=0.0153,Y=0.0432,Z=-0.1487,W=0.9878),Translation=(X=0,Y=0,Z=0),Scale3D=(X=1,Y=1,Z=1))),(Item=(Type=Bone,Name=\"base_skirt_sd_l\"),Transform=(Rotation=(X=-0.008,Y=0.0404,Z=-0.3531,W=0.9347),Trans ...(truncated)"
    WeightMinimum=0
    Weight=&RigVMFunction_MathDoubleRemap(SourceMaximum=1, SourceMinimum=0.6, TargetMaximum=1, TargetMinimum=0, Value=@n6.Result).Result
   )
   > &RigUnit_ModifyTransforms(
    ItemToModify="((Item=(Type=Bone,Name=\"base_skirt_fr_l\"),Transform=(Rotation=(X=0.0015,Y=-0.0602,Z=-0.2587,W=0.9641),Translation=(X=2.158,Y=3.7493,Z=0.1471),Scale3D=(X=1,Y=1,Z=1))),(Item=(Type=Bone,Name=\"base_skirt_sd_l\"),Transform=(Rotation=(X=-0.4339,Y=-0.4318,Z=-0.441 ...(truncated)"
    WeightMinimum=0
    Weight=&RigVMFunction_MathDoubleMul(A=RigVMVariableNode() {variable=Leg Back L}.Value, B=RigVMVariableNode() {variable=Crouching}.Value).Result
   )
   > &RigUnit_ModifyTransforms(
    ItemToModify="((Item=(Type=Bone,Name=\"base_skirt_fr_l\"),Transform=(Rotation=(X=0,Y=0,Z=0,W=1),Translation=(X=0,Y=0,Z=0),Scale3D=(X=1,Y=1,Z=1))),(Item=(Type=Bone,Name=\"dyn_skirt_fr_l_02\"),Transform=(Rotation=(X=0,Y=0,Z=0,W=1),Translation=(X=-9.2389,Y=5.8047,Z=2.1814),Sca ...(truncated)"
    WeightMinimum=0
    Weight=&RigVMFunction_MathDoubleMul(A=RigVMVariableNode() {variable=Leg FrontInside L}.Value, B=RigVMVariableNode() {variable=Crouching}.Value).Result
   )
   > &RigUnit_ModifyTransforms(
    ItemToModify="((Item=(Type=Bone,Name=\"base_skirt_fr_r\"),Transform=(Rotation=(X=0.2689,Y=0.0535,Z=0.6602,W=-0.6993),Translation=(X=-3.7412,Y=0.7142,Z=0.3702),Scale3D=(X=1,Y=1,Z=1))),(Item=(Type=Bone,Name=\"base_skirt_fr_l\"),Transform=(Rotation=(X=0,Y=0,Z=0.0748,W=0.9972), ...(truncated)"
    WeightMinimum=0
    Weight=&RigVMFunction_MathDoubleMul(A=RigVMVariableNode() {variable=Leg Front R}.Value, B=RigVMVariableNode() {variable=Crouching}.Value).Result
   )
   > &RigUnit_ModifyTransforms(
    ItemToModify="((Item=(Type=Bone,Name=\"base_skirt_fr_r\"),Transform=(Rotation=(X=-0.3939,Y=-0.508,Z=-0.5546,W=0.5284),Translation=(X=0,Y=0,Z=0),Scale3D=(X=1,Y=1,Z=1))),(Item=(Type=Bone,Name=\"base_skirt_sd_r\"),Transform=(Rotation=(X=-0.3934,Y=-0.242,Z=0.0396,W=0.8861),Tran ...(truncated)"
    WeightMinimum=0
    Weight=&RigVMFunction_MathDoubleMul(A=RigVMVariableNode() {variable=Leg Side R}.Value, B=RigVMVariableNode() {variable=Crouching}.Value).Result
   )
   > &RigUnit_ModifyTransforms(
    ItemToModify="((Item=(Type=Bone,Name=\"base_skirt_fr_r\"),Transform=(Rotation=(X=0.0888,Y=-0.1438,Z=-0.373,W=0.9123),Translation=(X=0,Y=0,Z=0),Scale3D=(X=1,Y=1,Z=1))),(Item=(Type=Bone,Name=\"base_skirt_sd_r\"),Transform=(Rotation=(X=0.0085,Y=-0.2884,Z=-0.591,W=0.7533),Trans ...(truncated)"
    WeightMinimum=0
    Weight=&RigVMFunction_MathDoubleMul(A=RigVMVariableNode() {variable=Leg BackSide R}.Value, B=RigVMVariableNode() {variable=Crouching}.Value).Result
   )
   > &RigUnit_ModifyTransforms(
    ItemToModify="((Item=(Type=Bone,Name=\"base_skirt_fr_r\"),Transform=(Rotation=(X=0.0357,Y=-0.1143,Z=0.2961,W=0.9476),Translation=(X=0,Y=0,Z=0),Scale3D=(X=1,Y=1,Z=1))),(Item=(Type=Bone,Name=\"base_skirt_sd_r\"),Transform=(Rotation=(X=0.1817,Y=-0.5491,Z=0.2734,W=0.7686),Trans ...(truncated)"
    WeightMinimum=0
    Weight=&RigVMFunction_MathDoubleMul(A=RigVMVariableNode() {variable=Leg Back R}.Value, B=RigVMVariableNode() {variable=Crouching}.Value).Result
   )
   > &RigUnit_ModifyTransforms(
    ItemToModify="((Item=(Type=Bone,Name=\"base_skirt_fr_l\"),Transform=(Rotation=(X=0,Y=0,Z=0,W=1),Translation=(X=0,Y=0,Z=0),Scale3D=(X=1,Y=1,Z=1))),(Item=(Type=Bone,Name=\"dyn_skirt_fr_l_02\"),Transform=(Rotation=(X=0,Y=0,Z=0,W=1),Translation=(X=0,Y=0,Z=2.6234),Scale3D=(X=1,Y ...(truncated)"
    WeightMinimum=0
    Weight=&RigVMFunction_MathDoubleMul(A=RigVMVariableNode() {variable=Leg FrontInside R}.Value, B=RigVMVariableNode() {variable=Crouching}.Value).Result
   )
   n7 = &RigVMFunction_AlphaInterp(
    InterpSpeedDecreasing=10
    InterpSpeedIncreasing=10
    Scale=1
    bClampResult=true
    Value=&RigVMFunction_MathDoubleMul(A=RigVMVariableNode() {variable=Idle}.Value, B=RigVMVariableNode() {variable=Crouching}.Value).Result
   )
   > &RigUnit_ModifyTransforms(
    ItemToModify="((Item=(Type=Bone,Name=\"base_skirt_fr_l\"),Transform=(Rotation=(X=-0.0007,Y=-0.0069,Z=0.0999,W=-0.995),Translation=(X=-0.0283,Y=0.1367,Z=0.2468),Scale3D=(X=1,Y=1,Z=1))),(Item=(Type=Bone,Name=\"base_skirt_sd_l\"),Transform=(Rotation=(X=0,Y=0,Z=0,W=1),Translati ...(truncated)"
    WeightMinimum=0
    Weight=@n7.Result
   )
   Multiply_21 = RigVMAggregateNode(
    A=RigVMVariableNode() {variable=Leg Front L}.Value
    B=RigVMVariableNode() {variable=Leg Front R}.Value
    C=RigVMVariableNode() {variable=Crouching}.Value
   )
   > &RigUnit_ModifyTransforms(
    ItemToModify=((Item=(Type=Bone,Name="base_skirt_fr_l"),Transform=(Rotation=(X=0,Y=0,Z=-0.1902,W=0.9817),Translation=(X=0,Y=0,Z=0),Scale3D=(X=1,Y=1,Z=1))))
    WeightMinimum=0
    Weight=@Multiply_21.Result
   )
   RigVMFunction_MathBoolToFloat = &RigVMFunction_MathBoolToFloat(Value=RigVMVariableNode() {variable=DBNO}.Value)
   AlphaInterp = &RigVMFunction_AlphaInterp(InterpSpeedDecreasing=10, InterpSpeedIncreasing=10, Scale=1, bClampResult=true, Value=@RigVMFunction_MathBoolToFloat.Result)
   &RigUnit_ModifyTransforms(
    ItemToModify="((Item=(Type=Bone,Name=\"base_skirt_fr_l\"),Transform=(Rotation=(X=0,Y=0,Z=0,W=1),Translation=(X=0,Y=0,Z=0),Scale3D=(X=1,Y=1,Z=1))),(Item=(Type=Bone,Name=\"base_skirt_sd_l\"),Transform=(Rotation=(X=-0.1166,Y=-0.0433,Z=0.0549,W=0.9907),Translation=(X=0,Y=0,Z=0) ...(truncated)"
    WeightMinimum=0
    Weight=&RigVMFunction_MathDoubleMul(B=@AlphaInterp.Result, A=RigVMVariableNode() {variable=Leg Front L}.Value).Result
    ExecutePin=@RigVMFunction_Sequence.E
   )
   > &RigUnit_ModifyTransforms(
    ItemToModify="((Item=(Type=Bone,Name=\"base_skirt_fr_r\"),Transform=(Rotation=(X=0,Y=0,Z=0,W=1),Translation=(X=0,Y=0,Z=0),Scale3D=(X=1,Y=1,Z=1))),(Item=(Type=Bone,Name=\"base_skirt_fr_l\"),Transform=(Rotation=(X=0,Y=0,Z=0,W=1),Translation=(X=0,Y=0,Z=0),Scale3D=(X=1,Y=1,Z=1) ...(truncated)"
    WeightMinimum=0
    Weight=&RigVMFunction_MathDoubleMul(B=@AlphaInterp.Result, A=RigVMVariableNode() {variable=Leg Front R}.Value).Result
   )
   > &RigUnit_ModifyTransforms(
    ItemToModify="((Item=(Type=Bone,Name=\"base_skirt_fr_l\"),Transform=(Rotation=(X=0,Y=0,Z=0,W=1),Translation=(X=0,Y=0,Z=0),Scale3D=(X=1,Y=1,Z=1))),(Item=(Type=Bone,Name=\"base_skirt_sd_l\"),Transform=(Rotation=(X=0.0908,Y=-0.3898,Z=0.0337,W=-0.9158),Translation=(X=0,Y=0,Z=0) ...(truncated)"
    WeightMinimum=0
    Weight=&RigVMFunction_MathDoubleMul(B=@AlphaInterp.Result, A=RigVMVariableNode() {variable=Leg Side L}.Value).Result
   )
   > &RigUnit_ModifyTransforms(
    ItemToModify="((Item=(Type=Bone,Name=\"base_skirt_fr_r\"),Transform=(Rotation=(X=0,Y=0,Z=0,W=1),Translation=(X=0,Y=0,Z=0),Scale3D=(X=1,Y=1,Z=1))),(Item=(Type=Bone,Name=\"base_skirt_sd_r\"),Transform=(Rotation=(X=0,Y=0.1816,Z=0,W=0.9834),Translation=(X=0,Y=0,Z=0),Scale3D=(X= ...(truncated)"
    WeightMinimum=0
    Weight=&RigVMFunction_MathDoubleMul(B=@AlphaInterp.Result, A=RigVMVariableNode() {variable=Leg Side R}.Value).Result
   )
  ]}
 New Function_ContainedGraph {nodes [
   RigVMFunctionEntryNode()
   > RigVMFunction_Sequence = RigVMAggregateNode()
   n1 = &RigUnit_SphericalPoseReader(
    ActiveRegionScaleFactors=(PositiveWidth=1,NegativeWidth=1,PositiveHeight=1,NegativeHeight=1)
    ActiveRegionSize=0.05
    Debug=(bDrawDebug=false,bDraw2D=False,bDrawLocalAxes=False,DebugScale=25,DebugSegments=20,DebugThickness=0.25)
    DriverAxis=(X=-1,Y=0,Z=0)
    DriverItem=(Type=Bone,Name="thigh_l")
    FalloffRegionScaleFactors=(PositiveWidth=1,NegativeWidth=1.12,PositiveHeight=1,NegativeHeight=1)
    FalloffSize=0.55
    FlipHeightScaling=False
    FlipWidthScaling=false
    OptionalParentItem=(Type=Bone,Name="None")
    RotationOffset=(X=90,Y=0,Z=-20)
    ExecutePin=@RigVMFunction_Sequence.A
   )
   > n2 = &RigUnit_SphericalPoseReader(
    ActiveRegionScaleFactors=(PositiveWidth=1,NegativeWidth=1,PositiveHeight=1,NegativeHeight=1)
    ActiveRegionSize=0.05
    Debug=(bDrawDebug=false,bDraw2D=False,bDrawLocalAxes=False,DebugScale=25,DebugSegments=20,DebugThickness=0.25)
    DriverAxis=(X=-1,Y=0,Z=0)
    DriverItem=(Type=Bone,Name="thigh_l")
    FalloffRegionScaleFactors=(PositiveWidth=1,NegativeWidth=1,PositiveHeight=1,NegativeHeight=1)
    FalloffSize=0.25
    FlipHeightScaling=False
    FlipWidthScaling=False
    OptionalParentItem=(Type=Bone,Name="None")
    RotationOffset=(X=140,Y=30,Z=0)
   )
   > n3 = &RigUnit_SphericalPoseReader(
    ActiveRegionScaleFactors=(PositiveWidth=1,NegativeWidth=1,PositiveHeight=1,NegativeHeight=1)
    ActiveRegionSize=0.05
    Debug=(bDrawDebug=false,bDraw2D=False,bDrawLocalAxes=False,DebugScale=25,DebugSegments=20,DebugThickness=0.25)
    DriverAxis=(X=-1,Y=0,Z=0)
    DriverItem=(Type=Bone,Name="thigh_l")
    FalloffRegionScaleFactors=(PositiveWidth=1,NegativeWidth=1.15,PositiveHeight=1,NegativeHeight=1)
    FalloffSize=0.55
    FlipHeightScaling=False
    FlipWidthScaling=False
    OptionalParentItem=(Type=Bone,Name="None")
    RotationOffset=(X=180,Y=20,Z=0)
   )
   > n4 = &RigUnit_SphericalPoseReader(
    ActiveRegionScaleFactors=(PositiveWidth=1,NegativeWidth=1,PositiveHeight=1,NegativeHeight=1)
    ActiveRegionSize=0.05
    Debug=(bDrawDebug=false,bDraw2D=False,bDrawLocalAxes=False,DebugScale=25,DebugSegments=20,DebugThickness=0.25)
    DriverAxis=(X=-1,Y=0,Z=0)
    DriverItem=(Type=Bone,Name="thigh_l")
    FalloffRegionScaleFactors=(PositiveWidth=1,NegativeWidth=1.5,PositiveHeight=1,NegativeHeight=1.26)
    FalloffSize=0.26
    FlipHeightScaling=true
    FlipWidthScaling=false
    OptionalParentItem=(Type=Bone,Name="None")
    RotationOffset=(X=-140,Y=20,Z=0)
   )
   > n5 = &RigUnit_SphericalPoseReader(
    ActiveRegionScaleFactors=(PositiveWidth=1,NegativeWidth=1,PositiveHeight=1,NegativeHeight=1)
    ActiveRegionSize=0.05
    Debug=(bDrawDebug=false,bDraw2D=False,bDrawLocalAxes=False,DebugScale=25,DebugSegments=20,DebugThickness=0.25)
    DriverAxis=(X=-1,Y=0,Z=0)
    DriverItem=(Type=Bone,Name="thigh_l")
    FalloffRegionScaleFactors=(PositiveWidth=1,NegativeWidth=1.1,PositiveHeight=1,NegativeHeight=1)
    FalloffSize=0.55
    FlipHeightScaling=False
    FlipWidthScaling=False
    OptionalParentItem=(Type=Bone,Name="None")
    RotationOffset=(X=-90,Y=0,Z=20)
   )
   > n6 = &RigUnit_SphericalPoseReader(
    ActiveRegionScaleFactors=(PositiveWidth=1,NegativeWidth=1,PositiveHeight=1,NegativeHeight=1)
    ActiveRegionSize=0.05
    Debug=(bDrawDebug=false,bDraw2D=False,bDrawLocalAxes=False,DebugScale=25,DebugSegments=20,DebugThickness=0.25)
    DriverAxis=(X=-1,Y=0,Z=0)
    DriverItem=(Type=Bone,Name="thigh_l")
    FalloffRegionScaleFactors=(PositiveWidth=2,NegativeWidth=2,PositiveHeight=1,NegativeHeight=2)
    FalloffSize=0.2
    FlipHeightScaling=true
    FlipWidthScaling=true
    OptionalParentItem=(Type=Bone,Name="None")
    RotationOffset=(X=50,Y=0,Z=0)
   )
   > n7 = &RigUnit_SphericalPoseReader(
    ActiveRegionScaleFactors=(PositiveWidth=1,NegativeWidth=1,PositiveHeight=1,NegativeHeight=1)
    ActiveRegionSize=0.05
    Debug=(bDrawDebug=false,bDraw2D=False,bDrawLocalAxes=False,DebugScale=25,DebugSegments=20,DebugThickness=0.25)
    DriverAxis=(X=1,Y=0,Z=0)
    DriverItem=(Type=Bone,Name="thigh_r")
    FalloffRegionScaleFactors=(PositiveWidth=1,NegativeWidth=1.1,PositiveHeight=1,NegativeHeight=1)
    FalloffSize=0.55
    FlipHeightScaling=true
    FlipWidthScaling=true
    OptionalParentItem=(Type=Bone,Name="None")
    RotationOffset=(X=-90,Y=0,Z=-20)
   )
   > n8 = &RigUnit_SphericalPoseReader(
    ActiveRegionScaleFactors=(PositiveWidth=1,NegativeWidth=1,PositiveHeight=1,NegativeHeight=1)
    ActiveRegionSize=0.05
    Debug=(bDrawDebug=false,bDraw2D=False,bDrawLocalAxes=False,DebugScale=25,DebugSegments=20,DebugThickness=0.25)
    DriverAxis=(X=1,Y=0,Z=0)
    DriverItem=(Type=Bone,Name="thigh_r")
    FalloffRegionScaleFactors=(PositiveWidth=1,NegativeWidth=1.15,PositiveHeight=1,NegativeHeight=1)
    FalloffSize=0.55
    FlipHeightScaling=true
    FlipWidthScaling=true
    OptionalParentItem=(Type=Bone,Name="None")
    RotationOffset=(X=0,Y=20,Z=0)
   )
   > n9 = &RigUnit_SphericalPoseReader(
    ActiveRegionScaleFactors=(PositiveWidth=1,NegativeWidth=1,PositiveHeight=1,NegativeHeight=1)
    ActiveRegionSize=0.05
    Debug=(bDrawDebug=false,bDraw2D=False,bDrawLocalAxes=False,DebugScale=25,DebugSegments=20,DebugThickness=0.25)
    DriverAxis=(X=1,Y=0,Z=0)
    DriverItem=(Type=Bone,Name="thigh_r")
    FalloffRegionScaleFactors=(PositiveWidth=1,NegativeWidth=1.5,PositiveHeight=1,NegativeHeight=1.26)
    FalloffSize=0.26
    FlipHeightScaling=true
    FlipWidthScaling=true
    OptionalParentItem=(Type=Bone,Name="None")
    RotationOffset=(X=40,Y=20,Z=0)
   )
   > n10 = &RigUnit_SphericalPoseReader(
    ActiveRegionScaleFactors=(PositiveWidth=1,NegativeWidth=1,PositiveHeight=1,NegativeHeight=1)
    ActiveRegionSize=0.05
    Debug=(bDrawDebug=false,bDraw2D=false,bDrawLocalAxes=False,DebugScale=25,DebugSegments=20,DebugThickness=0.25)
    DriverAxis=(X=1,Y=0,Z=0)
    DriverItem=(Type=Bone,Name="thigh_r")
    FalloffRegionScaleFactors=(PositiveWidth=1,NegativeWidth=1.1,PositiveHeight=1,NegativeHeight=1)
    FalloffSize=0.55
    FlipHeightScaling=true
    FlipWidthScaling=true
    OptionalParentItem=(Type=Bone,Name="None")
    RotationOffset=(X=90,Y=0,Z=20)
   )
   > n11 = &RigUnit_SphericalPoseReader(
    ActiveRegionScaleFactors=(PositiveWidth=1,NegativeWidth=1,PositiveHeight=1,NegativeHeight=1)
    ActiveRegionSize=0.05
    Debug=(bDrawDebug=false,bDraw2D=false,bDrawLocalAxes=False,DebugScale=25,DebugSegments=20,DebugThickness=0.25)
    DriverAxis=(X=1,Y=0,Z=0)
    DriverItem=(Type=Bone,Name="thigh_r")
    FalloffRegionScaleFactors=(PositiveWidth=2,NegativeWidth=1.5,PositiveHeight=1,NegativeHeight=1)
    FalloffSize=0.22
    FlipHeightScaling=true
    FlipWidthScaling=true
    OptionalParentItem=(Type=Bone,Name="None")
    RotationOffset=(X=-130,Y=20,Z=0)
   )
   RigVMVariableNode(ExecuteContext=@RigVMFunction_Sequence.B, Value=@n1.OutputParam) {variable=Leg Front L}
   > RigVMVariableNode(Value=@n2.OutputParam) {variable=Leg FrontSide L}
   > RigVMVariableNode(Value=@n3.OutputParam) {variable=Leg Side L}
   > RigVMVariableNode(Value=@n4.OutputParam) {variable=Leg BackSide L}
   > RigVMVariableNode(Value=@n5.OutputParam) {variable=Leg Back L}
   > RigVMVariableNode(Value=@n6.OutputParam) {variable=Leg FrontInside L}
   > RigVMVariableNode(Value=@n7.OutputParam) {variable=Leg Front R}
   > RigVMVariableNode(Value=@n8.OutputParam) {variable=Leg Side R}
   > RigVMVariableNode(Value=@n9.OutputParam) {variable=Leg BackSide R}
   > RigVMVariableNode(Value=@n10.OutputParam) {variable=Leg Back R}
   > RigVMVariableNode(Value=@n11.OutputParam) {variable=Leg FrontInside R}
   RigVMFunctionReturnNode(ExecuteContext=@RigVMFunction_Sequence.C)
   n12 = &RigUnit_SphericalPoseReader(
    ActiveRegionScaleFactors=(PositiveWidth=1,NegativeWidth=1,PositiveHeight=1,NegativeHeight=1)
    ActiveRegionSize=0.05
    Debug=(bDrawDebug=false,bDraw2D=False,bDrawLocalAxes=False,DebugScale=25,DebugSegments=20,DebugThickness=0.25)
    DriverAxis=(X=1,Y=0,Z=0)
    DriverItem=(Type=Bone,Name="thigh_r")
    ExecutePin=()
    FalloffRegionScaleFactors=(PositiveWidth=1,NegativeWidth=1,PositiveHeight=1,NegativeHeight=1)
    FalloffSize=0.25
    FlipHeightScaling=false
    FlipWidthScaling=false
    OptionalParentItem=(Type=Bone,Name="None")
    RotationOffset=(X=-40,Y=30,Z=0)
   )
   RigVMVariableNode(ExecuteContext=(), Value=@n12.OutputParam) {variable=Leg FrontSide R}
  ]}
 RigVMFunctionLibrary {nodes [
   Collar Setup = RigVMCollapseNode(ExecuteContext=())
   Convert Variables = RigVMCollapseNode(ExecuteContext=())
   Delta Rotation = RigVMCollapseNode(ExecuteContext=(), debugRotation=False, onlyTwist=False, twistAxis=(X=0,Y=0,Z=0))
   Delta Rotation X = RigVMCollapseNode(ExecuteContext=())
   ShoulderPad Setup = RigVMCollapseNode(ExecuteContext=())
   Skirt SPRs = RigVMCollapseNode(ExecuteContext=())
   Skirt Setup = RigVMCollapseNode(ExecuteContext=())
   Skirt Setup No RBAN = RigVMCollapseNode(ExecuteContext=())
   Sleeves Setup = RigVMCollapseNode(ExecuteContext=())
  ]}
 RigVMFunction_Sequence_1_ContainedGraph {nodes [
   n1 = RigVMFunctionEntryNode()
   RigVMFunction_Sequence = &RigVMFunction_Sequence(ExecuteContext=@n1.ExecuteContext)
   n2 = &RigVMFunction_Sequence(ExecuteContext=@RigVMFunction_Sequence.B)
   n3 = &RigVMFunction_Sequence(ExecuteContext=@n2.B)
   n4 = &RigVMFunction_Sequence(ExecuteContext=@n3.B)
   RigVMFunctionReturnNode(A=@RigVMFunction_Sequence.A, B=@n2.A, F=@n3.A, G=@n4.A, H=@n4.B)
  ]}
 RigVMFunction_Sequence_1_ContainedGraph {nodes [
   n1 = RigVMFunctionEntryNode()
   RigVMFunction_Sequence = &RigVMFunction_Sequence(ExecuteContext=@n1.ExecuteContext)
   n2 = &RigVMFunction_Sequence(ExecuteContext=@RigVMFunction_Sequence.B)
   RigVMFunctionReturnNode(A=@RigVMFunction_Sequence.A, B=@n2.A, C=@n2.B)
  ]}
 RigVMFunction_Sequence_1_ContainedGraph {nodes [
   n1 = RigVMFunctionEntryNode()
   RigVMFunction_Sequence = &RigVMFunction_Sequence(ExecuteContext=@n1.ExecuteContext)
   n2 = &RigVMFunction_Sequence(ExecuteContext=@RigVMFunction_Sequence.B)
   n3 = &RigVMFunction_Sequence(ExecuteContext=@n2.B)
   n4 = &RigVMFunction_Sequence(ExecuteContext=@n3.B)
   n5 = &RigVMFunction_Sequence(ExecuteContext=@n4.B)
   RigVMFunctionReturnNode(A=@RigVMFunction_Sequence.A, B=@n2.A, C=@n3.A, D=@n4.A, E=@n5.A, F=@n5.B)
  ]}
 RigVMFunction_Sequence_1_ContainedGraph {nodes [
   n1 = RigVMFunctionEntryNode()
   RigVMFunction_Sequence = &RigVMFunction_Sequence(ExecuteContext=@n1.ExecuteContext)
   n2 = &RigVMFunction_Sequence(ExecuteContext=@RigVMFunction_Sequence.B)
   n3 = &RigVMFunction_Sequence(ExecuteContext=@n2.B)
   n4 = &RigVMFunction_Sequence(ExecuteContext=@n3.B)
   n5 = &RigVMFunction_Sequence(ExecuteContext=@n4.B)
   RigVMFunctionReturnNode(A=@RigVMFunction_Sequence.A, B=@n2.A, C=@n3.A, D=@n4.A, E=@n5.A, F=@n5.B)
  ]}
 RigVMFunction_Sequence_1_ContainedGraph {nodes [
   n1 = RigVMFunctionEntryNode()
   RigVMFunction_Sequence = &RigVMFunction_Sequence(ExecuteContext=@n1.ExecuteContext)
   n2 = &RigVMFunction_Sequence(ExecuteContext=@RigVMFunction_Sequence.B)
   RigVMFunctionReturnNode(A=@RigVMFunction_Sequence.A, B=@n2.A, C=@n2.B)
  ]}
 RigVMModel {nodes [
   RigUnit_BeginExecution = &RigUnit_BeginExecution()
   FN_DeformRig_FNC = RigVMFunctionReferenceNode(
    DeformRig_ Curve="(deformRig_l_upperarm_up_alpha_72_#1=False,deformRig_r_upperarm_up_alpha_73_#2=False,deformRig_l_upperarm_fr_alpha_74_#3=False,deformRig_r_upperarm_fr_alpha_75_CB51B8 ...(truncated)"
    DeformRig_Bool="(upperarm_1_#4=True,elbow_3_#5=True,wrist_7_#6=True,groin_8_#7=True,glute_17_#8=false,knee_9_40C408324CA26 ...(truncated)"
    DeformRig_Debug=(upperarm_debug_18_#4=False,wrist_debug_20_#6=False,lat_debug_26_#9=False,groinGlute_debug_27_#7=False)
    DeformRig_Value="(upperarm_up_18_#10=(Rotation=(X=0,Y=0,Z=0,W=1),Translation=(X=0,Y=0,Z=8),Scale3D=(X=0.7,Y=1,Z=1)),upperarm_fwd_fr_19_#11=(Rotation=(X=0,Y=0,Z=0,W=1),Translation=(X=0,Y=10,Z=0),Scale3D=(X=0.5,Y=1,Z= ...(truncated)"
    ExecuteContext=@RigUnit_BeginExecution.ExecutePin
   )
   > Convert Variables = RigVMFunctionReferenceNode()
   > Skirt SPRs = RigVMFunctionReferenceNode()
   > Sleeves Setup = RigVMFunctionReferenceNode()
   > ShoulderPad Setup = RigVMFunctionReferenceNode()
   > Collar Setup = RigVMFunctionReferenceNode()
   RigVMFunction_ControlFlowBranch = &RigVMFunction_ControlFlowBranch(ExecuteContext=@"Collar Setup".ExecuteContext, Condition=RigVMVariableNode() {variable=RBAN Enable}.Value)
   Skirt Setup No RBAN = RigVMFunctionReferenceNode(ExecuteContext=@RigVMFunction_ControlFlowBranch.False)
   Skirt Setup = RigVMFunctionReferenceNode(ExecuteContext=@RigVMFunction_ControlFlowBranch.True)
   &RigUnit_PrepareForExecution()
  ]}
]
hierarchy [
 root {loc=<0 0 0>, rot=<0 0 0 1>}
 pelvis {parent=root, loc=<0 -0.5301 98.0571>, rot=<0.0127 0.707 -0.0127 -0.707>}
 spine_01 {parent=pelvis, loc=<6.1208 -0.9328 0>, rot=<0 0 0.034 -0.9994>}
 spine_02 {parent=spine_01, loc=<6.5997 -0.2095 0>, rot=<0 0 0.0238 -0.9997>}
 spine_03 {parent=spine_02, loc=<6.4914 -0.1237 0>, rot=<0 0 0.0149 -0.9999>}
 spine_04 {parent=spine_03, loc=<6.9598 0.0794 0>, rot=<0 0 -0.0132 -0.9999>}
 spine_05 {parent=spine_04, loc=<6.9987 0 0>, rot=<0 0 -0.0015 -1>}
 neck_01 {parent=spine_05, loc=<11.3138 0.1529 0>, rot=<0 0 -0.2003 -0.9797>}
 neck_02 {parent=neck_01, loc=<4.5627 0 0>, rot=<0 0 0.0504 -0.9987>}
 head {parent=neck_02, loc=<5.3914 0 0>, rot=<0 0 0.115 -0.9934>}
 ice_bk {parent=head, loc=<10.541 -7.667 0.013>, rot=<0.5021 -0.4979 0.4979 -0.5021>}
 base_pendants_sd_b_r {parent=head, loc=<6.9387 -9.7233 6.5644>, rot=<0.8586 0.0337 -0.0437 -0.5097>}
 dyn_pendants_sd_b_01_r {parent=base_pendants_sd_b_r, loc=<0 0 0>, rot=<0 0 0 1>}
 dyn_pendants_sd_b_02_r {parent=dyn_pendants_sd_b_01_r, loc=<-2.8722 0 0>, rot=<0 0 0 1>}
 base_pendants_sd_a_r {parent=head, loc=<7.5065 -3.6371 11.2015>, rot=<0.9956 0.0391 -0.0757 -0.04>}
 dyn_pendants_sd_a_01_r {parent=base_pendants_sd_a_r, loc=<0 0 0>, rot=<0 0 0 1>}
 dyn_pendants_sd_a_02_r {parent=dyn_pendants_sd_a_01_r, loc=<-3.2886 0 0>, rot=<0 0 0 1>}
 base_pendants_bk {parent=head, loc=<6.7295 -12.2728 -0.0066>, rot=<0.0029 -0.7071 0.7071 -0.0029>}
 dyn_pendants_bk_01 {parent=base_pendants_bk, loc=<0 0 0>, rot=<0 0 0 1>}
 dyn_pendants_bk_02 {parent=dyn_pendants_bk_01, loc=<2.6206 0 0>, rot=<0 0 0 1>}
 base_pendants_sd_b_l {parent=head, loc=<6.9384 -9.7234 -6.5644>, rot=<0.0337 -0.8586 0.5097 -0.0437>}
 dyn_pendants_sd_b_01_l {parent=base_pendants_sd_b_l, loc=<0 0 0>, rot=<0 0 0 1>}
 dyn_pendants_sd_b_02_l {parent=dyn_pendants_sd_b_01_l, loc=<2.8717 0 0>, rot=<0 0 0 1>}
 base_pendants_sd_a_l {parent=head, loc=<7.5065 -3.6371 -11.2015>, rot=<0.0391 -0.9956 0.04 -0.0757>}
 dyn_pendants_sd_a_01_l {parent=base_pendants_sd_a_l, loc=<0 0 0>, rot=<0 0 0 1>}
 dyn_pendants_sd_a_02_l {parent=dyn_pendants_sd_a_01_l, loc=<3.2884 0 0>, rot=<0 0 0 1>}
 clavicle_l {parent=spine_05, loc=<5.4831 -0.1053 -3.2946>, rot=<0.0465 -0.7289 -0.009 -0.6829>}
 upperarm_l {parent=clavicle_l, loc=<9.8658 -1.2136 -0.4105>, rot=<0.0164 -0.3245 -0.0269 -0.9454>}
 lowerarm_l {parent=upperarm_l, loc=<26.465 -0.8345 0.259>, rot=<0.0036 0.0161 -0.2298 -0.9731>}
 hand_l {parent=lowerarm_l, loc=<22.7355 -0.3203 -0.6105>, rot=<-0.6784 -0.0223 0.0735 -0.7307>}
 thumb_01_l {parent=hand_l, loc=<1.3118 -0.3191 -2.459>, rot=<0.6546 -0.4042 0.0664 -0.6354>}
 thumb_02_l {parent=thumb_01_l, loc=<5.9336 -0.0448 0.0401>, rot=<-0.0037 0.0213 0.298 -0.9543>}
 thumb_03_l {parent=thumb_02_l, loc=<2.2223 0.3682 -0.0907>, rot=<0.0118 -0.013 0.0066 -0.9998>}
 index_metacarpal_l {parent=hand_l, loc=<1.6535 0.0183 -2.96>, rot=<0.0158 -0.0124 -0.0173 -0.9996>}
 index_01_l {parent=index_metacarpal_l, loc=<8.3008 0.0357 -0.0524>, rot=<-0.0044 0.0485 0.0924 -0.9945>}
 index_02_l {parent=index_01_l, loc=<3.3587 -0.3004 0.0452>, rot=<-0.002 0.0027 0.3225 -0.9466>}
 index_03_l {parent=index_02_l, loc=<2.5578 -0.123 -0.0252>, rot=<-0.002 0.0067 0.2558 -0.9667>}
 middle_metacarpal_l {parent=hand_l, loc=<2.1856 -0.1725 -0.5984>, rot=<0.006 0.0168 -0.0329 -0.9993>}
 middle_01_l {parent=middle_metacarpal_l, loc=<7.3275 -0.2447 0.0129>, rot=<-0.0208 0.0042 0.0858 -0.9961>}
 middle_02_l {parent=middle_01_l, loc=<4.0123 0.1401 0.1175>, rot=<-0.0218 0.0258 0.3512 -0.9357>}
 middle_03_l {parent=middle_02_l, loc=<2.6077 0.0534 0.1373>, rot=<0.0082 0.0188 0.1708 -0.9851>}
 ring_metacarpal_l {parent=hand_l, loc=<2.2074 -0.3625 1.3638>, rot=<-0.0284 0.0722 -0.0458 -0.9959>}
 ring_01_l {parent=ring_metacarpal_l, loc=<7.0295 -0.6197 -0.1446>, rot=<-0.0489 -0.0252 0.1892 -0.9804>}
 ring_02_l {parent=ring_01_l, loc=<3.6174 0.4647 -0.1027>, rot=<0 0 0.2806 -0.9598>}
 ring_03_l {parent=ring_02_l, loc=<2.4152 0.114 0.0199>, rot=<0 0 0.1189 -0.9929>}
 pinky_metacarpal_l {parent=hand_l, loc=<1.7616 -0.9753 3.3814>, rot=<-0.0397 0.0814 -0.069 -0.9935>}
 pinky_01_l {parent=pinky_metacarpal_l, loc=<7.1709 -0.4831 -0.1094>, rot=<-0.0654 -0.0035 0.2242 -0.9723>}
 pinky_02_l {parent=pinky_01_l, loc=<2.8274 0.0893 -0.1043>, rot=<0.0037 0.0011 0.2854 -0.9584>}
 pinky_03_l {parent=pinky_02_l, loc=<2.2046 0.0307 -0.0078>, rot=<-0.0002 0.0098 0.1445 -0.9895>}
 deform_wrist_dn_base_l {parent=hand_l, loc=<0 -1.5326 0>, rot=<0 0 0 -1>}
 deform_wrist_dn_l {parent=deform_wrist_dn_base_l, loc=<0 0 0>, rot=<0 0 0 1>}
 deform_wrist_up_base_l {parent=hand_l, loc=<0 2.3009 0>, rot=<0 0 0 -1>}
 deform_wrist_up_l {parent=deform_wrist_up_base_l, loc=<0 0 0>, rot=<0 0 0 1>}
 lowerarm_twist_02_l {parent=lowerarm_l, loc=<10.6707 0 0>, rot=<0 0 0 1>}
 lowerarm_twist_01_l {parent=lowerarm_l, loc=<18.316 0 0>, rot=<0 0 0 1>}
 base_sleeve_l {parent=lowerarm_twist_01_l, loc=<-18.316 0 0>, rot=<0 0 0 1>}
 base_sleeves_twist_a_l {parent=base_sleeve_l, loc=<10.6707 0 0>, rot=<-0.3794 -0.3794 0.5967 -0.5967>}
 dyn_sleeves_twist_a_01_l {parent=base_sleeves_twist_a_l, loc=<0 0 0>, rot=<0 0 0 1>}
 dyn_sleeves_twist_a_02_l {parent=dyn_sleeves_twist_a_01_l, loc=<4.0997 0 0>, rot=<0 0 0 1>}
 base_sleeves_twist_b_l {parent=base_sleeve_l, loc=<18.316 0 0>, rot=<-0.3758 -0.3758 0.599 -0.599>}
 dyn_sleeves_twist_b_01_l {parent=base_sleeves_twist_b_l, loc=<0 0 0>, rot=<0 0 0 1>}
 dyn_sleeves_twist_b_02_l {parent=dyn_sleeves_twist_b_01_l, loc=<5.9238 0 0>, rot=<0 0 0 1>}
 dyn_sleeves_twist_b_03_l {parent=dyn_sleeves_twist_b_02_l, loc=<4.5539 0 0>, rot=<0 0 0 1>}
 lowerarm_twist_extra_l {parent=lowerarm_twist_01_l, loc=<0 0 0>, rot=<0 0 0 1>}
 deform_elbow_in_base_l {parent=lowerarm_l, loc=<0.1834 3.6062 -0.2572>, rot=<0 0 0 -1>}
 deform_elbow_in_l {parent=deform_elbow_in_base_l, loc=<0 0 0>, rot=<0 0 0 1>}
 deform_elbow_out_base_l {parent=lowerarm_l, loc=<0 -4 0>, rot=<0 0 0 1>}
 deform_elbow_out_l {parent=deform_elbow_out_base_l, loc=<0 0 0>, rot=<0 0 0 1>}
 upperarm_twist_01_l {parent=upperarm_l, loc=<7.2135 -0.1134 0>, rot=<0 0 0 1>}
 deform_upperarm_bk_base_l {parent=upperarm_twist_01_l, loc=<-6.4053 -4.3746 0.314>, rot=<0 0 0 -1>}
 deform_upperarm_bk_l {parent=deform_upperarm_bk_base_l, loc=<0 0 0>, rot=<0 0 0 1>}
 deform_upperarm_fr_base_l {parent=upperarm_twist_01_l, loc=<-5.6312 5.092 0.3354>, rot=<0 0 0 -1>}
 deform_upperarm_fr_l {parent=deform_upperarm_fr_base_l, loc=<0 0 0>, rot=<0 0 0 1>}
 deform_upperarm_up_base_l {parent=upperarm_twist_01_l, loc=<-8.891 0.4825 3.9517>, rot=<0 0 0 -1>}
 deform_upperarm_up_l {parent=deform_upperarm_up_base_l, loc=<0 0 0>, rot=<0 0 0 1>}
 base_ShoulderPad_general_l {parent=upperarm_twist_01_l, loc=<0 0 0>, rot=<0 0 0 1>}
 base_ShoulderPad_side_l {parent=base_ShoulderPad_general_l, loc=<0.5026 -1.15 4.1401>, rot=<0.0195 0.3559 -0.0643 -0.9321>}
 dyn_ShoulderPad_side_l {parent=base_ShoulderPad_side_l, loc=<0 0 0>, rot=<0 0 0 1>}
 base_ShoulderPad_front_l {parent=base_ShoulderPad_general_l, loc=<1.2738 3.5035 -0.6772>, rot=<0.6204 0.0486 0.0592 -0.7806>}
 dyn_ShoulderPad_front_l {parent=base_ShoulderPad_front_l, loc=<0 0 0>, rot=<0 0 0 1>}
 base_ShoulderPad_back_l {parent=base_ShoulderPad_general_l, loc=<1.5069 -5.9925 -0.2578>, rot=<-0.7903 -0.0105 -0.06 -0.6096>}
 dyn_ShoulderPad_back_l {parent=base_ShoulderPad_back_l, loc=<0 0 0>, rot=<0 0 0 1>}
 upperarm_twist_02_l {parent=upperarm_l, loc=<19.7998 -0.1134 0>, rot=<0 0 0 1>}
 clavicle_r {parent=spine_05, loc=<5.4831 -0.1053 3.2946>, rot=<0.7289 0.0465 -0.6829 0.009>}
 upperarm_r {parent=clavicle_r, loc=<-9.8658 1.2136 0.4105>, rot=<0.0164 -0.3245 -0.0269 -0.9454>}
 lowerarm_r {parent=upperarm_r, loc=<-26.465 0.8345 -0.259>, rot=<0.0036 0.0161 -0.2298 -0.9731>}
 hand_r {parent=lowerarm_r, loc=<-22.7355 0.3203 0.6106>, rot=<-0.6784 -0.0223 0.0735 -0.7307>}
 thumb_01_r {parent=hand_r, loc=<-1.3118 0.3191 2.459>, rot=<0.6546 -0.4042 0.0664 -0.6354>}
 thumb_02_r {parent=thumb_01_r, loc=<-5.9336 0.0448 -0.0401>, rot=<-0.0037 0.0213 0.298 -0.9543>}
 thumb_03_r {parent=thumb_02_r, loc=<-2.2223 -0.3682 0.0907>, rot=<0.0118 -0.013 0.0066 -0.9998>}
 index_metacarpal_r {parent=hand_r, loc=<-1.6535 -0.0183 2.96>, rot=<0.0158 -0.0124 -0.0173 -0.9996>}
 index_01_r {parent=index_metacarpal_r, loc=<-8.3008 -0.0357 0.0524>, rot=<-0.0044 0.0485 0.0924 -0.9945>}
 index_02_r {parent=index_01_r, loc=<-3.3587 0.3004 -0.0452>, rot=<-0.002 0.0027 0.3225 -0.9466>}
 index_03_r {parent=index_02_r, loc=<-2.5578 0.123 0.0252>, rot=<-0.002 0.0067 0.2558 -0.9667>}
 middle_metacarpal_r {parent=hand_r, loc=<-2.1856 0.1725 0.5984>, rot=<0.006 0.0168 -0.0329 -0.9993>}
 middle_01_r {parent=middle_metacarpal_r, loc=<-7.3275 0.2447 -0.0129>, rot=<-0.0208 0.0042 0.0858 -0.9961>}
 middle_02_r {parent=middle_01_r, loc=<-4.0123 -0.1401 -0.1175>, rot=<-0.0218 0.0258 0.3512 -0.9357>}
 middle_03_r {parent=middle_02_r, loc=<-2.6077 -0.0534 -0.1373>, rot=<0.0082 0.0188 0.1708 -0.9851>}
 ring_metacarpal_r {parent=hand_r, loc=<-2.2074 0.3625 -1.3638>, rot=<-0.0284 0.0722 -0.0458 -0.9959>}
 ring_01_r {parent=ring_metacarpal_r, loc=<-7.0295 0.6197 0.1446>, rot=<-0.0489 -0.0252 0.1892 -0.9804>}
 ring_02_r {parent=ring_01_r, loc=<-3.6174 -0.4647 0.1027>, rot=<0 0 0.2806 -0.9598>}
 ring_03_r {parent=ring_02_r, loc=<-2.4152 -0.114 -0.0199>, rot=<0 0 0.1189 -0.9929>}
 pinky_metacarpal_r {parent=hand_r, loc=<-1.7616 0.9753 -3.3814>, rot=<-0.0397 0.0814 -0.069 -0.9935>}
 pinky_01_r {parent=pinky_metacarpal_r, loc=<-7.1709 0.4831 0.1094>, rot=<-0.0654 -0.0035 0.2242 -0.9723>}
 pinky_02_r {parent=pinky_01_r, loc=<-2.8274 -0.0893 0.1043>, rot=<0.0037 0.0011 0.2854 -0.9584>}
 pinky_03_r {parent=pinky_02_r, loc=<-2.2046 -0.0307 0.0078>, rot=<-0.0002 0.0098 0.1445 -0.9895>}
 deform_wrist_dn_base_r {parent=hand_r, loc=<-0.0001 1.5328 0>, rot=<0 0 0 -1>}
 deform_wrist_dn_r {parent=deform_wrist_dn_base_r, loc=<0 0 0>, rot=<0 0 0 1>}
 deform_wrist_up_base_r {parent=hand_r, loc=<-0.0002 -2.3007 -0.0001>, rot=<0 0 0 -1>}
 deform_wrist_up_r {parent=deform_wrist_up_base_r, loc=<0 0 0>, rot=<0 0 0 1>}
 lowerarm_twist_02_r {parent=lowerarm_r, loc=<-10.6707 0 0>, rot=<0 0 0 -1>}
 lowerarm_twist_01_r {parent=lowerarm_r, loc=<-18.316 0 0>, rot=<0 0 0 -1>}
 base_sleeve_r {parent=lowerarm_twist_01_r, loc=<18.316 0 0.0001>, rot=<0 0 0 -1>}
 base_sleeves_twist_b_r {parent=base_sleeve_r, loc=<-18.3163 0.0001 0.0002>, rot=<-0.3758 -0.3758 0.599 -0.599>}
 dyn_sleeves_twist_b_01_r {parent=base_sleeves_twist_b_r, loc=<0 0 0>, rot=<0 0 0 -1>}
 dyn_sleeves_twist_b_02_r {parent=dyn_sleeves_twist_b_01_r, loc=<-5.9233 0.0003 -0.0001>, rot=<0 0 0 1>}
 dyn_sleeves_twist_b_03_r {parent=dyn_sleeves_twist_b_02_r, loc=<-4.554 -0.0001 0>, rot=<0 0 0 1>}
 base_sleeves_twist_a_r {parent=base_sleeve_r, loc=<-10.6709 0.0001 0.0003>, rot=<-0.3794 -0.3794 0.5967 -0.5967>}
 dyn_sleeves_twist_a_01_r {parent=base_sleeves_twist_a_r, loc=<0 0 0>, rot=<0 0 0 1>}
 dyn_sleeves_twist_a_02_r {parent=dyn_sleeves_twist_a_01_r, loc=<-4.0991 0.0005 -0.0001>, rot=<0 0 0 1>}
 lowerarm_twist_extra_r {parent=lowerarm_twist_01_r, loc=<0 0 0>, rot=<0 0 0 1>}
 deform_elbow_in_base_r {parent=lowerarm_r, loc=<-0.1834 -3.6062 0.2572>, rot=<0 0 0 -1>}
 deform_elbow_in_r {parent=deform_elbow_in_base_r, loc=<0 0 0>, rot=<0 0 0 1>}
 deform_elbow_out_base_r {parent=lowerarm_r, loc=<0 4 -0.0001>, rot=<0 0 0 -1>}
 deform_elbow_out_r {parent=deform_elbow_out_base_r, loc=<0 0 0>, rot=<0 0 0 1>}
 upperarm_twist_01_r {parent=upperarm_r, loc=<-7.2135 0.1134 0>, rot=<0 0 0 -1>}
 deform_upperarm_bk_base_r {parent=upperarm_twist_01_r, loc=<6.4053 4.3746 -0.314>, rot=<0 0 0 -1>}
 deform_upperarm_bk_r {parent=deform_upperarm_bk_base_r, loc=<0 0 0>, rot=<0 0 0 1>}
 deform_upperarm_fr_base_r {parent=upperarm_twist_01_r, loc=<5.6312 -5.092 -0.3354>, rot=<0 0 0 -1>}
 deform_upperarm_fr_r {parent=deform_upperarm_fr_base_r, loc=<0 0 0>, rot=<0 0 0 1>}
 deform_upperarm_up_base_r {parent=upperarm_twist_01_r, loc=<8.891 -0.4825 -3.9517>, rot=<0 0 0 -1>}
 deform_upperarm_up_r {parent=deform_upperarm_up_base_r, loc=<0 0 0>, rot=<0 0 0 1>}
 base_ShoulderPad_general_r {parent=upperarm_twist_01_r, loc=<0 0 0>, rot=<0 0 0 1>}
 base_ShoulderPad_back_r {parent=base_ShoulderPad_general_r, loc=<-1.5068 5.9925 0.2577>, rot=<-0.7903 -0.0105 -0.06 -0.6096>}
 dyn_ShoulderPad_back_r {parent=base_ShoulderPad_back_r, loc=<0 0 0>, rot=<0 0 0 1>}
 base_ShoulderPad_front_r {parent=base_ShoulderPad_general_r, loc=<-1.2734 -3.5035 0.6769>, rot=<0.6204 0.0486 0.0592 -0.7806>}
 dyn_ShoulderPad_front_r {parent=base_ShoulderPad_front_r, loc=<0 0 0>, rot=<0 0 0 -1>}
 base_ShoulderPad_side_r {parent=base_ShoulderPad_general_r, loc=<-0.5023 1.15 -4.1404>, rot=<0.0195 0.3559 -0.0643 -0.9321>}
 dyn_ShoulderPad_side_r {parent=base_ShoulderPad_side_r, loc=<0 0 0>, rot=<0 0 0 -1>}
 upperarm_twist_02_r {parent=upperarm_r, loc=<-19.7998 0.1134 0>, rot=<0 0 0 -1>}
 attach_backpack {parent=spine_05, loc=<-0.1232 -13.4716 0.001>, rot=<-0.0093 -0.707 0.0093 -0.707>}
 Backpack_BR {parent=attach_backpack, type=Null, loc=<0 3 0>, rot=<0 0 0 1>}
 attach_cape {parent=spine_05, loc=<7.9643 -12.0739 0.3071>, rot=<0.0094 -0.7022 0.0288 -0.7113>}
 cape {parent=attach_cape, type=Null, loc=<0 0 0>, rot=<-0.0117 0 0 0.9999>}
 base_collar_a_l {parent=spine_05, loc=<8.9997 1.8965 -8.8979>, rot=<0.0001 -0.0788 -0.0784 -0.9938>}
 dyn_collar_a_01_l {parent=base_collar_a_l, loc=<0 0 0>, rot=<0 0 0 1>}
 dyn_collar_a_02_l {parent=dyn_collar_a_01_l, loc=<4.011 0 0>, rot=<0 0 0 1>}
 base_collar_b_l {parent=spine_05, loc=<10.9673 -2.5292 -7.6802>, rot=<0.3276 -0.1493 -0.077 -0.9297>}
 dyn_collar_b_01_l {parent=base_collar_b_l, loc=<0 0 0>, rot=<0 0 0 1>}
 dyn_collar_b_02_l {parent=dyn_collar_b_01_l, loc=<4.8426 0 0>, rot=<0 0 0 1>}
 base_collar_a_r {parent=spine_05, loc=<9 1.8965 8.8979>, rot=<-0.0788 -0.0001 0.9938 -0.0784>}
 dyn_collar_a_01_r {parent=base_collar_a_r, loc=<0 0 0>, rot=<0 0 0 1>}
 dyn_collar_a_02_r {parent=dyn_collar_a_01_r, loc=<-4.0106 0 0.0001>, rot=<0 0 0 1>}
 base_collar_b_r {parent=spine_05, loc=<10.9669 -2.5292 7.6802>, rot=<-0.1493 -0.3276 0.9297 -0.077>}
 dyn_collar_b_01_r {parent=base_collar_b_r, loc=<0 0 0>, rot=<0 0 0 1>}
 dyn_collar_b_02_r {parent=dyn_collar_b_01_r, loc=<-4.8425 0 0>, rot=<0 0 0 1>}
 base_collar_m {parent=spine_05, loc=<11.8704 -5.6573 0>, rot=<0 0 0.1521 -0.9884>}
 dyn_collar_01_m {parent=base_collar_m, loc=<0 0 0>, rot=<0.7071 0 0 -0.7071>}
 dyn_collar_02_m {parent=dyn_collar_01_m, loc=<7.0945 0 0>, rot=<0 0 0 1>}
 base_collar_c_l {parent=spine_05, loc=<12.2663 -5.6204 -4.0262>, rot=<0.5623 -0.1153 -0.0034 -0.8188>}
 dyn_collar_c_01_l {parent=base_collar_c_l, loc=<0 0 0>, rot=<0 0 0 1>}
 dyn_collar_c_02_l {parent=dyn_collar_c_01_l, loc=<5.6254 0 0>, rot=<0 0 0 1>}
 base_collar_c_r {parent=spine_05, loc=<12.2658 -5.6204 4.0262>, rot=<-0.1153 -0.5623 0.8188 -0.0034>}
 dyn_collar_c_01_r {parent=base_collar_c_r, loc=<0 0 0>, rot=<0 0 0 1>}
 dyn_collar_c_02_r {parent=dyn_collar_c_01_r, loc=<-5.6262 0.0001 -0.0002>, rot=<0 0 0 1>}
 head_null {parent=spine_05, type=Null, loc=<20.6573 3.5489 0>, rot=<0 0 -0.036 -0.9994>}
 deform_pec_aim_base_l {parent=spine_04, loc=<3.2456 9.3309 -1.2237>, rot=<0.0603 -0.4033 0.3318 -0.8507>}
 deform_pec_aim_l {parent=deform_pec_aim_base_l, loc=<0 0 0>, rot=<0 0 0 1>}
 deform_pec_base_l {parent=deform_pec_aim_l, loc=<9.3665 0 0>, rot=<0 0 0 1>}
 deform_pec_l {parent=deform_pec_base_l, loc=<0 0 0>, rot=<0 0 0 1>}
 deform_pec_aim_base_r {parent=spine_04, loc=<3.2452 9.3309 1.2237>, rot=<-0.4033 -0.0603 0.8507 0.3318>}
 deform_pec_aim_r {parent=deform_pec_aim_base_r, loc=<0 0 0>, rot=<0 0 0 1>}
 deform_pec_base_r {parent=deform_pec_aim_r, loc=<-9.3667 -0.0003 -0.0004>, rot=<0 0 0 1>}
 deform_pec_r {parent=deform_pec_base_r, loc=<0 0 0>, rot=<0 0 0 1>}
 base_spine_pendant {parent=spine_04, loc=<-4.1019 8.9325 5.3992>, rot=<-0.0339 0.5788 0.8139 -0.0387>}
 dyn_spine_pendant_01 {parent=base_spine_pendant, loc=<0 0 0>, rot=<0 0 0 1>}
 dyn_spine_pendant_02 {parent=dyn_spine_pendant_01, loc=<3.2749 0.5336 0.5497>, rot=<0 0 0 1>}
 dyn_spine_pendant_03 {parent=dyn_spine_pendant_02, loc=<2.4154 0 0>, rot=<0 0 0 1>}
 thigh_l {parent=pelvis, loc=<-1.0855 -0.2109 -8.7053>, rot=<0.0009 0.0355 0.0155 -0.9993>}
 calf_l {parent=thigh_l, loc=<-42.5284 -0.0019 -0.7297>, rot=<0.0006 -0.0161 -0.0391 -0.9991>}
 foot_l {parent=calf_l, loc=<-46.7401 -2.5014 -0.592>, rot=<0.0427 -0.0238 0.0353 -0.9982>}
 ball_l {parent=foot_l, loc=<-3.5658 10.8066 -0.1544>, rot=<0.0001 -0.0001 -0.7063 -0.7079>}
 foot_bk_l {parent=foot_l, loc=<-0.3354 -1.7938 0.6647>, rot=<-0.0357 -0.7045 -0.0284 -0.7082>}
 calf_twist_02_l {parent=calf_l, loc=<-19.1546 -1.7943 -0.2638>, rot=<0.0237 0.0095 -0.0035 -0.9997>}
 base_ClothPants_fr_l {parent=calf_twist_02_l, loc=<12.7154 11.7821 0.7185>, rot=<-0.0505 -0.7353 -0.6712 -0.0797>}
 dyn_ClothPants_fr_01_l {parent=base_ClothPants_fr_l, loc=<0 0 0>, rot=<0 0 0 1>}
 dyn_ClothPants_fr_02_l {parent=dyn_ClothPants_fr_01_l, loc=<20.1883 0 0>, rot=<0 0 0 1>}
 base_ClothPants_in_l {parent=calf_twist_02_l, loc=<11.5947 0.2014 8.0452>, rot=<-0.0062 -0.0175 0.9994 0.0302>}
 dyn_ClothPants_in_01_l {parent=base_ClothPants_in_l, loc=<0 0 0>, rot=<0 0 0 1>}
 dyn_ClothPants_in_02_l {parent=dyn_ClothPants_in_01_l, loc=<20.1883 0 0>, rot=<0 0 0 1>}
 base_ClothPants_bk_l {parent=calf_twist_02_l, loc=<10.9653 -10.7191 0.4591>, rot=<-0.0159 -0.719 0.6948 -0.0021>}
 dyn_ClothPants_bk_01_l {parent=base_ClothPants_bk_l, loc=<0 0 0>, rot=<0 0 0 1>}
 dyn_ClothPants_bk_02_l {parent=dyn_ClothPants_bk_01_l, loc=<20.1883 0 0>, rot=<0 0 0 1>}
 base_ClothPants_out_l {parent=calf_twist_02_l, loc=<12.0071 1.1267 -6.2095>, rot=<-0.0107 -0.9976 -0.0165 -0.0665>}
 dyn_ClothPants_out_01_l {parent=base_ClothPants_out_l, loc=<0 0 0>, rot=<0 0 0 1>}
 dyn_ClothPants_out_02_l {parent=dyn_ClothPants_out_01_l, loc=<20.1883 0 0>, rot=<0 0 0 1>}
 calf_twist_01_l {parent=calf_l, loc=<-35.8483 -1.9197 -0.5794>, rot=<0.0237 0.0095 -0.0035 -0.9997>}
 deform_knee_in_base_l {parent=calf_l, loc=<0.8031 -5.0231 0.2456>, rot=<0 0 0 -1>}
 deform_knee_in_l {parent=deform_knee_in_base_l, loc=<0 0 0>, rot=<0 0 0 1>}
 deform_knee_out_base_l {parent=calf_l, loc=<0 4 0>, rot=<0 0 0 -1>}
 deform_knee_out_l {parent=deform_knee_out_base_l, loc=<0 0 0>, rot=<0 0 0 1>}
 thigh_twist_01_l {parent=thigh_l, loc=<-27.952 0 0>, rot=<0.0013 0 0 -1>}
 deform_groin_base_l {parent=thigh_l, loc=<0.4801 5.6776 -0.2486>, rot=<0 0 0 -1>}
 deform_groin_l {parent=deform_groin_base_l, loc=<0 0 0>, rot=<0 0 0 1>}
 calf_l_null {parent=thigh_l, type=Null, loc=<-42.5284 -0.0019 -0.7297>, rot=<0.0006 -0.0161 -0.0391 -0.9991>}
 thigh_r {parent=pelvis, loc=<-1.0855 -0.2109 8.7053>, rot=<0.0355 -0.0009 0.9993 0.0155>}
 calf_r {parent=thigh_r, loc=<42.5284 0.0019 0.7297>, rot=<0.0006 -0.0161 -0.0391 -0.9991>}
 foot_r {parent=calf_r, loc=<46.7401 2.5014 0.592>, rot=<0.0427 -0.0238 0.0353 -0.9982>}
 ball_r {parent=foot_r, loc=<3.5658 -10.8066 0.1544>, rot=<0.0001 -0.0001 -0.7063 -0.7079>}
 foot_bk_r {parent=foot_r, loc=<0.3354 1.7938 -0.6647>, rot=<-0.0357 -0.7045 -0.0284 -0.7082>}
 calf_twist_02_r {parent=calf_r, loc=<19.1546 1.7943 0.2638>, rot=<0.0237 0.0095 -0.0035 -0.9997>}
 base_ClothPants_fr_r {parent=calf_twist_02_r, loc=<-12.7154 -11.7821 -0.7185>, rot=<-0.6712 0.0797 0.0505 -0.7353>}
 dyn_ClothPants_fr_01_r {parent=base_ClothPants_fr_r, loc=<0 0 0>, rot=<0 0 0 1>}
 dyn_ClothPants_fr_02_r {parent=dyn_ClothPants_fr_01_r, loc=<20.1883 0 0>, rot=<0 0 0 1>}
 base_ClothPants_in_r {parent=calf_twist_02_r, loc=<-11.5947 -0.2013 -8.0452>, rot=<0.9994 -0.0302 0.0062 -0.0175>}
 dyn_ClothPants_in_01_r {parent=base_ClothPants_in_r, loc=<0 0 0>, rot=<0 0 0 1>}
 dyn_ClothPants_in_02_r {parent=dyn_ClothPants_in_01_r, loc=<20.1883 0 0>, rot=<0 0 0 1>}
 base_ClothPants_bk_r {parent=calf_twist_02_r, loc=<-10.9653 10.7192 -0.4591>, rot=<0.6948 0.0021 0.0159 -0.719>}
 dyn_ClothPants_bk_01_r {parent=base_ClothPants_bk_r, loc=<0 0 0>, rot=<0 0 0 1>}
 dyn_ClothPants_bk_02_r {parent=dyn_ClothPants_bk_01_r, loc=<20.1884 -0.0001 0>, rot=<0 0 0 1>}
 base_ClothPants_out_r {parent=calf_twist_02_r, loc=<-12.0071 -1.1267 6.2095>, rot=<-0.0165 0.0665 0.0107 -0.9976>}
 dyn_ClothPants_out_01_r {parent=base_ClothPants_out_r, loc=<0 0 0>, rot=<0 0 0 1>}
 dyn_ClothPants_out_02_r {parent=dyn_ClothPants_out_01_r, loc=<20.1883 0 -0.0001>, rot=<0 0 0 -1>}
 calf_twist_01_r {parent=calf_r, loc=<35.8483 1.9197 0.5794>, rot=<0.0237 0.0095 -0.0035 -0.9997>}
 deform_knee_in_base_r {parent=calf_r, loc=<-0.8031 5.0231 -0.2457>, rot=<0 0 0 -1>}
 deform_knee_in_r {parent=deform_knee_in_base_r, loc=<0 0 0>, rot=<0 0 0 1>}
 deform_knee_out_base_r {parent=calf_r, loc=<0 -4 0>, rot=<0 0 0 -1>}
 deform_knee_out_r {parent=deform_knee_out_base_r, loc=<0 0 0>, rot=<0 0 0 1>}
 thigh_twist_01_r {parent=thigh_r, loc=<27.952 0 0>, rot=<0.0013 0 0 -1>}
 deform_groin_base_r {parent=thigh_r, loc=<-0.4801 -5.6776 0.2486>, rot=<0 0 0 -1>}
 deform_groin_r {parent=deform_groin_base_r, loc=<0 0 0>, rot=<0 0 0 1>}
 calf_r_null {parent=thigh_r, type=Null, loc=<42.5284 0.0019 0.7297>, rot=<0.0006 -0.0161 -0.0391 -0.9991>}
 deform_glute_base_l {parent=pelvis, loc=<-7.3466 -9.0213 -9.1245>, rot=<0.0172 0.031 -0.4477 -0.8935>}
 deform_glute_l {parent=deform_glute_base_l, loc=<0 0 0>, rot=<0 0 0 1>}
 deform_glute_base_r {parent=pelvis, loc=<-7.3464 -9.0213 9.1245>, rot=<0.031 -0.0172 0.8935 -0.4477>}
 deform_glute_r {parent=deform_glute_base_r, loc=<0 0 0>, rot=<0 0 0 1>}
 base_skirt_sd_r {parent=pelvis, loc=<4.6904 -0.4311 10.731>, rot=<0.1512 -0.3808 0.9109 -0.0493>}
 dyn_skirt_sd_r_01 {parent=base_skirt_sd_r, loc=<0 0 0>, rot=<0 0 0 1>}
 dyn_skirt_sd_r_02 {parent=dyn_skirt_sd_r_01, loc=<17.4292 0 0>, rot=<-0.0176 -0.0329 -0.0279 -0.9989>}
 dyn_skirt_sd_r_03 {parent=dyn_skirt_sd_r_02, loc=<17.4268 0 0>, rot=<0.0537 0.012 -0.0189 -0.9983>}
 dyn_skirt_sd_r_04 {parent=dyn_skirt_sd_r_03, loc=<17.4326 0 0>, rot=<-0.0105 -0.0083 0.0032 -0.9999>}
 dyn_skirt_sd_r_05 {parent=dyn_skirt_sd_r_04, loc=<17.4329 0 -0.0001>, rot=<-0.0062 0.0008 0.0018 -1>}
 dyn_skirt_sd_r_06 {parent=dyn_skirt_sd_r_05, loc=<17.4333 0 0.0001>, rot=<-0.0113 0.0414 0.0037 -0.9991>}
 base_skirt_sd_l {parent=pelvis, loc=<4.6906 -0.4311 -10.731>, rot=<0.0493 -0.9109 0.3808 -0.1512>}
 dyn_skirt_sd_l_01 {parent=base_skirt_sd_l, loc=<0 0 0>, rot=<0 0 0 1>}
 dyn_skirt_sd_l_02 {parent=dyn_skirt_sd_l_01, loc=<17.4293 0 0>, rot=<0.0176 -0.0329 0.0279 -0.9989>}
 dyn_skirt_sd_l_03 {parent=dyn_skirt_sd_l_02, loc=<17.4269 0 0>, rot=<-0.0537 0.012 0.0189 -0.9983>}
 dyn_skirt_sd_l_04 {parent=dyn_skirt_sd_l_03, loc=<17.4326 0 0>, rot=<0.0105 -0.0083 -0.0032 -0.9999>}
 dyn_skirt_sd_l_05 {parent=dyn_skirt_sd_l_04, loc=<17.4329 0 0>, rot=<0.0062 0.0008 -0.0018 -1>}
 dyn_skirt_sd_l_06 {parent=dyn_skirt_sd_l_05, loc=<17.4332 0 0>, rot=<0.0113 0.0414 -0.0037 -0.9991>}
 base_skirt_fr_r {parent=pelvis, loc=<5.1484 6.6668 4.1466>, rot=<0.1627 -0.2366 0.9563 0.0549>}
 dyn_skirt_fr_r_01 {parent=base_skirt_fr_r, loc=<0 0 0>, rot=<0 0 0 1>}
 dyn_skirt_fr_r_02 {parent=dyn_skirt_fr_r_01, loc=<13 0 0>, rot=<0 0 0 1>}
 dyn_skirt_fr_r_03 {parent=dyn_skirt_fr_r_02, loc=<13 0 0>, rot=<0 0 0 1>}
 dyn_skirt_fr_r_04 {parent=dyn_skirt_fr_r_03, loc=<13 0 0>, rot=<0 0 -0.0411 -0.9992>}
 dyn_skirt_fr_r_05 {parent=dyn_skirt_fr_r_04, loc=<13 0 0>, rot=<0 0 -0.0275 -0.9996>}
 dyn_skirt_fr_r_06 {parent=dyn_skirt_fr_r_05, loc=<13 0 0>, rot=<0.0204 -0.0002 -0.0288 -0.9994>}
 dyn_skirt_fr_r_07 {parent=dyn_skirt_fr_r_06, loc=<13 0 0>, rot=<0 0 0 1>}
 dyn_skirt_fr_r_08 {parent=dyn_skirt_fr_r_07, loc=<13 0 0>, rot=<0 0 0 1>}
 dyn_skirt_fr_r_01_extra {parent=dyn_skirt_fr_r_01, loc=<0 0 0>, rot=<0 0 0 1>}
 base_skirt_fr_l {parent=pelvis, loc=<4.3177 5.8613 6.6132>, rot=<-0.1281 -0.963 0.0855 -0.221>}
 dyn_skirt_fr_l_01 {parent=base_skirt_fr_l, loc=<0 0 0>, rot=<0 0 0 1>}
 dyn_skirt_fr_l_02 {parent=dyn_skirt_fr_l_01, loc=<12.717 0.0993 0>, rot=<0 0 0.0463 -0.9989>}
 dyn_skirt_fr_l_03 {parent=dyn_skirt_fr_l_02, loc=<13 0 0>, rot=<0 0 0.0296 -0.9996>}
 dyn_skirt_fr_l_04 {parent=dyn_skirt_fr_l_03, loc=<13 0 0>, rot=<0 0 0.0431 -0.9991>}
 dyn_skirt_fr_l_05 {parent=dyn_skirt_fr_l_04, loc=<13 0 0>, rot=<0 0 0.0368 -0.9993>}
 dyn_skirt_fr_l_06 {parent=dyn_skirt_fr_l_05, loc=<13 0 0>, rot=<0 0 0.0585 -0.9983>}
 dyn_skirt_fr_l_07 {parent=dyn_skirt_fr_l_06, loc=<13 0 0>, rot=<0 0 0.0298 -0.9996>}
 dyn_skirt_fr_l_08 {parent=dyn_skirt_fr_l_07, loc=<13 0 0>, rot=<0 0 0 1>}
 dyn_skirt_fr_l_01_extra {parent=dyn_skirt_fr_l_01, loc=<0 0 0>, rot=<0 0 0 1>}
 base_skirt_bk_c {parent=pelvis, loc=<4.2714 -10.9847 0.0007>, rot=<0.0611 -0.7044 0.7045 -0.0611>}
 dyn_skirt_bk_01_c {parent=base_skirt_bk_c, loc=<0 0 0>, rot=<0 0 0 1>}
 dyn_skirt_bk_02_c {parent=dyn_skirt_bk_01_c, loc=<17.2607 0 0>, rot=<0.0003 -0.0123 -0.0001 -0.9999>}
 dyn_skirt_bk_03_c {parent=dyn_skirt_bk_02_c, loc=<17.2641 0 0>, rot=<-0.0002 0.0222 0.0001 -0.9998>}
 dyn_skirt_bk_04_c {parent=dyn_skirt_bk_03_c, loc=<17.2644 0 0>, rot=<0 0.0151 0 -0.9999>}
 dyn_skirt_bk_05_c {parent=dyn_skirt_bk_04_c, loc=<17.2644 0 0>, rot=<0 0.0243 0 -0.9997>}
 dyn_skirt_bk_06_c {parent=dyn_skirt_bk_05_c, loc=<17.2638 0 0>, rot=<0 0.0333 0 -0.9994>}
 skirt_pelvis_fr {parent=pelvis, loc=<1.1169 6.4705 5.3435>, rot=<0.0127 0.707 0.707 0.0127>}
 spine_05_null {parent=pelvis, type=Null, loc=<32.9717 -4.2232 0>, rot=<0 0 0.058 -0.9983>}
 floor {parent=root, loc=<0 0 -4.3982>, rot=<0 0 0.6078 -0.7941>}
 floor_aim {parent=root, loc=<-0.0036 -0.001 -4.3982>, rot=<0 0 0.6078 -0.7941>}
 dyn_skirt_fr_l_03_ctrl_null {type=Null, loc=<4.7535 8.9788 79.3962>, rot=<-0.0958 -0.5339 0.0431 -0.839>}
 dyn_skirt_fr_l_03_ctrl {controlType=EULER_TRANSFORM, parent=dyn_skirt_fr_l_03_ctrl_null, type=Control, loc=<0 0 0>, rot=<0 0 0 1>}
 dyn_skirt_fr_l_01_ctrl_null {type=Null, loc=<-6.6132 5.4821 102.1621>, rot=<-0.1361 -0.5251 -0.0207 -0.8398>}
 dyn_skirt_fr_l_01_ctrl {controlType=EULER_TRANSFORM, parent=dyn_skirt_fr_l_01_ctrl_null, type=Control, loc=<0 0 0>, rot=<0 0 0 1>}
 dyn_skirt_fr_l_05_ctrl_null {type=Null, loc=<15.6154 8.6612 55.8025>, rot=<-0.0529 -0.5399 0.11 -0.8328>}
 dyn_skirt_fr_l_05_ctrl {controlType=EULER_TRANSFORM, parent=dyn_skirt_fr_l_05_ctrl_null, type=Control, loc=<0 0 0>, rot=<0 0 0 1>}
 dyn_skirt_fr_l_07_ctrl_null {type=Null, loc=<25.4295 3.9204 32.2463>, rot=<-0.005 -0.5425 0.183 -0.8199>}
 dyn_skirt_fr_l_07_ctrl {controlType=EULER_TRANSFORM, parent=dyn_skirt_fr_l_07_ctrl_null, type=Control, loc=<0 0 0>, rot=<0 0 0 1>}
]
variables [
 AlphaEmotes {type=double, default=0}
 Crouching {type=double, default=0}
 CrouchingBool {type=bool, default=False}
 DBNO {type=bool, default=False}
 Emotes {type=bool, default=False}
 FEIdle {type=double, default=0}
 HangGliding {type=double, default=0}
 HeadRotateExtremCrouch {type=double, default=0}
 Idle {type=double, default=0}
 Leg Back L {type=double, default=0}
 Leg Back R {type=double, default=0}
 Leg BackSide L {type=double, default=0}
 Leg BackSide R {type=double, default=0}
 Leg Front L {type=double, default=0}
 Leg Front R {type=double, default=0}
 Leg FrontInside L {type=double, default=0}
 Leg FrontInside R {type=double, default=0}
 Leg FrontSide L {type=double, default=0}
 Leg FrontSide R {type=double, default=0}
 Leg Side L {type=double, default=0}
 Leg Side R {type=double, default=0}
 Moving2D {type=bool, default=False}
 OnGround {type=double, default=0}
 Pistol {type=double, default=0}
 RBAN Enable {type=bool, default=False}
 Rail {type=bool, default=False}
 Skydiving {type=bool, default=False}
 Standing {type=double, default=0}
 Targeting {type=bool, default=False}
]
