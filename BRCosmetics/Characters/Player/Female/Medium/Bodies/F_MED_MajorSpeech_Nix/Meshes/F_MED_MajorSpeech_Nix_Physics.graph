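# Physics asset for the F_MED_MajorSpeech_Nix female player body, covering its long skirt and the legs with capsule bodies.
# Front left/right, side left/right and back skirt chains hang from the pelvis, alongside thigh and calf capsules.
# Free joints attach chain roots and legs, while most links down the chains lock or tightly limit swing and twist.
# The right side chain also has extra free constraints linking segments to its first bone.
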PhysicsAsset /BRCosmetics/Characters/Player/Female/Medium/Bodies/F_MED_MajorSpeech_Nix/Meshes/F_MED_MajorSpeech_Nix_Physics
bodies {
 base_skirt_bk_c [{shapeType=Capsule, length=0.505, radius=0.505, shapeName=base_skirt_bk_c_capsule, rotation=<rot 0 -90 0>}]
 base_skirt_fr_l [{shapeType=Capsule, length=1, radius=1, shapeName=base_skirt_fr_l_capsule}]
 base_skirt_sd_l [{shapeType=Capsule, length=0.505, radius=0.505, shapeName=base_skirt_sd_l_capsule, rotation=<rot 0 -90 0>}]
 base_skirt_sd_r [{shapeType=Capsule, length=0.505, radius=0.505, shapeName=base_skirt_sd_r_capsule, rotation=<rot 0 -90 0>}]
 calf_l [
  {shapeType=Capsule, length=33.1543, radius=15.8671, shapeName=calf_l_capsule, center=<-31.9648 -0.58 -0.3011>, rotation=<rot 83.328 180 180>}
  {shapeType=Capsule, length=31.3189, radius=10.0721, shapeName=calf_l_capsule, center=<-31.857 -14.6273 0>, rotation=<rot 90 6.5457 0>}
 ]
 calf_r [
  {shapeType=Capsule, length=33.1543, radius=15.8671, shapeName=calf_r_capsule, center=<31.9648 0.58 0.3011>, rotation=<rot 83.328 180 0>}
  {shapeType=Capsule, length=15.7798, radius=14.2202, shapeName=calf_r_capsule, center=<20.1332 8.1212 3.1662>, rotation=<rot 90 0 0>}
 ]
 dyn_skirt_bk_01_c [{shapeType=Capsule, length=13.7135, radius=9, shapeName=dyn_skirt_bk_01_c_capsule, center=<8.4162 0 7.6342>, rotation=<rot 0 0 90>}]
 dyn_skirt_bk_02_c [{shapeType=Capsule, length=13.7135, radius=9, shapeName=dyn_skirt_bk_02_c_capsule, center=<8.4162 0 7.7824>, rotation=<rot 0 0 90>}]
 dyn_skirt_bk_03_c [{shapeType=Capsule, length=30.5872, radius=11, shapeName=dyn_skirt_bk_03_c_capsule, center=<8.4162 0 2.7798>, rotation=<rot 0 0 90>}]
 dyn_skirt_bk_04_c [{shapeType=Capsule, length=37.674, radius=11, shapeName=dyn_skirt_bk_04_c_capsule, center=<8.4162 0 -1.3837>, rotation=<rot 0 0 90>}]
 dyn_skirt_bk_05_c [{shapeType=Capsule, length=36.9201, radius=11, shapeName=dyn_skirt_bk_05_c_capsule, center=<8.4162 0 -1.1428>, rotation=<rot 0 0 90>}]
 dyn_skirt_bk_06_c [{shapeType=Capsule, length=36.1968, radius=11, shapeName=dyn_skirt_bk_06_c_capsule, center=<8.4162 0 -1.6204>, rotation=<rot 0 0 90>}]
 dyn_skirt_fr_l_02 [{shapeType=Capsule, length=1, radius=1, shapeName=dyn_skirt_fr_l_02_capsule}]
 dyn_skirt_fr_l_03 [{shapeType=Capsule, length=1, radius=1, shapeName=dyn_skirt_fr_l_03_capsule}]
 dyn_skirt_fr_l_04 [{
   shapeType=Capsule
   length=13.7495
   radius=7
   shapeName=dyn_skirt_fr_l_04_capsule
   center=<3.058 3.719 7.5114>
   rotation=<rot -1.3509 0.6307 -32.9667>
  }]
 dyn_skirt_fr_l_05 [{
   shapeType=Capsule
   length=13.7495
   radius=7
   shapeName=dyn_skirt_fr_l_05_capsule
   center=<3 1.7722 7.7621>
   rotation=<rot -0.0002 0.0006 -38.149>
  }]
 dyn_skirt_fr_l_06 [{
   shapeType=Capsule
   length=13.7495
   radius=8
   shapeName=dyn_skirt_fr_l_06_capsule
   center=<3 0.1098 9.157>
   rotation=<rot -0.0001 0.0004 -53.2062>
  }]
 dyn_skirt_fr_l_07 [{
   shapeType=Capsule
   length=13.7495
   radius=8
   shapeName=dyn_skirt_fr_l_07_capsule
   center=<3 -1.4212 10.87>
   rotation=<rot -0.0118 0.0344 -59.6619>
  }]
 dyn_skirt_fr_l_08 [{
   shapeType=Capsule
   length=13.7495
   radius=8
   shapeName=dyn_skirt_fr_l_08_capsule
   center=<3 -2.4933 7.9779>
   rotation=<rot -0.0001 0.0002 -69.0158>
  }]
 dyn_skirt_fr_r_02 [{
   shapeType=Capsule
   length=4.8215
   radius=4.0282
   shapeName=dyn_skirt_fr_r_02_capsule
   center=<4.6605 0.4377 1.8962>
   rotation=<rot 0 -91.8528 -179.9077>
  }]
 dyn_skirt_fr_r_03 [{
   shapeType=Capsule
   length=13.7
   radius=6
   shapeName=dyn_skirt_fr_r_03_capsule
   center=<2 -4.5375 1.8524>
   rotation=<rot -0.0001 -0.0002 17.8722>
  }]
 dyn_skirt_fr_r_04 [{
   shapeType=Capsule
   length=13.7495
   radius=6
   shapeName=dyn_skirt_fr_r_04_capsule
   center=<2 -4.5375 1.8524>
   rotation=<rot -0.0001 -0.0001 21.7883>
  }]
 dyn_skirt_fr_r_05 [{
   shapeType=Capsule
   length=13.7495
   radius=6
   shapeName=dyn_skirt_fr_r_05_capsule
   center=<2 -4.5375 1.8524>
   rotation=<rot -0.0001 -0.0001 32.0227>
  }]
 dyn_skirt_fr_r_06 [{
   shapeType=Capsule
   length=13.7495
   radius=6
   shapeName=dyn_skirt_fr_r_06_capsule
   center=<2 -4.5375 1.8524>
   rotation=<rot -0.0001 -0.0002 42.0625>
  }]
 dyn_skirt_fr_r_07 [{
   shapeType=Capsule
   length=13.7495
   radius=6
   shapeName=dyn_skirt_fr_r_07_capsule
   center=<2 -4.5375 1.8524>
   rotation=<rot -0.0001 -0.0001 42.6159>
  }]
 dyn_skirt_fr_r_08 [{
   shapeType=Capsule
   length=13.7495
   radius=6
   shapeName=dyn_skirt_fr_r_08_capsule
   center=<2 -4.5375 1.8524>
   rotation=<rot -0.0001 -0.0001 42.6159>
  }]
 dyn_skirt_sd_l_01 [{
   shapeType=Capsule
   length=11.0005
   radius=7
   shapeName=dyn_skirt_sd_l_01_capsule
   center=<5.3712 -0.7937 9.7804>
   rotation=<rot -69.8993 -89.9343 0.1085>
  }]
 dyn_skirt_sd_l_02 [{
   shapeType=Capsule
   length=9.9518
   radius=7
   shapeName=dyn_skirt_sd_l_02_capsule
   center=<5.3665 -2.2723 9.5543>
   rotation=<rot -83.3711 -90.0002 0.0008>
  }]
 dyn_skirt_sd_l_03 [{
   shapeType=Capsule
   length=15.8944
   radius=9
   shapeName=dyn_skirt_sd_l_03_capsule
   center=<5.3681 -3.6188 9.0139>
   rotation=<rot -89.7233 90.0586 179.9967>
  }]
 dyn_skirt_sd_l_04 [{
   shapeType=Capsule
   length=35.8714
   radius=9
   shapeName=dyn_skirt_sd_l_04_capsule
   center=<5.3683 -2.9601 5.0904>
   rotation=<rot -83.6325 90.0376 179.9996>
  }]
 dyn_skirt_sd_l_05 [{
   shapeType=Capsule
   length=35.8714
   radius=9
   shapeName=dyn_skirt_sd_l_05_capsule
   center=<5.3664 0 1.8173>
   rotation=<rot -83.0037 89.9997 179.9997>
  }]
 dyn_skirt_sd_l_06 [{
   shapeType=Capsule
   length=35.8714
   radius=9
   shapeName=dyn_skirt_sd_l_06_capsule
   center=<10.4135 0 1.3342>
   rotation=<rot -81.4159 90.0559 -175.5985>
  }]
 dyn_skirt_sd_r_01 [{
   shapeType=Capsule
   length=11.0005
   radius=7
   shapeName=dyn_skirt_sd_r_01_capsule
   center=<5.368 0.4341 7.9022>
   rotation=<rot 78.0973 -89.8592 -179.819>
  }]
 dyn_skirt_sd_r_02 [{
   shapeType=Capsule
   length=9.9518
   radius=7
   shapeName=dyn_skirt_sd_r_02_capsule
   center=<5.3664 1.2274 8.4665>
   rotation=<rot 87.9535 89.9641 -0.0011>
  }]
 dyn_skirt_sd_r_03 [{
   shapeType=Capsule
   length=15.8944
   radius=9
   shapeName=dyn_skirt_sd_r_03_capsule
   center=<5.3664 0 4.8296>
   rotation=<rot 87.4568 -89.9988 -179.9988>
  }]
 dyn_skirt_sd_r_04 [{
   shapeType=Capsule
   length=32.7602
   radius=9
   shapeName=dyn_skirt_sd_r_04_capsule
   center=<5.3664 0 2.9565>
   rotation=<rot 83.6325 90.0376 -0.0004>
  }]
 dyn_skirt_sd_r_05 [{
   shapeType=Capsule
   length=32.7602
   radius=9
   shapeName=dyn_skirt_sd_r_05_capsule
   center=<5.3664 0 1.8173>
   rotation=<rot 83.0037 89.9997 -0.0003>
  }]
 dyn_skirt_sd_r_06 [{
   shapeType=Capsule
   length=32.7602
   radius=9
   shapeName=dyn_skirt_sd_r_06_capsule
   center=<10.4135 0 1.3342>
   rotation=<rot 81.4159 90.0559 4.4015>
  }]
 pelvis [{shapeType=Capsule, length=5, radius=5}]
 thigh_l [{
   shapeType=Capsule
   length=34.0762
   radius=10.1077
   shapeName=thigh_l_capsule
   center=<-22.8188 -0.906 1.98>
   rotation=<rot 1.6219 83.6407 -86.5558>
  }]
 thigh_r [{
   shapeType=Capsule
   length=34.0762
   radius=10.1077
   shapeName=thigh_r_capsule
   center=<22.8188 0.906 -1.98>
   rotation=<rot 1.6219 83.6407 93.4442>
  }]
}
constraints [
 {
  bone1Name=dyn_skirt_fr_l_04
  bone2Name=dyn_skirt_fr_l_03
  swing1LimitDegrees=10
  swing1Motion=Locked
  swing2LimitDegrees=7
  twistLimitDegrees=15
  twistMotion=Locked
 }
 {
  bone1Name=dyn_skirt_fr_l_05
  bone2Name=dyn_skirt_fr_l_04
  swing1LimitDegrees=10
  swing1Motion=Locked
  swing2LimitDegrees=12
  twistLimitDegrees=4
 }
 {
  bone1Name=dyn_skirt_fr_l_06
  bone2Name=dyn_skirt_fr_l_05
  swing1LimitDegrees=10
  swing1Motion=Locked
  swing2LimitDegrees=20
  twistLimitDegrees=7
 }
 {bone1Name=dyn_skirt_fr_l_07, bone2Name=dyn_skirt_fr_l_06, swing1LimitDegrees=2, swing2LimitDegrees=20, twistLimitDegrees=7}
 {
  bone1Name=dyn_skirt_fr_l_08
  bone2Name=dyn_skirt_fr_l_07
  swing1LimitDegrees=10
  swing1Motion=Locked
  swing2LimitDegrees=20
  twistLimitDegrees=5
 }
 {
  bone1Name=base_skirt_fr_l
  bone2Name=pelvis
  swing1LimitDegrees=45
  swing1Motion=Free
  swing2LimitDegrees=45
  swing2Motion=Free
  twistLimitDegrees=45
  twistMotion=Free
 }
 {
  bone1Name=dyn_skirt_fr_l_02
  bone2Name=base_skirt_fr_l
  swing1LimitDegrees=45
  swing1Motion=Free
  swing2LimitDegrees=45
  swing2Motion=Free
  twistLimitDegrees=45
  twistMotion=Free
 }
 {
  bone1Name=dyn_skirt_fr_l_03
  bone2Name=dyn_skirt_fr_l_02
  swing1LimitDegrees=45
  swing1Motion=Free
  swing2LimitDegrees=45
  swing2Motion=Free
  twistLimitDegrees=45
  twistMotion=Free
 }
 {
  bone1Name=dyn_skirt_sd_l_01
  bone2Name=base_skirt_sd_l
  swing1LimitDegrees=5
  swing1Motion=Locked
  swing2LimitDegrees=5
  twistLimitDegrees=45
  twistMotion=Locked
 }
 {
  bone1Name=dyn_skirt_sd_l_02
  bone2Name=dyn_skirt_sd_l_01
  swing1LimitDegrees=5
  swing1Motion=Locked
  swing2LimitDegrees=10
  twistLimitDegrees=45
  twistMotion=Locked
 }
 {
  bone1Name=dyn_skirt_sd_l_03
  bone2Name=dyn_skirt_sd_l_02
  swing1LimitDegrees=5
  swing1Motion=Locked
  swing2LimitDegrees=10
  twistLimitDegrees=45
  twistMotion=Locked
 }
 {
  bone1Name=dyn_skirt_sd_l_04
  bone2Name=dyn_skirt_sd_l_03
  swing1LimitDegrees=5
  swing1Motion=Locked
  swing2LimitDegrees=10
  twistLimitDegrees=5
  twistMotion=Locked
 }
 {
  bone1Name=dyn_skirt_sd_l_05
  bone2Name=dyn_skirt_sd_l_04
  swing1LimitDegrees=1.5
  swing2LimitDegrees=10
  twistLimitDegrees=5
  twistMotion=Locked
 }
 {
  bone1Name=dyn_skirt_sd_l_06
  bone2Name=dyn_skirt_sd_l_05
  swing1LimitDegrees=2
  swing2LimitDegrees=10
  twistLimitDegrees=5
  twistMotion=Locked
 }
 {
  bone1Name=base_skirt_sd_l
  bone2Name=pelvis
  swing1LimitDegrees=45
  swing1Motion=Free
  swing2LimitDegrees=45
  swing2Motion=Free
  twistLimitDegrees=45
  twistMotion=Free
 }
 {
  bone1Name=base_skirt_bk_c
  bone2Name=pelvis
  swing1LimitDegrees=45
  swing1Motion=Free
  swing2LimitDegrees=45
  swing2Motion=Free
  twistLimitDegrees=45
  twistMotion=Free
 }
 {
  bone1Name=dyn_skirt_bk_01_c
  bone2Name=base_skirt_bk_c
  swing1LimitDegrees=10
  swing1Motion=Locked
  swing2LimitDegrees=10
  twistLimitDegrees=5
 }
 {
  bone1Name=dyn_skirt_bk_02_c
  bone2Name=dyn_skirt_bk_01_c
  swing1LimitDegrees=10
  swing1Motion=Locked
  swing2LimitDegrees=10
  twistLimitDegrees=5
 }
 {
  bone1Name=dyn_skirt_bk_03_c
  bone2Name=dyn_skirt_bk_02_c
  swing1LimitDegrees=10
  swing1Motion=Locked
  swing2LimitDegrees=10
  twistLimitDegrees=5
 }
 {
  bone1Name=dyn_skirt_bk_04_c
  bone2Name=dyn_skirt_bk_03_c
  swing1LimitDegrees=10
  swing1Motion=Locked
  swing2LimitDegrees=10
  twistLimitDegrees=7
 }
 {bone1Name=dyn_skirt_bk_05_c, bone2Name=dyn_skirt_bk_04_c, swing1LimitDegrees=5, swing2LimitDegrees=20, twistLimitDegrees=7}
 {bone1Name=dyn_skirt_bk_06_c, bone2Name=dyn_skirt_bk_05_c, swing1LimitDegrees=5, swing2LimitDegrees=20, twistLimitDegrees=7}
 {
  bone1Name=thigh_l
  bone2Name=pelvis
  swing1LimitDegrees=45
  swing1Motion=Free
  swing2LimitDegrees=45
  swing2Motion=Free
  twistLimitDegrees=45
  twistMotion=Free
 }
 {
  bone1Name=thigh_r
  bone2Name=pelvis
  swing1LimitDegrees=45
  swing1Motion=Free
  swing2LimitDegrees=45
  swing2Motion=Free
  twistLimitDegrees=45
  twistMotion=Free
 }
 {
  bone1Name=calf_l
  bone2Name=thigh_l
  swing1LimitDegrees=45
  swing1Motion=Free
  swing2LimitDegrees=45
  swing2Motion=Free
  twistLimitDegrees=45
  twistMotion=Free
 }
 {
  bone1Name=calf_r
  bone2Name=thigh_r
  swing1LimitDegrees=45
  swing1Motion=Free
  swing2LimitDegrees=45
  swing2Motion=Free
  twistLimitDegrees=45
  twistMotion=Free
 }
 {
  bone1Name=dyn_skirt_sd_r_01
  bone2Name=base_skirt_sd_r
  swing1LimitDegrees=5
  swing1Motion=Locked
  swing2LimitDegrees=5
  twistLimitDegrees=45
  twistMotion=Locked
 }
 {
  bone1Name=dyn_skirt_sd_r_02
  bone2Name=dyn_skirt_sd_r_01
  swing1LimitDegrees=5
  swing1Motion=Locked
  swing2LimitDegrees=10
  twistLimitDegrees=10
  twistMotion=Locked
 }
 {
  bone1Name=dyn_skirt_sd_r_03
  bone2Name=dyn_skirt_sd_r_02
  swing1LimitDegrees=5
  swing1Motion=Locked
  swing2LimitDegrees=10
  twistLimitDegrees=10
  twistMotion=Locked
 }
 {
  bone1Name=dyn_skirt_sd_r_04
  bone2Name=dyn_skirt_sd_r_03
  swing1LimitDegrees=5
  swing1Motion=Locked
  swing2LimitDegrees=10
  twistLimitDegrees=5
  twistMotion=Locked
 }
 {
  bone1Name=dyn_skirt_sd_r_05
  bone2Name=dyn_skirt_sd_r_04
  swing1LimitDegrees=1.5
  swing2LimitDegrees=10
  twistLimitDegrees=5
  twistMotion=Locked
 }
 {
  bone1Name=dyn_skirt_sd_r_06
  bone2Name=dyn_skirt_sd_r_05
  swing1LimitDegrees=2
  swing2LimitDegrees=10
  twistLimitDegrees=5
  twistMotion=Locked
 }
 {
  bone1Name=base_skirt_sd_r
  bone2Name=pelvis
  swing1LimitDegrees=45
  swing1Motion=Free
  swing2LimitDegrees=45
  swing2Motion=Free
  twistLimitDegrees=45
  twistMotion=Free
 }
 {
  bone1Name=dyn_skirt_fr_r_08
  bone2Name=dyn_skirt_fr_r_07
  swing1LimitDegrees=20
  swing1Motion=Locked
  swing2LimitDegrees=10
  twistLimitDegrees=45
  twistMotion=Locked
 }
 {
  bone1Name=dyn_skirt_fr_r_06
  bone2Name=dyn_skirt_fr_r_05
  swing1LimitDegrees=10
  swing1Motion=Locked
  swing2LimitDegrees=10
  twistLimitDegrees=45
  twistMotion=Locked
 }
 {
  bone1Name=dyn_skirt_fr_r_07
  bone2Name=dyn_skirt_fr_r_06
  swing1LimitDegrees=20
  swing1Motion=Locked
  swing2LimitDegrees=10
  twistLimitDegrees=45
  twistMotion=Locked
 }
 {
  bone1Name=dyn_skirt_fr_r_05
  bone2Name=dyn_skirt_fr_r_04
  swing1LimitDegrees=10
  swing1Motion=Locked
  swing2LimitDegrees=10
  twistLimitDegrees=45
  twistMotion=Locked
 }
 {
  bone1Name=dyn_skirt_fr_r_04
  bone2Name=dyn_skirt_fr_r_03
  swing1LimitDegrees=10
  swing1Motion=Locked
  swing2LimitDegrees=10
  twistLimitDegrees=45
  twistMotion=Locked
 }
 {
  bone1Name=dyn_skirt_fr_r_03
  bone2Name=dyn_skirt_fr_r_02
  swing1LimitDegrees=20
  swing2LimitDegrees=10
  twistLimitDegrees=45
  twistMotion=Locked
 }
 {
  bone1Name=dyn_skirt_fr_r_02
  bone2Name=pelvis
  swing1LimitDegrees=45
  swing1Motion=Free
  swing2LimitDegrees=45
  swing2Motion=Free
  twistLimitDegrees=45
  twistMotion=Free
 }
 {
  bone1Name=dyn_skirt_sd_r_06
  bone2Name=dyn_skirt_sd_r_01
  swing1LimitDegrees=45
  swing1Motion=Free
  swing2LimitDegrees=45
  swing2Motion=Free
  twistLimitDegrees=45
  twistMotion=Free
 }
 {
  bone1Name=dyn_skirt_sd_r_02
  bone2Name=dyn_skirt_sd_r_01
  swing1LimitDegrees=45
  swing1Motion=Free
  swing2LimitDegrees=45
  swing2Motion=Free
  twistLimitDegrees=45
  twistMotion=Free
 }
 {
  bone1Name=dyn_skirt_sd_r_03
  bone2Name=dyn_skirt_sd_r_01
  swing1LimitDegrees=45
  swing1Motion=Free
  swing2LimitDegrees=45
  swing2Motion=Free
  twistLimitDegrees=45
  twistMotion=Free
 }
 {
  bone1Name=dyn_skirt_sd_r_04
  bone2Name=dyn_skirt_sd_r_01
  swing1LimitDegrees=45
  swing1Motion=Free
  swing2LimitDegrees=45
  swing2Motion=Free
  twistLimitDegrees=45
  twistMotion=Free
 }
 {
  bone1Name=dyn_skirt_sd_r_05
  bone2Name=dyn_skirt_sd_r_01
  swing1LimitDegrees=45
  swing1Motion=Free
  swing2LimitDegrees=45
  swing2Motion=Free
  twistLimitDegrees=45
  twistMotion=Free
 }
]
