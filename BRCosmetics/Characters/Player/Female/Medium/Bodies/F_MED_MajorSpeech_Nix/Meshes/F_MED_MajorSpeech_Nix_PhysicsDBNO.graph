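# Physics asset for the skirt of the Nix MajorSpeech medium female body, apparently for its DBNO (downed) state.
# Capsule chains form the front-left, front-right, side and back panels, rooted at the pelvis, mostly through base bones, with free root joints.
# Links down the chains mostly use limited swing with selected axes locked.
# Thigh and calf capsules stand in for the legs, and a large offset floor sphere approximates the ground.
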PhysicsAsset /BRCosmetics/Characters/Player/Female/Medium/Bodies/F_MED_MajorSpeech_Nix/Meshes/F_MED_MajorSpeech_Nix_PhysicsDBNO
bodies {
 base_skirt_bk_c [{shapeType=Capsule, length=0.505, radius=0.505, shapeName=base_skirt_bk_c_capsule, rotation=<rot 0 -90 0>}]
 base_skirt_fr_l [{shapeType=Capsule, length=1, radius=1, shapeName=base_skirt_fr_l_capsule}]
 base_skirt_sd_l [{shapeType=Capsule, length=0.505, radius=0.505, shapeName=base_skirt_sd_l_capsule, rotation=<rot 0 -90 0>}]
 base_skirt_sd_r [{shapeType=Capsule, length=0.505, radius=0.505, shapeName=base_skirt_sd_r_capsule, rotation=<rot 0 -90 0>}]
 calf_l [{shapeType=Capsule, length=25, radius=18, shapeName=calf_l_capsule, center=<-27.796 -0.58 0.1866>, rotation=<rot 83.328 180 180>}]
 calf_r [{shapeType=Capsule, length=24.7831, radius=17, shapeName=calf_r_capsule, center=<27.6355 0.58 1.1853>, rotation=<rot 83.328 180 0>}]
 dyn_skirt_bk_01_c [{shapeType=Capsule, length=13.7135, radius=9, shapeName=dyn_skirt_bk_01_c_capsule, center=<8.4162 0 6.4565>, rotation=<rot 0 0 90>}]
 dyn_skirt_bk_02_c [{shapeType=Capsule, length=13.7135, radius=9, shapeName=dyn_skirt_bk_02_c_capsule, center=<8.4162 0 6.6873>, rotation=<rot 0 0 90>}]
 dyn_skirt_bk_03_c [{shapeType=Capsule, length=30.5872, radius=11, shapeName=dyn_skirt_bk_03_c_capsule, center=<8.4162 0 2.7798>, rotation=<rot 0 0 90>}]
 dyn_skirt_bk_04_c [{shapeType=Capsule, length=30.5872, radius=11, shapeName=dyn_skirt_bk_04_c_capsule, center=<8.4162 0 6.4292>, rotation=<rot 0 0 90>}]
 dyn_skirt_bk_05_c [{shapeType=Capsule, length=29.8333, radius=11, shapeName=dyn_skirt_bk_05_c_capsule, center=<8.4162 0 6.5233>, rotation=<rot 0 0 90>}]
 dyn_skirt_bk_06_c [{shapeType=Capsule, length=29.1101, radius=11, shapeName=dyn_skirt_bk_06_c_capsule, center=<7.1487 0 3.5197>, rotation=<rot 0 0 90>}]
 dyn_skirt_fr_l_02 [{shapeType=Capsule, length=1, radius=1, shapeName=dyn_skirt_fr_l_02_capsule}]
 dyn_skirt_fr_l_03 [{shapeType=Capsule, length=1, radius=1, shapeName=dyn_skirt_fr_l_03_capsule}]
 dyn_skirt_fr_l_04 [{
   shapeType=Capsule
   length=13.7495
   radius=6
   shapeName=dyn_skirt_fr_l_04_capsule
   center=<3.0559 1.9201 6.5866>
   rotation=<rot -1.3503 0.6279 -24.933>
  }]
 dyn_skirt_fr_l_05 [{
   shapeType=Capsule
   length=13.7495
   radius=8
   shapeName=dyn_skirt_fr_l_05_capsule
   center=<3 0.1432 6.4318>
   rotation=<rot -0.0001 0.0004 -63.2757>
  }]
 dyn_skirt_fr_l_06 [{
   shapeType=Capsule
   length=13.7495
   radius=12.5
   shapeName=dyn_skirt_fr_l_06_capsule
   center=<3 -0.5895 5.3961>
   rotation=<rot -0.0001 0.0004 -71.693>
  }]
 dyn_skirt_fr_l_07 [{
   shapeType=Capsule
   length=13.7495
   radius=12.5
   shapeName=dyn_skirt_fr_l_07_capsule
   center=<3 -1.1143 5.3946>
   rotation=<rot -0.0001 0.0003 -75.0842>
  }]
 dyn_skirt_fr_l_08 [{
   shapeType=Capsule
   length=13.7495
   radius=12.5
   shapeName=dyn_skirt_fr_l_08_capsule
   center=<3 -0.1018 5.1053>
   rotation=<rot -0.0001 0.0003 -75.0842>
  }]
 dyn_skirt_fr_r_02 [{
   shapeType=Capsule
   length=4.8215
   radius=4.0282
   shapeName=dyn_skirt_fr_r_02_capsule
   center=<4.6605 0.4377 1.8962>
   rotation=<rot 0 -91.8528 -179.9077>
  }]
 dyn_skirt_fr_r_03 [{
   shapeType=Capsule
   length=13.7
   radius=6
   shapeName=dyn_skirt_fr_r_03_capsule
   center=<2 -4.5375 1.8524>
   rotation=<rot -0.0001 -0.0002 17.8722>
  }]
 dyn_skirt_fr_r_04 [{
   shapeType=Capsule
   length=13.7495
   radius=7.5
   shapeName=dyn_skirt_fr_r_04_capsule
   center=<2 -4.3047 1.6367>
   rotation=<rot -0.0001 -0.0002 42.8117>
  }]
 dyn_skirt_fr_r_05 [{
   shapeType=Capsule
   length=13.7495
   radius=7.5
   shapeName=dyn_skirt_fr_r_05_capsule
   center=<2 -3.6916 2.6785>
   rotation=<rot -0.0001 -0.0002 45.6828>
  }]
 dyn_skirt_fr_r_06 [{
   shapeType=Capsule
   length=24.5731
   radius=9
   shapeName=dyn_skirt_fr_r_06_capsule
   center=<2 -2.277 3.4335>
   rotation=<rot -0.0002 -0.0002 55.4484>
  }]
 dyn_skirt_fr_r_07 [{
   shapeType=Capsule
   length=24.5731
   radius=9
   shapeName=dyn_skirt_fr_r_07_capsule
   center=<2 -4.1759 6.1058>
   rotation=<rot -0.0001 -0.0002 56.0018>
  }]
 dyn_skirt_fr_r_08 [{
   shapeType=Capsule
   length=24.5731
   radius=9
   shapeName=dyn_skirt_fr_r_08_capsule
   center=<2 -4.2366 5.8776>
   rotation=<rot -0.0001 -0.0002 56.0018>
  }]
 dyn_skirt_sd_l_01 [{shapeType=Capsule, length=11.0005, radius=8, shapeName=dyn_skirt_sd_l_01_capsule, center=<5.3664 0 2.9074>, rotation=<rot 90 -90 0>}]
 dyn_skirt_sd_l_02 [{shapeType=Capsule, length=13.5775, radius=8, shapeName=dyn_skirt_sd_l_02_capsule, center=<5.3664 0 2.9074>, rotation=<rot 90 -90 0>}]
 dyn_skirt_sd_l_03 [{shapeType=Capsule, length=19.52, radius=9, shapeName=dyn_skirt_sd_l_03_capsule, center=<5.3664 -2.4338 2.9074>, rotation=<rot 90 -90 0>}]
 dyn_skirt_sd_l_04 [{
   shapeType=Capsule
   length=31.0203
   radius=9
   shapeName=dyn_skirt_sd_l_04_capsule
   center=<5.3664 -2.2754 6.2321>
   rotation=<rot 87.8533 -89.9946 0.0052>
  }]
 dyn_skirt_sd_l_05 [{
   shapeType=Capsule
   length=31.0203
   radius=8
   shapeName=dyn_skirt_sd_l_05_capsule
   center=<5.3664 -0.057 2.7227>
   rotation=<rot 85.7364 -89.9921 0.0075>
  }]
 dyn_skirt_sd_l_06 [{
   shapeType=Capsule
   length=31.0203
   radius=8
   shapeName=dyn_skirt_sd_l_06_capsule
   center=<10.3837 -2.7508 5.4911>
   rotation=<rot 81.6867 -89.9985 -4.4455>
  }]
 dyn_skirt_sd_r_01 [{
   shapeType=Capsule
   length=11.0005
   radius=8
   shapeName=dyn_skirt_sd_r_01_capsule
   center=<5.3664 0 5.9403>
   rotation=<rot 80.2406 -89.8198 -179.7804>
  }]
 dyn_skirt_sd_r_02 [{
   shapeType=Capsule
   length=9.9518
   radius=8
   shapeName=dyn_skirt_sd_r_02_capsule
   center=<5.3664 0.1052 5.2598>
   rotation=<rot 81.9901 -89.9998 -179.9997>
  }]
 dyn_skirt_sd_r_03 [{
   shapeType=Capsule
   length=15.8944
   radius=8
   shapeName=dyn_skirt_sd_r_03_capsule
   center=<5.3664 0.0697 3.2603>
   rotation=<rot 87.4568 -89.9988 -179.9988>
  }]
 dyn_skirt_sd_r_04 [{
   shapeType=Capsule
   length=28
   radius=8
   shapeName=dyn_skirt_sd_r_04_capsule
   center=<5.3665 -0.17 1.4333>
   rotation=<rot 83.6325 90.0376 -0.0004>
  }]
 dyn_skirt_sd_r_05 [{
   shapeType=Capsule
   length=28
   radius=8
   shapeName=dyn_skirt_sd_r_05_capsule
   center=<5.3664 0 1.8173>
   rotation=<rot 83.0037 89.9997 -0.0003>
  }]
 dyn_skirt_sd_r_06 [{
   shapeType=Capsule
   length=28
   radius=8
   shapeName=dyn_skirt_sd_r_06_capsule
   center=<10.4135 0 1.3342>
   rotation=<rot 81.4159 90.0559 4.4015>
  }]
 floor [{shapeType=Sphere, radius=345, shapeName=floor_sphere, center=<0 0 -330>}]
 pelvis [{shapeType=Capsule, length=5.923, radius=6.5, center=<0 -2.5968 0>}]
 thigh_l [{
   shapeType=Capsule
   length=25.2088
   radius=10.0711
   shapeName=thigh_l_capsule
   center=<-32.4447 -1.0684 1.0051>
   rotation=<rot 3.9381 79.4961 -90>
  }]
 thigh_r [{shapeType=Capsule, length=25.2088, radius=12, shapeName=thigh_r_capsule, center=<32.4447 1.0684 -1.0051>, rotation=<rot 3.9381 79.4961 90>}]
}
constraints [
 {
  bone1Name=dyn_skirt_fr_l_04
  bone2Name=dyn_skirt_fr_l_03
  swing1LimitDegrees=20
  swing2LimitDegrees=7
  swing2Motion=Locked
  twistLimitDegrees=45
  twistMotion=Locked
 }
 {
  bone1Name=dyn_skirt_fr_l_05
  bone2Name=dyn_skirt_fr_l_04
  swing1LimitDegrees=20
  swing2LimitDegrees=10
  swing2Motion=Locked
  twistLimitDegrees=10
 }
 {
  bone1Name=dyn_skirt_fr_l_06
  bone2Name=dyn_skirt_fr_l_05
  swing1LimitDegrees=20
  swing2LimitDegrees=10
  swing2Motion=Locked
  twistLimitDegrees=10
 }
 {
  bone1Name=dyn_skirt_fr_l_07
  bone2Name=dyn_skirt_fr_l_06
  swing1LimitDegrees=20
  swing2LimitDegrees=10
  swing2Motion=Locked
  twistLimitDegrees=10
 }
 {
  bone1Name=dyn_skirt_fr_l_08
  bone2Name=dyn_skirt_fr_l_07
  swing1LimitDegrees=20
  swing2LimitDegrees=10
  swing2Motion=Locked
  twistLimitDegrees=10
 }
 {
  bone1Name=base_skirt_fr_l
  bone2Name=pelvis
  swing1LimitDegrees=45
  swing1Motion=Free
  swing2LimitDegrees=45
  swing2Motion=Free
  twistLimitDegrees=45
  twistMotion=Free
 }
 {
  bone1Name=dyn_skirt_fr_l_02
  bone2Name=base_skirt_fr_l
  swing1LimitDegrees=45
  swing1Motion=Free
  swing2LimitDegrees=45
  swing2Motion=Free
  twistLimitDegrees=45
  twistMotion=Free
 }
 {
  bone1Name=dyn_skirt_fr_l_03
  bone2Name=dyn_skirt_fr_l_02
  swing1LimitDegrees=45
  swing1Motion=Free
  swing2LimitDegrees=45
  swing2Motion=Free
  twistLimitDegrees=45
  twistMotion=Free
 }
 {
  bone1Name=dyn_skirt_sd_l_01
  bone2Name=base_skirt_sd_l
  swing1LimitDegrees=5
  swing1Motion=Locked
  swing2LimitDegrees=10
  twistLimitDegrees=45
  twistMotion=Locked
 }
 {
  bone1Name=dyn_skirt_sd_l_02
  bone2Name=dyn_skirt_sd_l_01
  swing1LimitDegrees=5
  swing1Motion=Locked
  swing2LimitDegrees=25
  twistLimitDegrees=45
  twistMotion=Locked
 }
 {
  bone1Name=dyn_skirt_sd_l_03
  bone2Name=dyn_skirt_sd_l_02
  swing1LimitDegrees=5
  swing1Motion=Locked
  swing2LimitDegrees=25
  twistLimitDegrees=45
  twistMotion=Locked
 }
 {
  bone1Name=dyn_skirt_sd_l_04
  bone2Name=dyn_skirt_sd_l_03
  swing1LimitDegrees=5
  swing1Motion=Locked
  swing2LimitDegrees=20
  twistLimitDegrees=5
  twistMotion=Locked
 }
 {
  bone1Name=dyn_skirt_sd_l_05
  bone2Name=dyn_skirt_sd_l_04
  swing1LimitDegrees=5
  swing1Motion=Locked
  swing2LimitDegrees=20
  twistLimitDegrees=5
  twistMotion=Locked
 }
 {
  bone1Name=dyn_skirt_sd_l_06
  bone2Name=dyn_skirt_sd_l_05
  swing1LimitDegrees=5
  swing1Motion=Locked
  swing2LimitDegrees=20
  twistLimitDegrees=5
  twistMotion=Locked
 }
 {
  bone1Name=base_skirt_sd_l
  bone2Name=pelvis
  swing1LimitDegrees=45
  swing1Motion=Free
  swing2LimitDegrees=45
  swing2Motion=Free
  twistLimitDegrees=45
  twistMotion=Free
 }
 {
  bone1Name=base_skirt_bk_c
  bone2Name=pelvis
  swing1LimitDegrees=45
  swing1Motion=Free
  swing2LimitDegrees=45
  swing2Motion=Free
  twistLimitDegrees=45
  twistMotion=Free
 }
 {
  bone1Name=dyn_skirt_bk_01_c
  bone2Name=base_skirt_bk_c
  swing1LimitDegrees=10
  swing1Motion=Locked
  swing2LimitDegrees=30
  twistLimitDegrees=45
  twistMotion=Locked
 }
 {
  bone1Name=dyn_skirt_bk_02_c
  bone2Name=dyn_skirt_bk_01_c
  swing1LimitDegrees=10
  swing1Motion=Locked
  swing2LimitDegrees=30
  twistLimitDegrees=45
  twistMotion=Locked
 }
 {
  bone1Name=dyn_skirt_bk_03_c
  bone2Name=dyn_skirt_bk_02_c
  swing1LimitDegrees=10
  swing1Motion=Locked
  swing2LimitDegrees=30
  twistLimitDegrees=5
  twistMotion=Locked
 }
 {
  bone1Name=dyn_skirt_bk_04_c
  bone2Name=dyn_skirt_bk_03_c
  swing1LimitDegrees=10
  swing1Motion=Locked
  swing2LimitDegrees=30
  twistLimitDegrees=5
  twistMotion=Locked
 }
 {
  bone1Name=dyn_skirt_bk_05_c
  bone2Name=dyn_skirt_bk_04_c
  swing1LimitDegrees=10
  swing1Motion=Locked
  swing2LimitDegrees=40
  twistLimitDegrees=5
  twistMotion=Locked
 }
 {
  bone1Name=dyn_skirt_bk_06_c
  bone2Name=dyn_skirt_bk_05_c
  swing1LimitDegrees=10
  swing1Motion=Locked
  swing2LimitDegrees=40
  twistLimitDegrees=5
  twistMotion=Locked
 }
 {
  bone1Name=thigh_l
  bone2Name=pelvis
  swing1LimitDegrees=45
  swing1Motion=Free
  swing2LimitDegrees=45
  swing2Motion=Free
  twistLimitDegrees=45
  twistMotion=Free
 }
 {
  bone1Name=thigh_r
  bone2Name=pelvis
  swing1LimitDegrees=45
  swing1Motion=Free
  swing2LimitDegrees=45
  swing2Motion=Free
  twistLimitDegrees=45
  twistMotion=Free
 }
 {
  bone1Name=calf_l
  bone2Name=thigh_l
  swing1LimitDegrees=45
  swing1Motion=Free
  swing2LimitDegrees=45
  swing2Motion=Free
  twistLimitDegrees=45
  twistMotion=Free
 }
 {
  bone1Name=calf_r
  bone2Name=thigh_r
  swing1LimitDegrees=45
  swing1Motion=Free
  swing2LimitDegrees=45
  swing2Motion=Free
  twistLimitDegrees=45
  twistMotion=Free
 }
 {
  bone1Name=dyn_skirt_sd_r_01
  bone2Name=base_skirt_sd_r
  swing1LimitDegrees=5
  swing1Motion=Locked
  swing2LimitDegrees=10
  twistLimitDegrees=45
  twistMotion=Locked
 }
 {
  bone1Name=dyn_skirt_sd_r_02
  bone2Name=dyn_skirt_sd_r_01
  swing1LimitDegrees=5
  swing1Motion=Locked
  swing2LimitDegrees=25
  twistLimitDegrees=45
  twistMotion=Locked
 }
 {
  bone1Name=dyn_skirt_sd_r_03
  bone2Name=dyn_skirt_sd_r_02
  swing1LimitDegrees=5
  swing1Motion=Locked
  swing2LimitDegrees=25
  twistLimitDegrees=5
  twistMotion=Locked
 }
 {
  bone1Name=dyn_skirt_sd_r_04
  bone2Name=dyn_skirt_sd_r_03
  swing1LimitDegrees=5
  swing1Motion=Locked
  swing2LimitDegrees=25
  twistLimitDegrees=5
  twistMotion=Locked
 }
 {
  bone1Name=dyn_skirt_sd_r_05
  bone2Name=dyn_skirt_sd_r_04
  swing1LimitDegrees=5
  swing1Motion=Locked
  swing2LimitDegrees=25
  twistLimitDegrees=5
  twistMotion=Locked
 }
 {
  bone1Name=dyn_skirt_sd_r_06
  bone2Name=dyn_skirt_sd_r_05
  swing1LimitDegrees=5
  swing1Motion=Locked
  swing2LimitDegrees=25
  twistLimitDegrees=5
  twistMotion=Locked
 }
 {
  bone1Name=base_skirt_sd_r
  bone2Name=pelvis
  swing1LimitDegrees=45
  swing1Motion=Free
  swing2LimitDegrees=45
  swing2Motion=Free
  twistLimitDegrees=45
  twistMotion=Free
 }
 {bone1Name=dyn_skirt_fr_r_08, bone2Name=dyn_skirt_fr_r_07, swing1LimitDegrees=20, swing2LimitDegrees=10, twistLimitDegrees=5}
 {bone1Name=dyn_skirt_fr_r_06, bone2Name=dyn_skirt_fr_r_05, swing1LimitDegrees=20, swing2LimitDegrees=10, twistLimitDegrees=5}
 {bone1Name=dyn_skirt_fr_r_07, bone2Name=dyn_skirt_fr_r_06, swing1LimitDegrees=20, swing2LimitDegrees=10, twistLimitDegrees=5}
 {bone1Name=dyn_skirt_fr_r_05, bone2Name=dyn_skirt_fr_r_04, swing1LimitDegrees=20, swing2LimitDegrees=10, twistLimitDegrees=5}
 {bone1Name=dyn_skirt_fr_r_04, bone2Name=dyn_skirt_fr_r_03, swing1LimitDegrees=40, swing2LimitDegrees=10, twistLimitDegrees=5}
 {
  bone1Name=dyn_skirt_fr_r_03
  bone2Name=dyn_skirt_fr_r_02
  swing1LimitDegrees=40
  swing2LimitDegrees=10
  twistLimitDegrees=45
  twistMotion=Locked
 }
 {
  bone1Name=dyn_skirt_fr_r_02
  bone2Name=pelvis
  swing1LimitDegrees=45
  swing1Motion=Free
  swing2LimitDegrees=45
  swing2Motion=Free
  twistLimitDegrees=45
  twistMotion=Free
 }
]
