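# Post-process style AnimBlueprint for the female MajorSpeech Nix cosmetic body, adding secondary motion and pose corrections to skirt, collar, sleeves, pants and pendant.
# Cached poses feed additive bone offsets and RigidBody simulations with separate physics assets, selected by crouch and DBNO (downed) bools.
# A collar RigidBody with upward gravity follows the crouch blend.
# Chained spring-based AnimDynamics nodes drive pants, sleeves and pendant, followed by further skirt and collar offsets.
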AnimBlueprint /BRCosmetics/Characters/Player/Female/Medium/Bodies/F_MED_MajorSpeech_Nix/Meshes/F_MED_MajorSpeech_Nix_AnimBP
target_skeleton=&/F_MED_MajorSpeech_Nix_Skeleton
graphs [
 AnimGraph {nodes [
   n1 = "Transform (Modify) Bone - Bone: base_skirt_bk_c"!(
    ComponentPose=Local To Component(LocalPose=Use cached pose 'Initial Adjust'())
    bAlphaBoolEnabled:bool=True
    Translation="-6.227,-0.000,-2.729"
    Rotation="86.025,-74.018,-73.031"
   ) -> (Pose) {
    alphaInputType=Bool
    rotationMode=BMM_Additive
    rotationSpace=BCS_BoneSpace
    translationMode=BMM_Additive
    translationSpace=BCS_BoneSpace
    alphaBoolBlend {blendInTime=0.5, blendOutTime=0.1}
    boneToModify {boneName=base_skirt_bk_c}
    rotation=<rot 86.025 -74.018 -73.031>
    translation {x=-6.227, z=-2.729}
   }
   n2 = "Transform (Modify) Bone - Bone: dyn_skirt_fr_l_02"!(ComponentPose=@n1, bAlphaBoolEnabled:bool=True, Translation="0,0,0", Rotation="15.526,-12.449,-0.010") -> (Pose) {
    alphaInputType=Bool
    rotationMode=BMM_Additive
    rotationSpace=BCS_BoneSpace
    translationMode=BMM_Additive
    translationSpace=BCS_BoneSpace
    alphaBoolBlend {blendInTime=0.3, blendOutTime=0.1}
    boneToModify {boneName=dyn_skirt_fr_l_02}
    rotation=<rot 15.526 -12.449 -0.01>
   }
   n3 = "Transform (Modify) Bone - Bone: base_skirt_sd_r"!(ComponentPose=@n2, bAlphaBoolEnabled:bool=True, Translation="-7.504,0.317,-4.306", Rotation="70.669,-22.925,-28.522") -> (Pose) {
    alphaInputType=Bool
    rotationMode=BMM_Additive
    rotationSpace=BCS_BoneSpace
    translationMode=BMM_Additive
    translationSpace=BCS_BoneSpace
    alphaBoolBlend {blendInTime=0.4, blendOutTime=0.1}
    boneToModify {boneName=base_skirt_sd_r}
    rotation=<rot 70.669 -22.925 -28.522>
    translation=<-7.504 0.317 -4.306>
   }
   n4 = "Transform (Modify) Bone - Bone: dyn_skirt_fr_l_03"!(ComponentPose=@n3, bAlphaBoolEnabled:bool=True, Translation="0,0,0", Rotation="6.109,9.701,7.331") -> (Pose) {
    alphaInputType=Bool
    rotationMode=BMM_Additive
    rotationSpace=BCS_BoneSpace
    translationMode=BMM_Additive
    translationSpace=BCS_BoneSpace
    alphaBoolBlend {blendInTime=0.3, blendOutTime=0.1}
    boneToModify {boneName=dyn_skirt_fr_l_03}
    rotation=<rot 6.109 9.701 7.331>
   }
   n5 = "Transform (Modify) Bone - Bone: dyn_skirt_fr_r_02"!(ComponentPose=@n4, bAlphaBoolEnabled:bool=True, Translation="0.340,2.199,-0.174", Rotation="-11.064,-94.762,-92.389") -> (Pose) {
    alphaInputType=Bool
    rotationMode=BMM_Additive
    rotationSpace=BCS_BoneSpace
    translationMode=BMM_Additive
    translationSpace=BCS_BoneSpace
    alphaBoolBlend {blendInTime=0.5, blendOutTime=0.1}
    boneToModify {boneName=dyn_skirt_fr_r_02}
    rotation=<rot -11.064 -94.762 -92.389>
    translation=<0.34 2.199 -0.174>
   }
   n6 = "Transform (Modify) Bone - Bone: base_skirt_sd_l"!(ComponentPose=@n5, bAlphaBoolEnabled:bool=True, Translation="0,0,0", Rotation="65.248,-3.074,13.091") -> (Pose) {
    alphaInputType=Bool
    rotationMode=BMM_Additive
    rotationSpace=BCS_BoneSpace
    translationMode=BMM_Additive
    translationSpace=BCS_BoneSpace
    alphaBoolBlend {blendInTime=0.3, blendOutTime=0.1}
    boneToModify {boneName=base_skirt_sd_l}
    rotation=<rot 65.248 -3.074 13.091>
   }
   n7 = "Transform (Modify) Bone - Bone: dyn_skirt_fr_r_01_extra"!(ComponentPose=@n6, bAlphaBoolEnabled:bool=True, Translation="0,0,0", Rotation="-0.000,-0.000,-77.011") -> (Pose) {
    alphaInputType=Bool
    rotationMode=BMM_Additive
    rotationSpace=BCS_BoneSpace
    translationMode=BMM_Additive
    translationSpace=BCS_BoneSpace
    alphaBoolBlend {blendInTime=0.5, blendOutTime=0.1}
    boneToModify {boneName=dyn_skirt_fr_r_01_extra}
    rotation {roll=-77.011}
   }
   n8 = "Transform (Modify) Bone - Bone: floor"!(ComponentPose=@n7, bAlphaBoolEnabled:bool=True, Translation="0,0,0", Rotation="0,0,0") -> (Pose) {
    alphaInputType=Bool
    rotationMode=BMM_Additive
    rotationSpace=BCS_BoneSpace
    translationMode=BMM_Additive
    translationSpace=BCS_BoneSpace
    alphaBoolBlend {blendInTime=0.3, blendOutTime=0.1}
    boneToModify {boneName=floor}
   }
   n9 = "Transform (Modify) Bone - Bone: base_skirt_bk_c"(ComponentPose=@n8, bAlphaBoolEnabled=True, Translation="-4.466,-0.000,1.862", Rotation="88.773,-180.000,-180.000") {
    alphaInputType=Bool
    rotationMode=BMM_Additive
    rotationSpace=BCS_BoneSpace
    translationMode=BMM_Additive
    translationSpace=BCS_BoneSpace
    alphaBoolBlend {blendInTime=0.85, blendOutTime=0.2}
    boneToModify {boneName=base_skirt_bk_c}
    rotation=<rot 88.773 -180 -180>
    translation {x=-4.466, z=1.862}
   }
   n10 = "Transform (Modify) Bone - Bone: base_skirt_sd_l"(ComponentPose=@n9, bAlphaBoolEnabled=True, Translation="0,0,0", Rotation="39.064,-0.000,-0.000") {
    alphaInputType=Bool
    rotationMode=BMM_Additive
    rotationSpace=BCS_BoneSpace
    translationMode=BMM_Additive
    translationSpace=BCS_BoneSpace
    alphaBoolBlend {blendInTime=0.4, blendOutTime=0.2}
    boneToModify {boneName=base_skirt_sd_l}
    rotation=<rot 39.064 0 0>
   }
   n11 = "Transform (Modify) Bone - Bone: dyn_skirt_fr_l_02"(ComponentPose=@n10, bAlphaBoolEnabled=True, Translation="0,0,0", Rotation="56.806,22.841,3.506") {
    alphaInputType=Bool
    rotationMode=BMM_Additive
    rotationSpace=BCS_BoneSpace
    translationMode=BMM_Additive
    translationSpace=BCS_BoneSpace
    alphaBoolBlend {blendInTime=0.4, blendOutTime=0.2}
    boneToModify {boneName=dyn_skirt_fr_l_02}
    rotation=<rot 56.806 22.841 3.506>
   }
   n12 = "Transform (Modify) Bone - Bone: base_skirt_sd_r"(ComponentPose=@n11, bAlphaBoolEnabled=True, Translation="-5.110,-0.692,6.488", Rotation="5.863,-5.750,-9.796") {
    alphaInputType=Bool
    rotationMode=BMM_Additive
    rotationSpace=BCS_BoneSpace
    translationMode=BMM_Additive
    translationSpace=BCS_BoneSpace
    alphaBoolBlend {blendInTime=0.4, blendOutTime=0.2}
    boneToModify {boneName=base_skirt_sd_r}
    rotation=<rot 5.863 -5.75 -9.796>
    translation=<-5.1099 -0.6919 6.4875>
   }
   n13 = RigidBody!(
    ComponentPose=@n12
    Alpha
    OverrideWorldGravity:Vector="0,0,-980"
    ExternalForce
    SimSpaceSettings:Sim Space Settings Structure=(WorldAlpha=0,VelocityScaleZ=1,DampingAlpha=1,MaxLinearVelocity=10000,MaxAngularVelocity=10000,MaxLinearAcceleration=10000,MaxAngularAcceleration=10000,ExternalLinearDragV=(X=0,Y=0,Z=0),ExternalLinearVelocity=(X=0,Y=0,Z=0),ExternalAngularVelocity=(X=0,Y=0,Z=0))
   ) -> (Pose) {
    bOverrideWorldGravity=true
    overridePhysicsAsset=&/F_MED_MajorSpeech_Nix_Crouching_Physics
    simulationSpace=BaseBoneSpace
    alphaScaleBiasClamp {bInterpResult=true}
    baseBoneRef {boneName=pelvis}
    componentAppliedLinearAccClamp=<3000 3000 3000>
    overrideWorldGravity {z=-980}
   }
   n14 = Component To Local(ComponentPose=@n13)
   n15 = "Transform (Modify) Bone - Bone: lowerarm_twist_extra_l"(
    ComponentPose=Local To Component(LocalPose=Use cached pose 'Initial Adjust'())
    Alpha=1
    Translation="0,0,0"
    Rotation="0,0,-70"
    Scale="1,1,1"
   ) {
    rotationMode=BMM_Additive
    rotationSpace=BCS_BoneSpace
    translationMode=BMM_Additive
    translationSpace=BCS_BoneSpace
    alphaBoolBlend {blendInTime=0.4, blendOutTime=0.2}
    boneToModify {boneName=lowerarm_twist_extra_l}
    rotation {roll=-70}
   }
   n16 = "Transform (Modify) Bone - Bone: lowerarm_twist_extra_r"(ComponentPose=@n15, Alpha=1, Translation="0,0,0", Rotation="0,0,-75", Scale="1,1,1") {
    rotationMode=BMM_Additive
    rotationSpace=BCS_BoneSpace
    translationMode=BMM_Additive
    translationSpace=BCS_BoneSpace
    alphaBoolBlend {blendInTime=0.4, blendOutTime=0.2}
    boneToModify {boneName=lowerarm_twist_extra_r}
    rotation {roll=-75}
   }
   n17 = RigidBody(
    ComponentPose=@n16
    Alpha=0.7
    OverrideWorldGravity="0,0,-980"
    SimSpaceSettings=(WorldAlpha=0,VelocityScaleZ=1,DampingAlpha=1,MaxLinearVelocity=10000,MaxAngularVelocity=10000,MaxLinearAcceleration=10000,MaxAngularAcceleration=10000,ExternalLinearDragV=(X=0,Y=0,Z=0),ExternalLinearVelocity=(X=0,Y=0,Z=0),ExternalAngularVelocity=(X=0,Y=0,Z=0))
   ) {
    bOverrideWorldGravity=true
    overridePhysicsAsset=&/F_MED_MajorSpeech_Nix_Physics
    simulationSpace=BaseBoneSpace
    alphaScaleBiasClamp {bInterpResult=true}
    baseBoneRef {boneName=pelvis}
    componentAppliedLinearAccClamp=<3000 3000 3000>
    overrideWorldGravity {z=-980}
   }
   n18 = Component To Local(ComponentPose=@n17)
   n19 = Get bIsCrouching()
   n20 = RigidBody!(
    ComponentPose=Local To Component(LocalPose=Blend Poses by bool(BlendPose_0=@n14, BlendPose_1=@n18, bActiveValue=@n19))
    Alpha=0
    ExternalForce
    SimSpaceSettings=(WorldAlpha=0,VelocityScaleZ=1,DampingAlpha=1,MaxLinearVelocity=10000,MaxAngularVelocity=10000,MaxLinearAcceleration=10000,MaxAngularAcceleration=10000,ExternalLinearDragV=(X=0,Y=0,Z=0),ExternalLinearVelocity=(X=0,Y=0,Z=0),ExternalAngularVelocity=(X=0,Y=0,Z=0))
   ) -> (Pose) {
    bOverrideWorldGravity=true
    overridePhysicsAsset=&/F_MED_MajorSpeech_Nix_Collar_Physics
    simulationSpace=BaseBoneSpace
    alphaScaleBiasClamp {bInterpResult=true, interpSpeedIncreasing=1}
    baseBoneRef {boneName=spine_05}
    componentAppliedLinearAccClamp=<3000 3000 3000>
    overrideWorldGravity {z=980}
   }
   n21 = Component To Local(ComponentPose=@n20)
   n22 = RigidBody!(
    ComponentPose=Local To Component(LocalPose=Use cached pose 'CtrlRigCache'())
    Alpha
    OverrideWorldGravity:Vector="0,0,-980"
    ExternalForce
    SimSpaceSettings=(WorldAlpha=0,VelocityScaleZ=1,DampingAlpha=1,MaxLinearVelocity=10000,MaxAngularVelocity=10000,MaxLinearAcceleration=10000,MaxAngularAcceleration=10000,ExternalLinearDragV=(X=0,Y=0,Z=0),ExternalLinearVelocity=(X=0,Y=0,Z=0),ExternalAngularVelocity=(X=0,Y=0,Z=0))
   ) -> (Pose) {
    bOverrideWorldGravity=true
    overridePhysicsAsset=&/F_MED_MajorSpeech_Nix_PhysicsDBNO
    simulationSpace=BaseBoneSpace
    alphaScaleBiasClamp {bInterpResult=true}
    baseBoneRef {boneName=root}
    componentAppliedLinearAccClamp=<3000 3000 3000>
    overrideWorldGravity {z=-980}
   }
   n23 = "Transform (Modify) Bone - Bone: dyn_skirt_fr_r_02"!(ComponentPose=@n22, Alpha:float=1, Translation="0,0,0", Rotation="-5.908,-5.654,0.584") -> (Pose) {rotationMode=BMM_Additive, rotationSpace=BCS_BoneSpace, boneToModify {boneName=dyn_skirt_fr_r_02}, rotation=<rot -5.908 -5.654 0.584>}
   n24 = "Transform (Modify) Bone - Bone: dyn_skirt_fr_l_07"!(ComponentPose=@n23, Alpha=1, Translation="0,0,0", Rotation="-21.672,12.576,-2.633") -> (Pose) {rotationMode=BMM_Additive, rotationSpace=BCS_BoneSpace, boneToModify {boneName=dyn_skirt_fr_l_07}, rotation=<rot -21.672 12.576 -2.633>}
   n25 = "Transform (Modify) Bone - Bone: dyn_skirt_fr_r_02"!(ComponentPose=@n24, bAlphaBoolEnabled:bool=True, Translation="0,0,0", Rotation="2.180,-17.798,-5.467") -> (Pose) {
    alphaInputType=Bool
    rotationMode=BMM_Additive
    rotationSpace=BCS_BoneSpace
    alphaBoolBlend {blendInTime=0.2, blendOutTime=0.2}
    boneToModify {boneName=dyn_skirt_fr_r_02}
    rotation=<rot 2.18 -17.798 -5.467>
   }
   n26 = Component To Local(ComponentPose=@n25)
   n27 = Get bIsDBNO()
   n28 = Anim Dynamics - Bone: dyn_ClothPants_out_01_l!(
    ComponentPose=Local To Component(LocalPose=Blend Poses by bool(BlendPose_0=@n26, BlendPose_1=@n21, bActiveValue=@n27))
    Alpha=0.3
    GravityOverride:Vector="0,0,-980"
    ExternalForce="0,0,0"
   ) -> (Pose) {
    angularDampingOverride=1
    angularSpringConstant=20000
    bAngularSpring=true
    bOverrideAngularDamping=true
    bUseGravityOverride=true
    gravityScale=0.35
    simulationSpace=BoneRelative
    boundBone {boneName=dyn_ClothPants_out_01_l}
    componentAppliedLinearAccClamp=<3000 3000 3000>
    componentLinearAccScale=<1 1 1>
    componentLinearVelScale=<1 1 1>
    gravityOverride {z=-980}
    physicsBodyDefinitions [{
      boundBone {boneName=dyn_ClothPants_out_01_l}
      boxExtents=<30 30 30>
      constraintSetup {angularLimitsMax=<0 0 5>, angularLimitsMin=<0 -10 -5>, angularTarget=<1 0 0>}
      localJointOffset=<3 0 0>
     }]
    relativeSpaceBone {boneName=thigh_l}
   }
   n29 = Anim Dynamics - Bone: dyn_ClothPants_bk_01_l!(ComponentPose=@n28, Alpha=0.3, GravityOverride:Vector="0,0,-980", ExternalForce="0,0,0") -> (Pose) {
    angularDampingOverride=1
    angularSpringConstant=20000
    bAngularSpring=true
    bOverrideAngularDamping=true
    bUseGravityOverride=true
    gravityScale=0.35
    simulationSpace=BoneRelative
    boundBone {boneName=dyn_ClothPants_bk_01_l}
    componentAppliedLinearAccClamp=<3000 3000 3000>
    componentLinearAccScale=<1 1 1>
    componentLinearVelScale=<1 1 1>
    gravityOverride {z=-980}
    physicsBodyDefinitions [{
      boundBone {boneName=dyn_ClothPants_bk_01_l}
      boxExtents=<30 30 30>
      constraintSetup {angularLimitsMax=<0 0 5>, angularLimitsMin=<0 -10 -5>, angularTarget=<1 0 0>}
      localJointOffset=<3 0 0>
     }]
    relativeSpaceBone {boneName=thigh_l}
   }
   n30 = Anim Dynamics - Bone: dyn_ClothPants_in_01_l!(ComponentPose=@n29, Alpha=0.3, GravityOverride:Vector="0,0,-980", ExternalForce="0,0,0") -> (Pose) {
    angularDampingOverride=1
    angularSpringConstant=20000
    bAngularSpring=true
    bOverrideAngularDamping=true
    bUseGravityOverride=true
    gravityScale=0.35
    simulationSpace=BoneRelative
    boundBone {boneName=dyn_ClothPants_in_01_l}
    componentAppliedLinearAccClamp=<3000 3000 3000>
    componentLinearAccScale=<1 1 1>
    componentLinearVelScale=<1 1 1>
    gravityOverride {z=-980}
    physicsBodyDefinitions [{
      boundBone {boneName=dyn_ClothPants_in_01_l}
      boxExtents=<30 30 30>
      constraintSetup {angularLimitsMax=<0 0 5>, angularLimitsMin=<0 -10 -5>, angularTarget=<1 0 0>}
      localJointOffset=<3 0 0>
     }]
    relativeSpaceBone {boneName=thigh_l}
   }
   n31 = Anim Dynamics - Bone: dyn_ClothPants_fr_01_l!(ComponentPose=@n30, Alpha=1, GravityOverride:Vector="0,0,-980", ExternalForce="0,0,0") -> (Pose) {
    angularDampingOverride=1
    angularSpringConstant=20000
    bAngularSpring=true
    bOverrideAngularDamping=true
    bUseGravityOverride=true
    gravityScale=0.35
    simulationSpace=BoneRelative
    boundBone {boneName=dyn_ClothPants_fr_01_l}
    componentAppliedLinearAccClamp=<3000 3000 3000>
    componentLinearAccScale=<1 1 1>
    componentLinearVelScale=<1 1 1>
    gravityOverride {z=-980}
    physicsBodyDefinitions [{
      boundBone {boneName=dyn_ClothPants_fr_01_l}
      boxExtents=<30 30 30>
      constraintSetup {angularLimitsMax=<0 0 5>, angularLimitsMin=<0 -10 -5>, angularTarget=<1 0 0>}
      localJointOffset=<3 0 0>
     }]
    relativeSpaceBone {boneName=thigh_l}
   }
   n32 = Anim Dynamics - Bone: dyn_ClothPants_out_01_r!(ComponentPose=@n31, Alpha=1, GravityOverride:Vector="0,0,-980", ExternalForce="0,0,0") -> (Pose) {
    angularDampingOverride=1
    angularSpringConstant=20000
    bAngularSpring=true
    bOverrideAngularDamping=true
    bUseGravityOverride=true
    gravityScale=0.35
    simulationSpace=BoneRelative
    boundBone {boneName=dyn_ClothPants_out_01_r}
    componentAppliedLinearAccClamp=<3000 3000 3000>
    componentLinearAccScale=<1 1 1>
    componentLinearVelScale=<1 1 1>
    gravityOverride {z=-980}
    physicsBodyDefinitions [{
      boundBone {boneName=dyn_ClothPants_out_01_r}
      boxExtents=<30 30 30>
      constraintSetup {angularLimitsMax=<0 0 5>, angularLimitsMin=<0 -10 -5>, angularTarget=<1 0 0>}
      localJointOffset=<3 0 0>
     }]
    relativeSpaceBone {boneName=thigh_r}
   }
   n33 = Anim Dynamics - Bone: dyn_ClothPants_bk_01_r!(ComponentPose=@n32, Alpha=1, GravityOverride:Vector="0,0,-980", ExternalForce="0,0,0") -> (Pose) {
    angularDampingOverride=1
    angularSpringConstant=20000
    bAngularSpring=true
    bOverrideAngularDamping=true
    bUseGravityOverride=true
    gravityScale=0.35
    simulationSpace=BoneRelative
    boundBone {boneName=dyn_ClothPants_bk_01_r}
    componentAppliedLinearAccClamp=<3000 3000 3000>
    componentLinearAccScale=<1 1 1>
    componentLinearVelScale=<1 1 1>
    gravityOverride {z=-980}
    physicsBodyDefinitions [{
      boundBone {boneName=dyn_ClothPants_bk_01_r}
      boxExtents=<30 30 30>
      constraintSetup {angularLimitsMax=<0 0 5>, angularLimitsMin=<0 -10 -5>, angularTarget=<1 0 0>}
      localJointOffset=<3 0 0>
     }]
    relativeSpaceBone {boneName=thigh_r}
   }
   n34 = Anim Dynamics - Bone: dyn_ClothPants_in_01_r!(ComponentPose=@n33, Alpha=1, GravityOverride:Vector="0,0,-980", ExternalForce="0,0,0") -> (Pose) {
    angularDampingOverride=1
    angularSpringConstant=20000
    bAngularSpring=true
    bOverrideAngularDamping=true
    bUseGravityOverride=true
    gravityScale=0.35
    simulationSpace=BoneRelative
    boundBone {boneName=dyn_ClothPants_in_01_r}
    componentAppliedLinearAccClamp=<3000 3000 3000>
    componentLinearAccScale=<1 1 1>
    componentLinearVelScale=<1 1 1>
    gravityOverride {z=-980}
    physicsBodyDefinitions [{
      boundBone {boneName=dyn_ClothPants_in_01_r}
      boxExtents=<30 30 30>
      constraintSetup {angularLimitsMax=<0 0 5>, angularLimitsMin=<0 -10 -5>, angularTarget=<1 0 0>}
      localJointOffset=<3 0 0>
     }]
    relativeSpaceBone {boneName=thigh_r}
   }
   n35 = Anim Dynamics - Bone: dyn_ClothPants_fr_01_r!(ComponentPose=@n34, Alpha=1, GravityOverride:Vector="0,0,-980", ExternalForce="0,0,0") -> (Pose) {
    angularDampingOverride=1
    angularSpringConstant=20000
    bAngularSpring=true
    bOverrideAngularDamping=true
    bUseGravityOverride=true
    gravityScale=0.35
    simulationSpace=BoneRelative
    boundBone {boneName=dyn_ClothPants_fr_01_r}
    componentAppliedLinearAccClamp=<3000 3000 3000>
    componentLinearAccScale=<1 1 1>
    componentLinearVelScale=<1 1 1>
    gravityOverride {z=-980}
    physicsBodyDefinitions [{
      boundBone {boneName=dyn_ClothPants_fr_01_r}
      boxExtents=<30 30 30>
      constraintSetup {angularLimitsMax=<0 0 5>, angularLimitsMin=<0 -10 -5>, angularTarget=<1 0 0>}
      localJointOffset=<3 0 0>
     }]
    relativeSpaceBone {boneName=thigh_r}
   }
   n36 = Anim Dynamics - Bone: dyn_sleeves_twist_b_01_l!(
    ComponentPose=@n35
    Alpha=1
    GravityOverride:Vector="0,0,-980"
    SimSpaceSettings:Anim Phys Sim Space Settings Structure=(SimSpaceAngularAlpha=0,MaxAngularVelocity=10000,MaxAngularAcceleration=10000,ExternalAngularVelocity=(X=0,Y=0,Z=0))
    ExternalForce="0,0,0"
   ) -> (Pose) {
    angularDampingOverride=0.999
    angularSpringConstant=3000
    bAngularSpring=true
    bOverrideAngularDamping=true
    bUseGravityOverride=true
    bUsePlanarLimit=false
    gravityScale=0.35
    linearDampingOverride=1
    linearSpringConstant=50
    simulationSpace=RootRelative
    alphaBoolBlend {blendOption=Linear}
    boundBone {boneName=dyn_sleeves_twist_b_01_l}
    componentAppliedLinearAccClamp=<5 5 5>
    componentLinearAccScale=<0.05 0.05 0.05>
    componentLinearVelScale=<0.05 0.05 0.05>
    gravityOverride {z=-980}
    physicsBodyDefinitions [{
      boundBone {boneName=dyn_sleeves_twist_b_01_l}
      boxExtents=<18 18 18>
      constraintSetup {angularLimitsMax=<1 2 2>, angularLimitsMin=<-1 -2 2>, angularTarget=<1 0 0>}
      localJointOffset=<1 0 0>
     }]
    relativeSpaceBone {boneName=root}
    simSpaceSettings {simSpaceAngularAlpha=1}
   }
   n37 = Anim Dynamics - Bone: dyn_sleeves_twist_b_02_l!(
    ComponentPose=@n36
    Alpha=1
    GravityOverride:Vector="0,0,-980"
    SimSpaceSettings:Anim Phys Sim Space Settings Structure=(SimSpaceAngularAlpha=1,MaxAngularVelocity=10000,MaxAngularAcceleration=10000,ExternalAngularVelocity=(X=0,Y=0,Z=0))
    ExternalForce="0,0,0"
   ) -> (Pose) {
    angularDampingOverride=0.999
    angularSpringConstant=3000
    bAngularSpring=true
    bOverrideAngularDamping=true
    bUseGravityOverride=true
    bUsePlanarLimit=false
    gravityScale=0.35
    linearDampingOverride=1
    linearSpringConstant=100
    simulationSpace=RootRelative
    alphaBoolBlend {blendOption=Linear}
    boundBone {boneName=dyn_sleeves_twist_b_02_l}
    componentAppliedLinearAccClamp=<5 5 5>
    componentLinearAccScale=<0.05 0.05 0.05>
    componentLinearVelScale=<0.05 0.05 0.05>
    gravityOverride {z=-980}
    physicsBodyDefinitions [{
      boundBone {boneName=dyn_sleeves_twist_b_02_l}
      boxExtents=<18 18 18>
      constraintSetup {angularLimitsMax=<0 20 10>, angularLimitsMin=<0 -20 -10>, angularTarget=<1 0 0>}
      localJointOffset=<4 0 0>
     }]
    relativeSpaceBone {boneName=root}
    simSpaceSettings {simSpaceAngularAlpha=1}
   }
   n38 = Anim Dynamics - Bone: dyn_sleeves_twist_b_03_l!(
    ComponentPose=@n37
    Alpha=1
    GravityOverride:Vector="0,0,-980"
    SimSpaceSettings:Anim Phys Sim Space Settings Structure=(SimSpaceAngularAlpha=1,MaxAngularVelocity=10000,MaxAngularAcceleration=10000,ExternalAngularVelocity=(X=0,Y=0,Z=0))
    ExternalForce="0,0,0"
   ) -> (Pose) {
    angularDampingOverride=0.999
    angularSpringConstant=3000
    bAngularSpring=true
    bOverrideAngularDamping=true
    bUseGravityOverride=true
    bUsePlanarLimit=false
    gravityScale=0.35
    linearDampingOverride=1
    linearSpringConstant=100
    simulationSpace=RootRelative
    alphaBoolBlend {blendOption=Linear}
    boundBone {boneName=dyn_sleeves_twist_b_03_l}
    componentAppliedLinearAccClamp=<5 5 5>
    componentLinearAccScale=<0.05 0.05 0.05>
    componentLinearVelScale=<0.05 0.05 0.05>
    gravityOverride {z=-980}
    physicsBodyDefinitions [{
      boundBone {boneName=dyn_sleeves_twist_b_03_l}
      boxExtents=<18 18 18>
      constraintSetup {angularLimitsMax=<0 20 10>, angularLimitsMin=<0 -20 -10>, angularTarget=<1 0 0>}
      localJointOffset=<4 0 0>
     }]
    relativeSpaceBone {boneName=root}
    simSpaceSettings {simSpaceAngularAlpha=1}
   }
   n39 = Anim Dynamics - Bone: dyn_sleeves_twist_a_01_l!(
    ComponentPose=@n38
    Alpha=1
    GravityOverride:Vector="0,0,-980"
    SimSpaceSettings:Anim Phys Sim Space Settings Structure=(SimSpaceAngularAlpha=1,MaxAngularVelocity=10000,MaxAngularAcceleration=10000,ExternalAngularVelocity=(X=0,Y=0,Z=0))
    ExternalForce="0,0,0"
   ) -> (Pose) {
    angularDampingOverride=0.999
    angularSpringConstant=3000
    bAngularSpring=true
    bOverrideAngularDamping=true
    bUseGravityOverride=true
    bUsePlanarLimit=false
    gravityScale=0.35
    linearDampingOverride=1
    linearSpringConstant=50
    simulationSpace=RootRelative
    alphaBoolBlend {blendOption=Linear}
    boundBone {boneName=dyn_sleeves_twist_a_01_l}
    componentAppliedLinearAccClamp=<5 5 5>
    componentLinearAccScale=<0.05 0.05 0.05>
    componentLinearVelScale=<0.05 0.05 0.05>
    gravityOverride {z=-980}
    physicsBodyDefinitions [{
      boundBone {boneName=dyn_sleeves_twist_a_01_l}
      boxExtents=<18 18 18>
      constraintSetup {angularLimitsMax=<1 2 2>, angularLimitsMin=<-1 -2 2>, angularTarget=<1 0 0>}
      localJointOffset=<1 0 0>
     }]
    relativeSpaceBone {boneName=root}
    simSpaceSettings {simSpaceAngularAlpha=1}
   }
   n40 = Anim Dynamics - Bone: dyn_sleeves_twist_a_02_l!(
    ComponentPose=@n39
    Alpha=1
    GravityOverride:Vector="0,0,-980"
    SimSpaceSettings:Anim Phys Sim Space Settings Structure=(SimSpaceAngularAlpha=1,MaxAngularVelocity=10000,MaxAngularAcceleration=10000,ExternalAngularVelocity=(X=0,Y=0,Z=0))
    ExternalForce="0,0,0"
   ) -> (Pose) {
    angularDampingOverride=0.999
    angularSpringConstant=3000
    bAngularSpring=true
    bOverrideAngularDamping=true
    bUseGravityOverride=true
    bUsePlanarLimit=false
    gravityScale=0.35
    linearDampingOverride=1
    linearSpringConstant=100
    simulationSpace=RootRelative
    alphaBoolBlend {blendOption=Linear}
    boundBone {boneName=dyn_sleeves_twist_a_02_l}
    componentAppliedLinearAccClamp=<5 5 5>
    componentLinearAccScale=<0.05 0.05 0.05>
    componentLinearVelScale=<0.05 0.05 0.05>
    gravityOverride {z=-980}
    physicsBodyDefinitions [{
      boundBone {boneName=dyn_sleeves_twist_a_02_l}
      boxExtents=<18 18 18>
      constraintSetup {angularLimitsMax=<10 20 20>, angularLimitsMin=<-10 -20 -20>, angularTarget=<1 0 0>}
      localJointOffset=<4 0 0>
     }]
    relativeSpaceBone {boneName=root}
    simSpaceSettings {simSpaceAngularAlpha=1}
   }
   n41 = Anim Dynamics - Bone: dyn_sleeves_twist_b_01_r!(
    ComponentPose=@n40
    Alpha=1
    GravityOverride:Vector="0,0,-980"
    SimSpaceSettings:Anim Phys Sim Space Settings Structure=(SimSpaceAngularAlpha=1,MaxAngularVelocity=10000,MaxAngularAcceleration=10000,ExternalAngularVelocity=(X=0,Y=0,Z=0))
    ExternalForce="0,0,0"
   ) -> (Pose) {
    angularDampingOverride=0.999
    angularSpringConstant=3000
    bAngularSpring=true
    bOverrideAngularDamping=true
    bUseGravityOverride=true
    bUsePlanarLimit=false
    gravityScale=0.35
    linearDampingOverride=1
    linearSpringConstant=50
    simulationSpace=RootRelative
    alphaBoolBlend {blendOption=Linear}
    boundBone {boneName=dyn_sleeves_twist_b_01_r}
    componentAppliedLinearAccClamp=<5 5 5>
    componentLinearAccScale=<0.05 0.05 0.05>
    componentLinearVelScale=<0.05 0.05 0.05>
    gravityOverride {z=-980}
    physicsBodyDefinitions [{
      boundBone {boneName=dyn_sleeves_twist_b_01_r}
      boxExtents=<18 18 18>
      constraintSetup {angularLimitsMax=<1 2 2>, angularLimitsMin=<-1 -2 -2>, angularTarget=<1 0 0>}
      localJointOffset=<-1 0 0>
     }]
    relativeSpaceBone {boneName=root}
    simSpaceSettings {simSpaceAngularAlpha=1}
   }
   n42 = Anim Dynamics - Bone: dyn_sleeves_twist_b_02_r!(
    ComponentPose=@n41
    Alpha=1
    GravityOverride:Vector="0,0,-980"
    SimSpaceSettings:Anim Phys Sim Space Settings Structure=(SimSpaceAngularAlpha=1,MaxAngularVelocity=10000,MaxAngularAcceleration=10000,ExternalAngularVelocity=(X=0,Y=0,Z=0))
    ExternalForce="0,0,0"
   ) -> (Pose) {
    angularDampingOverride=0.999
    angularSpringConstant=3000
    bAngularSpring=true
    bOverrideAngularDamping=true
    bUseGravityOverride=true
    bUsePlanarLimit=false
    gravityScale=0.35
    linearDampingOverride=1
    linearSpringConstant=100
    simulationSpace=RootRelative
    alphaBoolBlend {blendOption=Linear}
    boundBone {boneName=dyn_sleeves_twist_b_02_r}
    componentAppliedLinearAccClamp=<5 5 5>
    componentLinearAccScale=<0.05 0.05 0.05>
    componentLinearVelScale=<0.05 0.05 0.05>
    gravityOverride {z=-980}
    physicsBodyDefinitions [{
      boundBone {boneName=dyn_sleeves_twist_b_02_r}
      boxExtents=<18 18 18>
      constraintSetup {angularLimitsMax=<10 20 20>, angularLimitsMin=<-10 -20 -20>, angularTarget=<1 0 0>}
      localJointOffset=<-1 0 0>
     }]
    relativeSpaceBone {boneName=root}
    simSpaceSettings {simSpaceAngularAlpha=1}
   }
   n43 = Anim Dynamics - Bone: dyn_sleeves_twist_b_03_r!(
    ComponentPose=@n42
    Alpha=1
    GravityOverride:Vector="0,0,-980"
    SimSpaceSettings:Anim Phys Sim Space Settings Structure=(SimSpaceAngularAlpha=1,MaxAngularVelocity=10000,MaxAngularAcceleration=10000,ExternalAngularVelocity=(X=0,Y=0,Z=0))
    ExternalForce="0,0,0"
   ) -> (Pose) {
    angularDampingOverride=0.999
    angularSpringConstant=3000
    bAngularSpring=true
    bOverrideAngularDamping=true
    bUseGravityOverride=true
    bUsePlanarLimit=false
    gravityScale=0.35
    linearDampingOverride=1
    linearSpringConstant=100
    simulationSpace=RootRelative
    alphaBoolBlend {blendOption=Linear}
    boundBone {boneName=dyn_sleeves_twist_b_03_r}
    componentAppliedLinearAccClamp=<5 5 5>
    componentLinearAccScale=<0.05 0.05 0.05>
    componentLinearVelScale=<0.05 0.05 0.05>
    gravityOverride {z=-980}
    physicsBodyDefinitions [{
      boundBone {boneName=dyn_sleeves_twist_b_03_r}
      boxExtents=<18 18 18>
      constraintSetup {angularLimitsMax=<10 20 20>, angularLimitsMin=<-10 -20 -20>, angularTarget=<1 0 0>}
      localJointOffset=<-1 0 0>
     }]
    relativeSpaceBone {boneName=root}
    simSpaceSettings {simSpaceAngularAlpha=1}
   }
   n44 = Anim Dynamics - Bone: dyn_sleeves_twist_a_01_r!(
    ComponentPose=@n43
    Alpha=1
    GravityOverride:Vector="0,0,-980"
    SimSpaceSettings:Anim Phys Sim Space Settings Structure=(SimSpaceAngularAlpha=1,MaxAngularVelocity=10000,MaxAngularAcceleration=10000,ExternalAngularVelocity=(X=0,Y=0,Z=0))
    ExternalForce="0,0,0"
   ) -> (Pose) {
    angularDampingOverride=0.999
    angularSpringConstant=3000
    bAngularSpring=true
    bOverrideAngularDamping=true
    bUseGravityOverride=true
    bUsePlanarLimit=false
    gravityScale=0.35
    linearDampingOverride=1
    linearSpringConstant=50
    simulationSpace=RootRelative
    alphaBoolBlend {blendOption=Linear}
    boundBone {boneName=dyn_sleeves_twist_a_01_r}
    componentAppliedLinearAccClamp=<5 5 5>
    componentLinearAccScale=<0.05 0.05 0.05>
    componentLinearVelScale=<0.05 0.05 0.05>
    gravityOverride {z=-980}
    physicsBodyDefinitions [{
      boundBone {boneName=dyn_sleeves_twist_a_01_r}
      boxExtents=<18 18 18>
      constraintSetup {angularLimitsMax=<1 2 2>, angularLimitsMin=<-1 -2 -2>, angularTarget=<1 0 0>}
      localJointOffset=<-1 0 0>
     }]
    relativeSpaceBone {boneName=root}
    simSpaceSettings {simSpaceAngularAlpha=1}
   }
   n45 = Anim Dynamics - Bone: dyn_sleeves_twist_a_02_r!(
    ComponentPose=@n44
    Alpha=1
    GravityOverride:Vector="0,0,-980"
    SimSpaceSettings:Anim Phys Sim Space Settings Structure=(SimSpaceAngularAlpha=1,MaxAngularVelocity=10000,MaxAngularAcceleration=10000,ExternalAngularVelocity=(X=0,Y=0,Z=0))
    ExternalForce="0,0,0"
   ) -> (Pose) {
    angularDampingOverride=0.999
    angularSpringConstant=3000
    bAngularSpring=true
    bOverrideAngularDamping=true
    bUseGravityOverride=true
    bUsePlanarLimit=false
    gravityScale=0.35
    linearDampingOverride=1
    linearSpringConstant=100
    simulationSpace=RootRelative
    alphaBoolBlend {blendOption=Linear}
    boundBone {boneName=dyn_sleeves_twist_a_02_r}
    componentAppliedLinearAccClamp=<5 5 5>
    componentLinearAccScale=<0.05 0.05 0.05>
    componentLinearVelScale=<0.05 0.05 0.05>
    gravityOverride {z=-980}
    physicsBodyDefinitions [{
      boundBone {boneName=dyn_sleeves_twist_a_02_r}
      boxExtents=<18 18 18>
      constraintSetup {angularLimitsMax=<10 20 20>, angularLimitsMin=<-10 -20 -20>, angularTarget=<1 0 0>}
      localJointOffset=<-1 0 0>
     }]
    relativeSpaceBone {boneName=root}
    simSpaceSettings {simSpaceAngularAlpha=1}
   }
   n46 = Anim Dynamics - Bone: dyn_spine_pendant_03(ComponentPose=@n45, Alpha=1, ExternalForce="0,0,0") {
    angularDampingOverride=1
    angularSpringConstant=20000
    bAngularSpring=true
    bOverrideAngularDamping=true
    bUseGravityOverride=true
    gravityScale=0.35
    simulationSpace=RootRelative
    boundBone {boneName=dyn_spine_pendant_03}
    componentLinearAccScale=<1 1 1>
    componentLinearVelScale=<1 1 1>
    gravityOverride {z=-980}
    physicsBodyDefinitions [{
      boundBone {boneName=dyn_spine_pendant_03}
      boxExtents=<20 30 30>
      constraintSetup {angularLimitsMax=<5 0 5>, angularLimitsMin=<-5 -15 -5>, angularTarget=<1 0 0>}
      localJointOffset=<4 0 0>
     }]
   }
   n47 = Anim Dynamics - Bone: dyn_spine_pendant_02(ComponentPose=@n46, Alpha=1, ExternalForce="0,0,0") {
    angularDampingOverride=1
    angularSpringConstant=20000
    bAngularSpring=true
    bOverrideAngularDamping=true
    bUseGravityOverride=true
    gravityScale=0.35
    simulationSpace=RootRelative
    boundBone {boneName=dyn_spine_pendant_02}
    componentLinearAccScale=<1 1 1>
    componentLinearVelScale=<1 1 1>
    gravityOverride {z=-980}
    physicsBodyDefinitions [{
      boundBone {boneName=dyn_spine_pendant_02}
      boxExtents=<20 30 30>
      constraintSetup {angularLimitsMax=<5 0 5>, angularLimitsMin=<-5 -15 -5>, angularTarget=<1 0 0>}
      localJointOffset=<4 0 0>
     }]
   }
   n48 = "Transform (Modify) Bone - Bone: dyn_skirt_bk_02_c"!(ComponentPose=@n47, Alpha=1, Translation="0.568,-0.299,1.520", Rotation="-19.949,2.693,-3.831") -> (Pose) {
    rotationMode=BMM_Additive
    rotationSpace=BCS_BoneSpace
    translationMode=BMM_Additive
    translationSpace=BCS_BoneSpace
    alphaScaleBiasClamp {bInterpResult=true}
    boneToModify {boneName=dyn_skirt_bk_02_c}
    rotation=<rot -19.949 2.693 -3.831>
    translation=<0.568 -0.299 1.52>
   }
   n49 = "Transform (Modify) Bone - Bone: dyn_skirt_sd_r_02"!(ComponentPose=@n48, Alpha=1, Translation="0.049,4.741,-0.434", Rotation="-18.359,-2.643,0.775") -> (Pose) {
    rotationMode=BMM_Additive
    rotationSpace=BCS_BoneSpace
    translationMode=BMM_Additive
    translationSpace=BCS_BoneSpace
    alphaScaleBiasClamp {bInterpResult=true}
    boneToModify {boneName=dyn_skirt_sd_r_02}
    rotation=<rot -18.359 -2.643 0.775>
    translation=<0.049 4.741 -0.434>
   }
   n50 = "Transform (Modify) Bone - Bone: dyn_skirt_sd_l_03"!(ComponentPose=@n49, Alpha=1, Translation="0,0,0", Rotation="-16.801,-0.000,0.000") -> (Pose) {
    rotationMode=BMM_Additive
    rotationSpace=BCS_BoneSpace
    translationMode=BMM_Additive
    translationSpace=BCS_BoneSpace
    alphaScaleBiasClamp {bInterpResult=true}
    boneToModify {boneName=dyn_skirt_sd_l_03}
    rotation {pitch=-16.801}
   }
   n51 = "Transform (Modify) Bone - Bone: base_collar_m"(ComponentPose=@n50, Alpha=1, Translation="0,0,0", Rotation="-0.006,-20.510,0.008", Scale="1,1,1") {
    rotationMode=BMM_Additive
    rotationSpace=BCS_BoneSpace
    scaleMode=BMM_Additive
    scaleSpace=BCS_BoneSpace
    translationMode=BMM_Additive
    translationSpace=BCS_BoneSpace
    boneToModify {boneName=base_collar_m}
    rotation=<rot -0.006 -20.51 0.008>
   }
   n52 = "Transform (Modify) Bone - Bone: base_skirt_bk_c"!(ComponentPose=@n51, Alpha:float=1, Translation="0,0,0", Rotation="-9.011,-1.599,11.619", Scale:Vector="1,1,1") -> (Pose) {
    rotationMode=BMM_Additive
    rotationSpace=BCS_BoneSpace
    scaleMode=BMM_Additive
    scaleSpace=BCS_BoneSpace
    translationMode=BMM_Additive
    translationSpace=BCS_BoneSpace
    boneToModify {boneName=base_skirt_bk_c}
    rotation=<rot -9.011 -1.599 11.619>
   }
   n53 = "Transform (Modify) Bone - Bone: base_skirt_sd_r"!(ComponentPose=@n52, Alpha:float=1, Translation="0,0,0", Rotation="-16.810,-15.231,6.716", Scale:Vector="1,1,1") -> (Pose) {
    rotationMode=BMM_Additive
    rotationSpace=BCS_BoneSpace
    scaleMode=BMM_Additive
    scaleSpace=BCS_BoneSpace
    translationMode=BMM_Additive
    translationSpace=BCS_BoneSpace
    boneToModify {boneName=base_skirt_sd_r}
    rotation=<rot -16.81 -15.231 6.716>
   }
   n54 = "Transform (Modify) Bone - Bone: dyn_skirt_bk_02_c"!(
    ComponentPose=Local To Component(LocalPose=Use cached pose 'CtrlRigCache'())
    bAlphaBoolEnabled:bool=True
    Translation="-17.019,0.774,9.236"
    Rotation="77.259,85.493,88.409"
   ) -> (Pose) {
    alphaInputType=Bool
    rotationMode=BMM_Additive
    rotationSpace=BCS_BoneSpace
    translationMode=BMM_Additive
    translationSpace=BCS_BoneSpace
    alphaBoolBlend {blendInTime=0.5, blendOutTime=0.1}
    boneToModify {boneName=dyn_skirt_bk_02_c}
    rotation=<rot 77.259 85.493 88.409>
    translation=<-17.019 0.774 9.236>
   }
   n55 = "Transform (Modify) Bone - Bone: dyn_skirt_fr_l_03"(ComponentPose=@n54, bAlphaBoolEnabled=True, Translation="0,0,0", Rotation="6.711,-32.912,35.665") {
    alphaInputType=Bool
    rotationMode=BMM_Additive
    rotationSpace=BCS_BoneSpace
    translationMode=BMM_Additive
    translationSpace=BCS_BoneSpace
    alphaBoolBlend {blendInTime=0.3, blendOutTime=0.1}
    boneToModify {boneName=dyn_skirt_fr_l_03}
    rotation=<rot 6.711 -32.912 35.665>
   }
   n56 = "Transform (Modify) Bone - Bone: base_skirt_sd_r"!(ComponentPose=@n55, bAlphaBoolEnabled:bool=True, Translation="0,0,0", Rotation="49.960,-64.127,-51.314") -> (Pose) {
    alphaInputType=Bool
    rotationMode=BMM_Additive
    rotationSpace=BCS_BoneSpace
    translationMode=BMM_Additive
    translationSpace=BCS_BoneSpace
    alphaBoolBlend {blendInTime=0.4, blendOutTime=0.1}
    boneToModify {boneName=base_skirt_sd_r}
    rotation=<rot 49.96 -64.127 -51.314>
   }
   n57 = "Transform (Modify) Bone - Bone: dyn_skirt_fr_r_02"(ComponentPose=@n56, bAlphaBoolEnabled=True, Translation="-11.931,-0.554,1.622", Rotation="-1.925,-63.905,3.504") {
    alphaInputType=Bool
    rotationMode=BMM_Additive
    rotationSpace=BCS_BoneSpace
    translationMode=BMM_Additive
    translationSpace=BCS_BoneSpace
    alphaBoolBlend {blendInTime=0.5, blendOutTime=0.1}
    boneToModify {boneName=dyn_skirt_fr_r_02}
    rotation=<rot -1.925 -63.905 3.504>
    translation=<-11.931 -0.554 1.622>
   }
   n58 = "Transform (Modify) Bone - Bone: base_skirt_sd_l"(ComponentPose=@n57, bAlphaBoolEnabled=True, Translation="0,0,0", Rotation="49.134,-32.697,42.902") {
    alphaInputType=Bool
    rotationMode=BMM_Additive
    rotationSpace=BCS_BoneSpace
    translationMode=BMM_Additive
    translationSpace=BCS_BoneSpace
    alphaBoolBlend {blendInTime=0.3, blendOutTime=0.1}
    boneToModify {boneName=base_skirt_sd_l}
    rotation=<rot 49.134 -32.697 42.902>
   }
   n59 = "Transform (Modify) Bone - Bone: dyn_skirt_bk_02_c"(ComponentPose=@n58, bAlphaBoolEnabled=True, Translation="0.088,-1.298,8.913", Rotation="53.260,0.000,0.000") {
    alphaInputType=Bool
    rotationMode=BMM_Additive
    rotationSpace=BCS_BoneSpace
    translationMode=BMM_Additive
    translationSpace=BCS_BoneSpace
    alphaBoolBlend {blendInTime=0.4, blendOutTime=0.2}
    boneToModify {boneName=dyn_skirt_bk_02_c}
    rotation=<rot 53.26 0 0>
    translation=<0.0878 -1.298 8.9132>
   }
   n60 = "Transform (Modify) Bone - Bone: dyn_skirt_sd_r_02"!(ComponentPose=@n59, bAlphaBoolEnabled:bool=True, Translation="-3.644,5.038,-3.349", Rotation="15.582,-22.515,-33.471") -> (Pose) {
    alphaInputType=Bool
    rotationMode=BMM_Additive
    rotationSpace=BCS_BoneSpace
    translationMode=BMM_Additive
    translationSpace=BCS_BoneSpace
    alphaBoolBlend {blendInTime=0.4, blendOutTime=0.2}
    boneToModify {boneName=dyn_skirt_sd_r_02}
    rotation=<rot 15.582 -22.515 -33.471>
    translation=<-3.644 5.038 -3.349>
   }
   n61 = "Transform (Modify) Bone - Bone: base_skirt_sd_l"(ComponentPose=@n60, bAlphaBoolEnabled=True, Translation="-0.568,-2.103,-4.584", Rotation="13.126,-17.567,13.501") {
    alphaInputType=Bool
    rotationMode=BMM_Additive
    rotationSpace=BCS_BoneSpace
    translationMode=BMM_Additive
    translationSpace=BCS_BoneSpace
    alphaBoolBlend {blendInTime=0.4, blendOutTime=0.2}
    boneToModify {boneName=base_skirt_sd_l}
    rotation=<rot 13.126 -17.567 13.501>
    translation=<-0.568 -2.1031 -4.5842>
   }
   n62 = "Transform (Modify) Bone - Bone: base_skirt_fr_l"!(ComponentPose=@n61, bAlphaBoolEnabled:bool=True, Translation="-0.345,3.448,-0.103", Rotation="6.815,6.507,11.679") -> (Pose) {
    alphaInputType=Bool
    rotationMode=BMM_Additive
    rotationSpace=BCS_BoneSpace
    translationMode=BMM_Additive
    translationSpace=BCS_BoneSpace
    alphaBoolBlend {blendInTime=0.4, blendOutTime=0.2}
    boneToModify {boneName=base_skirt_fr_l}
    rotation=<rot 6.815 6.507 11.679>
    translation=<-0.345 3.448 -0.103>
   }
   n63 = "Transform (Modify) Bone - Bone: base_skirt_sd_l"(ComponentPose=@n62, bAlphaBoolEnabled=True, Translation="0,0,0", Rotation="10,0,0") {
    alphaInputType=Bool
    rotationMode=BMM_Additive
    rotationSpace=BCS_BoneSpace
    translationMode=BMM_Additive
    translationSpace=BCS_BoneSpace
    alphaBoolBlend {blendInTime=0.4, blendOutTime=0.2}
    boneToModify {boneName=base_skirt_sd_l}
    rotation {pitch=10}
   }
   n64 = "Transform (Modify) Bone - Bone: base_skirt_bk_c"!(ComponentPose=@n63, bAlphaBoolEnabled:bool=True, Translation="0,0,0", Rotation="10,0,0") -> (Pose) {
    alphaInputType=Bool
    rotationMode=BMM_Additive
    rotationSpace=BCS_BoneSpace
    translationMode=BMM_Additive
    translationSpace=BCS_BoneSpace
    alphaBoolBlend {blendInTime=0.4, blendOutTime=0.2}
    boneToModify {boneName=base_skirt_bk_c}
    rotation {pitch=10}
   }
   n65 = "Transform (Modify) Bone - Bone: base_skirt_sd_r"(ComponentPose=@n64, bAlphaBoolEnabled=True, Translation="0,0,0", Rotation="10,0,0") {
    alphaInputType=Bool
    rotationMode=BMM_Additive
    rotationSpace=BCS_BoneSpace
    translationMode=BMM_Additive
    translationSpace=BCS_BoneSpace
    alphaBoolBlend {blendInTime=0.4, blendOutTime=0.2}
    boneToModify {boneName=base_skirt_sd_r}
    rotation {pitch=10}
   }
   n66 = "Transform (Modify) Bone - Bone: dyn_skirt_sd_l_02"!(ComponentPose=@n65, bAlphaBoolEnabled:bool=True, Translation="0,0,0", Rotation="-14.488,-16.565,0.957") -> (Pose) {
    alphaInputType=Bool
    rotationMode=BMM_Additive
    rotationSpace=BCS_BoneSpace
    translationMode=BMM_Additive
    translationSpace=BCS_BoneSpace
    alphaBoolBlend {blendInTime=1, blendOutTime=0.5}
    alphaScaleBiasClamp {bInterpResult=true}
    boneToModify {boneName=dyn_skirt_sd_l_02}
    rotation=<rot -14.488 -16.565 0.957>
   }
   n67 = "Transform (Modify) Bone - Bone: dyn_skirt_bk_02_c"(ComponentPose=@n66, bAlphaBoolEnabled=True, Translation="0,0,0", Rotation="-11.349,0.114,-0.020") {
    alphaInputType=Bool
    rotationMode=BMM_Additive
    rotationSpace=BCS_BoneSpace
    translationMode=BMM_Additive
    translationSpace=BCS_BoneSpace
    alphaBoolBlend {blendInTime=1, blendOutTime=0.5}
    alphaScaleBiasClamp {bInterpResult=true}
    boneToModify {boneName=dyn_skirt_bk_02_c}
    rotation=<rot -11.349 0.114 -0.02>
   }
   n68 = "Transform (Modify) Bone - Bone: dyn_skirt_sd_r_02"(ComponentPose=@n67, bAlphaBoolEnabled=True, Translation="0,0,0", Rotation="-4.573,-0.000,0.000") {
    alphaInputType=Bool
    rotationMode=BMM_Additive
    rotationSpace=BCS_BoneSpace
    translationMode=BMM_Additive
    translationSpace=BCS_BoneSpace
    alphaBoolBlend {blendInTime=1, blendOutTime=0.5}
    alphaScaleBiasClamp {bInterpResult=true}
    boneToModify {boneName=dyn_skirt_sd_r_02}
    rotation=<rot -4.573 0 0>
   }
   n69 = "Transform (Modify) Bone - Bone: base_skirt_bk_c"!(ComponentPose=@n68, Alpha:float=1, Translation="0,0,0", Rotation="27.273,0.000,-0.000") -> (Pose) {
    rotationMode=BMM_Additive
    rotationSpace=BCS_BoneSpace
    translationMode=BMM_Additive
    translationSpace=BCS_BoneSpace
    alphaBoolBlend {blendInTime=0.4, blendOutTime=0.2}
    alphaScaleBiasClamp {bInterpResult=true, interpSpeedIncreasing=5}
    boneToModify {boneName=base_skirt_bk_c}
    rotation=<rot 27.273 0 0>
   }
   n70 = "Transform (Modify) Bone - Bone: base_skirt_sd_r"!(ComponentPose=@n69, Alpha:float=1, Translation="0,0,0", Rotation="17.285,-0.000,0.000") -> (Pose) {
    rotationMode=BMM_Additive
    rotationSpace=BCS_BoneSpace
    translationMode=BMM_Additive
    translationSpace=BCS_BoneSpace
    alphaBoolBlend {blendInTime=0.4, blendOutTime=0.2}
    alphaScaleBiasClamp {bInterpResult=true, interpSpeedIncreasing=5}
    boneToModify {boneName=base_skirt_sd_r}
    rotation=<rot 17.285 0 0>
   }
   n71 = "Transform (Modify) Bone - Bone: base_skirt_sd_l"!(ComponentPose=@n70, Alpha:float=1, Translation="0,0,0", Rotation="17.285,-0.000,0.000") -> (Pose) {
    rotationMode=BMM_Additive
    rotationSpace=BCS_BoneSpace
    translationMode=BMM_Additive
    translationSpace=BCS_BoneSpace
    alphaBoolBlend {blendInTime=0.4, blendOutTime=0.2}
    alphaScaleBiasClamp {bInterpResult=true, interpSpeedIncreasing=5}
    boneToModify {boneName=base_skirt_sd_l}
    rotation=<rot 17.285 0 0>
   }
   n72 = "Transform (Modify) Bone - Bone: dyn_skirt_fr_r_02"!(ComponentPose=@n71, Alpha:float=1, Translation="0,0,0", Rotation="16.757,-14.640,-2.216") -> (Pose) {
    rotationMode=BMM_Additive
    rotationSpace=BCS_BoneSpace
    translationMode=BMM_Additive
    translationSpace=BCS_BoneSpace
    alphaBoolBlend {blendInTime=0.4, blendOutTime=0.2}
    alphaScaleBiasClamp {bInterpResult=true, interpSpeedIncreasing=5}
    boneToModify {boneName=dyn_skirt_fr_r_02}
    rotation=<rot 16.757 -14.64 -2.216>
   }
   Output Pose(Result=Blend Poses by bool(BlendPose_0=Component To Local(ComponentPose=@n53), BlendPose_1=Component To Local(ComponentPose=@n72), bActiveValue=Get bRigidBodyEnabled())) {layerGroup=None}
   n73 = Control Rig!(
    Source=Copy Pose From Mesh() {bUseAttachedParent=true}
    Alpha
    CrouchingBool:bool=Get bIsCrouching()
    Idle:double=Get Alpha_Idle() -> (Alpha_Idle:double)
    HeadRotateExtremCrouch:double=Get Alpha_HeadRotateUpCrouchExtrem() -> (Alpha_HeadRotateUpCrouchExtrem:double)
    HangGliding:double=Get Alpha_HangGliding() -> (Alpha_HangGliding:double)
    OnGround:double=Get Alpha_OnGround() -> (Alpha_OnGround:double)
    Emotes:bool=Get bIsPlayingEmote()
    Pistol:double=Get Alpha_DetectedPistol() -> (Alpha_DetectedPistol:double)
    Skydiving:bool=Get bIsSkyDiving()
    Targeting:bool=Get bIsTargeting()
    DBNO:bool=Get bIsDBNO()
    RBAN Enable:bool=Get bRigidBodyEnabled()
    Moving2D:bool=Get bIsMoving2D()
    Rail:bool=Get bIsOnGrindRail()
    FEIdle:double=Get Alpha_FEIdle() -> (Alpha_FEIdle:double)
   ) -> (Pose) {controlRigAssetReference {blueprintRigClass=&/F_MED_MajorSpeech_Nix_CtrlRig.F_MED_MajorSpeech_Nix_CtrlRig_C}}
   Save cached pose 'CtrlRigCache'(Pose=@n73) {cacheName=CtrlRigCache}
   n74 = "Transform (Modify) Bone - Bone: base_skirt_bk_c"!(
    ComponentPose=Local To Component(LocalPose=Use cached pose 'CtrlRigCache'())
    bAlphaBoolEnabled:bool=True
    Translation="0,0,0"
    Rotation="20,0,0"
    Scale:Vector="1,1,1"
   ) -> (Pose) {
    alphaInputType=Bool
    rotationMode=BMM_Additive
    rotationSpace=BCS_BoneSpace
    translationMode=BMM_Additive
    translationSpace=BCS_BoneSpace
    alphaBoolBlend {blendInTime=0.2, blendOutTime=0.4}
    boneToModify {boneName=base_skirt_bk_c}
    rotation {pitch=20}
   }
   n75 = "Transform (Modify) Bone - Bone: base_skirt_sd_l"!(ComponentPose=@n74, bAlphaBoolEnabled:bool=True, Translation="0,0,0", Rotation="20,0,0", Scale:Vector="1,1,1") -> (Pose) {
    alphaInputType=Bool
    rotationMode=BMM_Additive
    rotationSpace=BCS_BoneSpace
    translationMode=BMM_Additive
    translationSpace=BCS_BoneSpace
    alphaBoolBlend {blendInTime=0.2, blendOutTime=0.4}
    boneToModify {boneName=base_skirt_sd_l}
    rotation {pitch=20}
   }
   n76 = "Transform (Modify) Bone - Bone: base_skirt_sd_r"!(ComponentPose=@n75, bAlphaBoolEnabled:bool=True, Translation="0,0,0", Rotation="20,0,0", Scale:Vector="1,1,1") -> (Pose) {
    alphaInputType=Bool
    rotationMode=BMM_Additive
    rotationSpace=BCS_BoneSpace
    translationMode=BMM_Additive
    translationSpace=BCS_BoneSpace
    alphaBoolBlend {blendInTime=0.2, blendOutTime=0.4}
    boneToModify {boneName=base_skirt_sd_r}
    rotation {pitch=20}
   }
   n77 = "Transform (Modify) Bone - Bone: base_skirt_sd_r"!(ComponentPose=@n76, Alpha:float=1, Translation="0,0,0", Rotation="30,0,0", Scale="1,1,1") -> (Pose) {
    rotationMode=BMM_Additive
    rotationSpace=BCS_BoneSpace
    translationMode=BMM_Additive
    translationSpace=BCS_BoneSpace
    alphaBoolBlend {blendInTime=0.6, blendOutTime=0.4}
    alphaScaleBiasClamp {bInterpResult=true, interpSpeedDecreasing=5}
    boneToModify {boneName=base_skirt_sd_r}
    rotation {pitch=30}
   }
   n78 = "Transform (Modify) Bone - Bone: base_skirt_sd_l"!(ComponentPose=@n77, Alpha:float=1, Translation="0,0,0", Rotation="20,0,0", Scale="1,1,1") -> (Pose) {
    rotationMode=BMM_Additive
    rotationSpace=BCS_BoneSpace
    translationMode=BMM_Additive
    translationSpace=BCS_BoneSpace
    alphaBoolBlend {blendInTime=0.6, blendOutTime=0.4}
    alphaScaleBiasClamp {bInterpResult=true, interpSpeedDecreasing=5}
    boneToModify {boneName=base_skirt_sd_l}
    rotation {pitch=20}
   }
   n79 = "Transform (Modify) Bone - Bone: base_skirt_bk_c"!(ComponentPose=@n78, Alpha:float=1, Translation="0,0,0", Rotation="20,0,0", Scale="1,1,1") -> (Pose) {
    rotationMode=BMM_Additive
    rotationSpace=BCS_BoneSpace
    translationMode=BMM_Additive
    translationSpace=BCS_BoneSpace
    alphaBoolBlend {blendInTime=0.6, blendOutTime=0.4}
    alphaScaleBiasClamp {bInterpResult=true, interpSpeedDecreasing=5}
    boneToModify {boneName=base_skirt_bk_c}
    rotation {pitch=20}
   }
   n80 = "Transform (Modify) Bone - Bone: base_skirt_sd_r"!(ComponentPose=@n79, bAlphaBoolEnabled:bool=True, Translation="0,0,0", Rotation="12,0,0", Scale="1,1,1") -> (Pose) {
    alphaInputType=Bool
    rotationMode=BMM_Additive
    rotationSpace=BCS_BoneSpace
    translationMode=BMM_Additive
    translationSpace=BCS_BoneSpace
    alphaBoolBlend {blendInTime=0.3, blendOutTime=0.3}
    alphaScaleBiasClamp {bInterpResult=true, interpSpeedDecreasing=5}
    boneToModify {boneName=base_skirt_sd_r}
    rotation {pitch=12}
   }
   Save cached pose 'Initial Adjust'(Pose=Component To Local(ComponentPose=@n80)) {cacheName=Initial Adjust}
  ]}
 EventGraph {nodes [
   Event BlueprintUpdateAnimation()
   > n1 = Sequence!(execute) -> (then_0, then_1, then_2:Exec, then_3:Exec, then_4:Exec, then_5:Exec, then_6:Exec, then_7:Exec, then_8:Exec)
   > n2 = "Gravity Override State Function (Parameter Asset) (Nativized)"(
    CharacterPartAnimInstance=Self-Reference()
    GameTimeInSeconds=GetGameTimeInSeconds()
    DebugJointName=spine_05
    Parameters=&/F_MED_MajorSpeech_NixParameters
   )
   > Set Gravity Override(Gravity Override=@n2)
   > Set RBAN Settings Out(RBAN Settings Out=@n2.OutRigidBodyAnimNodeParameters)
   > Set Is Wearing Cape(Is Wearing Cape=CheckForBackbling().bIsBackblingACape)
   n3 = Get bIsBaseballSliding()
   n4 = Branch(execute=@n1.then_1, Condition=AND Boolean(A=Get bIsCrouching(), B=NOT Boolean(A=@n3)))
   > Set Alpha_Crouching(Alpha_Crouching:double=1) -> (Output_Get:double)
   n5 = Branch(execute=@n4.else, Condition=Get bIsBaseballSliding())
   > Set Alpha_Crouching(Alpha_Crouching=0)
   Set Alpha_Crouching(execute=@n5.else, Alpha_Crouching=0)
   n6 = Branch(execute=@n1.then_4, Condition=Get bIsPlayingEmote())
   > Set Alpha_ClothPants(Alpha_ClothPants:double=0.8) -> (Output_Get:double)
   Set Alpha_ClothPants(execute=@n6.else, Alpha_ClothPants=1)
   n7 = Get bIsMoving2D()
   Set Alpha_Idle(execute=@n1.then_5, Alpha_Idle:double="To Float (Boolean)"(InBool=NOT Boolean(A=@n7))) -> (Output_Get:double)
   n8 = NOT Boolean(A=Get bIsPlayingEmote())
   n9 = NOT Boolean(A=Get bIsSkyDiving())
   > Set Alpha_OnGround(Alpha_OnGround:double="To Float (Boolean)"(InBool=AND Boolean(A=@n9, B=@n8))) -> (Output_Get:double)
   n10 = AND Boolean(A=Get bIsFalling(), B=NOT Boolean(A=Get bIsSkyDiving()))
   n11 = Get bIsTacticalSprinting()
   > Set Alpha_Falling(Alpha_Falling:double="To Float (Boolean)"(InBool=OR Boolean(A=@n10, B=@n11))) -> (Output_Get:double)
   n12 = NOT Boolean(A=Get bIsSkyDiving())
   n13 = NOT Boolean(A=Get bIsPlayingEmote())
   n14 = NOT Boolean(A=Get bIsMoving2D())
   n15 = Get bIsInFrontEnd()
   > Set Alpha_FEIdle(Alpha_FEIdle:double="To Float (Boolean)"(InBool=AND Boolean!(A=@n14, B=@n13, C:bool=@n12, D:bool=@n15) -> (ReturnValue))) -> (Output_Get:double)
   n16 = Get bIsInFrontEnd()
   n17 = Get bIsSkydiveDiveMode()
   > Set SkydiveMode FE(SkydiveMode FE:double="To Float (Boolean)"(InBool=AND Boolean!(A=@n16, B=@n17) -> (ReturnValue))) -> (Output_Get:double)
   n18 = Get bIsSkydiveFloating()
   n19 = Get bIsSkydiveDiveMode()
   > Set Alpha_FloatingAndSkydiveMode(Alpha_FloatingAndSkydiveMode:double="To Float (Boolean)"(InBool=OR Boolean(A=@n18, B=@n19))) -> (Output_Get:double)
   n20 = GetAnimInstance(self=GetOwnerSkeletalMeshComponent())
   > Set PlayingCelebration(PlayingCelebration:double="To Float (Boolean)"(InBool=Montage_IsPlaying(self=@n20, Montage=/BRCosmetics/Animation/Game/MainPlayer/Emotes/Emote_Celebration_Female_M))) -> (Output_Get:double)
   n21 = GetOwnerSkeletalMeshComponent()
   n22 = BreakRotator(InRot=BreakTransform(InTransform=GetDeltaTransformFromRefPose(self=@n21, BoneName=neck_02, BaseName=spine_05)).Rotation)
   n23 = Branch(execute=@n1.then_7, Condition=AND Boolean(A="float < float"(A=@n22.Yaw, B=-33), B=Get bIsCrouching()))
   > Set Alpha_HeadRotateUpCrouchExtrem(Alpha_HeadRotateUpCrouchExtrem:double=1) -> (Output_Get:double)
   Set Alpha_HeadRotateUpCrouchExtrem(execute=@n23.else, Alpha_HeadRotateUpCrouchExtrem=0)
   n24 = Branch(execute=@n1.then_2, Condition=Get Is Wearing Cape() -> (Is Wearing Cape:bool))
   > Set Alpha_RB Standing(Alpha_RB Standing:double=0.8) -> (Output_Get:double)
   n25 = Branch(execute=@n24.else, Condition=Get bIsSkydiveFloating())
   > Set Alpha_RB Standing(Alpha_RB Standing=0.7)
   Set Alpha_RB Standing(execute=@n25.else, Alpha_RB Standing=1)
   n26 = Branch(execute=@n1.then_3, Condition=Get bIsSkydiveDiveMode())
   > Set Alpha_SkydiveMode(Alpha_SkydiveMode:double=0) -> (Output_Get:double)
   Set Alpha_SkydiveMode(execute=@n26.else, Alpha_SkydiveMode=1)
   n27 = Get bIsPlayingEmote()
   n28 = Branch(execute=@n1.then_6, Condition=AND Boolean(A=Get bIsInFrontEnd(), B=NOT Boolean(A=@n27)))
   > Set Sim Space Settings(Sim Space Settings:Anim Phys Sim Space Settings Structure=Get Sim Space Settings FronEnt() -> (Sim Space Settings FronEnt:Anim Phys Sim Space Settings Structure)) -> (Output_Get:Anim Phys Sim Space Settings Structure)
   Set Sim Space Settings(execute=@n28.else)
   n29 = Get bIsTacticalSprinting()
   n30 = Get bIsBaseballSliding()
   Set TacticalSprintingSliding(execute=@n1.then_8, TacticalSprintingSliding:double="To Float (Boolean)"(InBool=AND Boolean(A=@n29, B=@n30))) -> (Output_Get:double)
   Event OnParachuteOpened()
   > n31 = Switch on EGliderType(Selection=Get PlayerGliderType() -> (PlayerGliderType=HangGlider))
   Set Alpha_HangGliding(execute=[@n31.HangGlider, @n31.Umbrella], Alpha_HangGliding:double=1) -> (Output_Get:double)
   Event OnParachuteClosed()
   > Set Alpha_HangGliding(Alpha_HangGliding=0)
   Event OnWeaponChanged()
   > n32 = Cast To FortPlayerAnimInstance(Object=GetAnimInstance(self=GetOwnerSkeletalMeshComponent()))
   > Set Alpha_DetectedPistol(Alpha_DetectedPistol:double=0) -> (Output_Get:double)
   > n33 = Switch on EFortWeaponCoreAnimation!(execute, Selection:EFortWeaponCoreAnimation Enum=Get WeaponCoreAnim(self=@n32)) -> (
    Melee:Exec
    Pistol:Exec
    Shotgun:Exec
    PaperBlueprint:Exec
    Rifle:Exec
    MeleeOneHand:Exec
    MachinePistol:Exec
    RocketLauncher:Exec
    GrenadeLauncher:Exec
    GoingCommando:Exec
    AssaultRifle:Exec
    TacticalShotgun:Exec
    SniperRifle:Exec
    TrapPlacement:Exec
    ShoulderLauncher:Exec
    AbilityDecoTool:Exec
    Crossbow:Exec
    C4:Exec
    RemoteControl:Exec
    DualWield:Exec
    AR_BullPup:Exec
    AR_ForwardGrip:Exec
    MedPackPaddles:Exec
    SMG_P90:Exec
    AR_DrumGun:Exec
    Consumable_Small:Exec
    Consumable_Large:Exec
    Balloon:Exec
    MountedTurret:Exec
    CreativeTool:Exec
    ExplosiveBow:Exec
    AshtonIndigo:Exec
    AshtonChicago:Exec
    MeleeDualWield:Exec
    Unarmed:Exec
   )
   Set Alpha_DetectedPistol(execute=@n33.Pistol, Alpha_DetectedPistol=1)
  ]}
]
variables [
 Alpha_ClothPants {type=number, default=0.5}
 Alpha_Crouching {type=number}
 Alpha_DetectedPistol {type=number}
 Alpha_Falling {type=number}
 Alpha_FEIdle {type=number}
 Alpha_FloatingAndSkydiveMode {type=number}
 Alpha_HangGliding {type=number}
 Alpha_HeadRotateUpCrouchExtrem {type=number}
 Alpha_Idle {type=number}
 Alpha_NotRailAndSkydiving {type=number}
 Alpha_OnGround {type=number}
 Alpha_RB Standing {type=number}
 Alpha_SkydiveMode {type=number}
 Gravity Override {type=Vector}
 Is Wearing Cape {type=boolean}
 PlayingCelebration {type=number}
 RBAN Settings Out {type=FortRigidBodyAnimNodeParameters}
 Sim Space Settings {
  type=AnimPhysSimSpaceSettings
  default {maxAngularAcceleration=10000, maxAngularVelocity=10000, simSpaceAngularAlpha=0, externalAngularVelocity=<0 0 0>}
 }
 Sim Space Settings FronEnt {
  type=AnimPhysSimSpaceSettings
  default {maxAngularAcceleration=1000, maxAngularVelocity=300, simSpaceAngularAlpha=1, externalAngularVelocity=<0 0 0>}
 }
 Simulation Space {type=AnimPhysSimSpaceType, default=RootRelative}
 SkydiveMode FE {type=number}
 TacticalSprintingSliding {type=number}
]
