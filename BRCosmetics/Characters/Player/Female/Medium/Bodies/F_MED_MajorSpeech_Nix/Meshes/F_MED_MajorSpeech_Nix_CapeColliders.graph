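# Physics asset for the MajorSpeech_Nix female medium body, providing capsule colliders for its cape.
# Back, left and right dyn_skirt chains each link six capsule bodies up to the pelvis, with all swing and twist motion free.
# Head, clavicles, spine, pelvis and feet carry body capsules, doubled on the middle spine bones.
# Neck, arm, thigh and calf bodies exist but have no shapes.
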PhysicsAsset /BRCosmetics/Characters/Player/Female/Medium/Bodies/F_MED_MajorSpeech_Nix/Meshes/F_MED_MajorSpeech_Nix_CapeColliders
bodies {
 calf_l []
 calf_r []
 clavicle_l [{shapeType=Capsule, length=6.5623, radius=7.5194, center=<3.9566 -1.0862 -2.1127>, rotation=<rot -77.5566 177.3915 7.7314>}]
 clavicle_r [{shapeType=Capsule, length=6.5623, radius=7.5194, center=<-3.9566 1.0862 2.1127>, rotation=<rot -77.5566 177.3915 -172.2686>}]
 dyn_skirt_bk_01_c [{shapeType=Capsule, length=16.4502, radius=15, shapeName=dyn_skirt_bk_01_c_capsule, center=<12.2424 0 -3.0334>, rotation=<rot 90 -90 0>}]
 dyn_skirt_bk_02_c [{shapeType=Capsule, length=16.4502, radius=15, shapeName=dyn_skirt_bk_02_c_capsule, center=<12.2424 0 0>, rotation=<rot 90 -90 0>}]
 dyn_skirt_bk_03_c [{shapeType=Capsule, length=36.5453, radius=10, shapeName=dyn_skirt_bk_03_c_capsule, center=<12.2424 0 0>, rotation=<rot 90 -90 0>}]
 dyn_skirt_bk_04_c [{shapeType=Capsule, length=36.5453, radius=10, shapeName=dyn_skirt_bk_04_c_capsule, center=<12.2424 0 0>, rotation=<rot 90 -90 0>}]
 dyn_skirt_bk_05_c [{shapeType=Capsule, length=36.5453, radius=10, shapeName=dyn_skirt_bk_05_c_capsule, center=<12.2424 0 3.6083>, rotation=<rot 90 -90 0>}]
 dyn_skirt_bk_06_c [{shapeType=Capsule, length=36.5453, radius=10, shapeName=dyn_skirt_bk_06_c_capsule, center=<12.2424 0 4.1567>, rotation=<rot 90 -90 0>}]
 dyn_skirt_sd_l_01 [{shapeType=Capsule, length=7.5708, radius=15, shapeName=dyn_skirt_sd_l_01_capsule, center=<12.2424 -5.2927 0>, rotation=<rot 90 -90 0>}]
 dyn_skirt_sd_l_02 [{
   shapeType=Capsule
   length=16.4502
   radius=15
   shapeName=dyn_skirt_sd_l_02_capsule
   center=<12.2424 0 0>
   rotation=<rot 77.0611 89.9998 179.9998>
  }]
 dyn_skirt_sd_l_03 [{
   shapeType=Capsule
   length=36.5453
   radius=10
   shapeName=dyn_skirt_sd_l_03_capsule
   center=<12.2424 0 0>
   rotation=<rot 79.7414 89.9998 179.9998>
  }]
 dyn_skirt_sd_l_04 [{
   shapeType=Capsule
   length=36.5453
   radius=10
   shapeName=dyn_skirt_sd_l_04_capsule
   center=<12.2424 0 0>
   rotation=<rot 78.9482 90.0586 179.9995>
  }]
 dyn_skirt_sd_l_05 [{
   shapeType=Capsule
   length=36.5453
   radius=10
   shapeName=dyn_skirt_sd_l_05_capsule
   center=<12.2424 0.4724 3.5772>
   rotation=<rot 82.4776 89.9995 179.9995>
  }]
 dyn_skirt_sd_l_06 [{shapeType=Capsule, length=36.5453, radius=10, shapeName=dyn_skirt_sd_l_06_capsule, center=<13.0752 0 -3.3854>, rotation=<rot 90 -90 0>}]
 dyn_skirt_sd_r_01 [{shapeType=Capsule, length=7.5708, radius=15, shapeName=dyn_skirt_sd_r_01_capsule, center=<12.2424 3.5931 0>, rotation=<rot 90 -90 0>}]
 dyn_skirt_sd_r_02 [{
   shapeType=Capsule
   length=16.4502
   radius=15
   shapeName=dyn_skirt_sd_r_02_capsule
   center=<12.2424 0 0>
   rotation=<rot 77.6649 -89.9997 0.0004>
  }]
 dyn_skirt_sd_r_03 [{
   shapeType=Capsule
   length=36.5453
   radius=10
   shapeName=dyn_skirt_sd_r_03_capsule
   center=<12.2424 0 0>
   rotation=<rot 79.9678 -89.9997 0.0004>
  }]
 dyn_skirt_sd_r_04 [{
   shapeType=Capsule
   length=36.5453
   radius=10
   shapeName=dyn_skirt_sd_r_04_capsule
   center=<12.2424 0 0>
   rotation=<rot 79.476 -89.9248 0.0907>
  }]
 dyn_skirt_sd_r_05 [{
   shapeType=Capsule
   length=36.5453
   radius=10
   shapeName=dyn_skirt_sd_r_05_capsule
   center=<12.2424 -0.5234 3.5701>
   rotation=<rot 81.6588 -89.9988 0.0014>
  }]
 dyn_skirt_sd_r_06 [{shapeType=Capsule, length=36.5453, radius=10, shapeName=dyn_skirt_sd_r_06_capsule, center=<13.0752 0 -3.3854>, rotation=<rot 90 -90 0>}]
 foot_l [{shapeType=Capsule, length=11.0765, radius=8.0344, center=<-4.1811 6.0027 -0.338>, rotation=<rot 0.0388 1.4701 -89.3982>}]
 foot_r [{shapeType=Capsule, length=11.0765, radius=8.0344, center=<4.1811 -6.0027 0.338>, rotation=<rot 0.0388 1.4701 90.6018>}]
 head [{shapeType=Capsule, length=0, radius=18, center=<9.0697 0.0746 0>, rotation=<rot 0 -90 0>}]
 lowerarm_l []
 lowerarm_r []
 neck_02 []
 pelvis [{shapeType=Capsule, length=9.8493, radius=14.7392, center=<-2.4471 -2.5289 -0.1981>}]
 spine_01 [{shapeType=Capsule, length=2.157, radius=20.1895, center=<-8.9021 -0.8973 0>}]
 spine_02 [
  {shapeType=Capsule, length=4.7501, radius=14.9903, center=<-2.6885 -1.2641 0>}
  {shapeType=Capsule, length=20.6875, radius=9.3125, center=<-1.8317 1.9362 0>, rotation=<rot 90 0 0>}
 ]
 spine_03 [
  {shapeType=Capsule, length=3.1628, radius=14.2347, center=<0.0574 0.4416 0>}
  {shapeType=Capsule, length=20.7188, radius=9.2812, center=<-1.2163 0.549 0>, rotation=<rot 90 180 -175.0732>}
 ]
 spine_04 [
  {shapeType=Capsule, length=6.7757, radius=12.8286, center=<5.6234 -0.6779 0>}
  {shapeType=Capsule, length=10.0312, radius=9.2812, center=<-1.2326 1.308 0>, rotation=<rot 90 0 0>}
 ]
 spine_05 [{shapeType=Capsule, length=7.4462, radius=14.794, center=<0.6969 1.91 0>}]
 thigh_l []
 thigh_r []
 upperarm_l []
 upperarm_r []
}
constraints [
 {
  bone1Name=dyn_skirt_sd_l_01
  bone2Name=pelvis
  swing1LimitDegrees=45
  swing1Motion=Free
  swing2LimitDegrees=45
  swing2Motion=Free
  twistLimitDegrees=45
  twistMotion=Free
 }
 {
  bone1Name=dyn_skirt_sd_l_02
  bone2Name=dyn_skirt_sd_l_01
  swing1LimitDegrees=45
  swing1Motion=Free
  swing2LimitDegrees=45
  swing2Motion=Free
  twistLimitDegrees=45
  twistMotion=Free
 }
 {
  bone1Name=dyn_skirt_sd_l_03
  bone2Name=dyn_skirt_sd_l_02
  swing1LimitDegrees=45
  swing1Motion=Free
  swing2LimitDegrees=45
  swing2Motion=Free
  twistLimitDegrees=45
  twistMotion=Free
 }
 {
  bone1Name=dyn_skirt_sd_l_04
  bone2Name=dyn_skirt_sd_l_03
  swing1LimitDegrees=45
  swing1Motion=Free
  swing2LimitDegrees=45
  swing2Motion=Free
  twistLimitDegrees=45
  twistMotion=Free
 }
 {
  bone1Name=dyn_skirt_sd_l_05
  bone2Name=dyn_skirt_sd_l_04
  swing1LimitDegrees=45
  swing1Motion=Free
  swing2LimitDegrees=45
  swing2Motion=Free
  twistLimitDegrees=45
  twistMotion=Free
 }
 {
  bone1Name=dyn_skirt_sd_l_06
  bone2Name=dyn_skirt_sd_l_05
  swing1LimitDegrees=45
  swing1Motion=Free
  swing2LimitDegrees=45
  swing2Motion=Free
  twistLimitDegrees=45
  twistMotion=Free
 }
 {
  bone1Name=dyn_skirt_sd_r_01
  bone2Name=pelvis
  swing1LimitDegrees=45
  swing1Motion=Free
  swing2LimitDegrees=45
  swing2Motion=Free
  twistLimitDegrees=45
  twistMotion=Free
 }
 {
  bone1Name=dyn_skirt_sd_r_02
  bone2Name=dyn_skirt_sd_r_01
  swing1LimitDegrees=45
  swing1Motion=Free
  swing2LimitDegrees=45
  swing2Motion=Free
  twistLimitDegrees=45
  twistMotion=Free
 }
 {
  bone1Name=dyn_skirt_sd_r_03
  bone2Name=dyn_skirt_sd_r_02
  swing1LimitDegrees=45
  swing1Motion=Free
  swing2LimitDegrees=45
  swing2Motion=Free
  twistLimitDegrees=45
  twistMotion=Free
 }
 {
  bone1Name=dyn_skirt_sd_r_04
  bone2Name=dyn_skirt_sd_r_03
  swing1LimitDegrees=45
  swing1Motion=Free
  swing2LimitDegrees=45
  swing2Motion=Free
  twistLimitDegrees=45
  twistMotion=Free
 }
 {
  bone1Name=dyn_skirt_sd_r_05
  bone2Name=dyn_skirt_sd_r_04
  swing1LimitDegrees=45
  swing1Motion=Free
  swing2LimitDegrees=45
  swing2Motion=Free
  twistLimitDegrees=45
  twistMotion=Free
 }
 {
  bone1Name=dyn_skirt_sd_r_06
  bone2Name=dyn_skirt_sd_r_05
  swing1LimitDegrees=45
  swing1Motion=Free
  swing2LimitDegrees=45
  swing2Motion=Free
  twistLimitDegrees=45
  twistMotion=Free
 }
 {
  bone1Name=dyn_skirt_bk_06_c
  bone2Name=dyn_skirt_bk_05_c
  swing1LimitDegrees=45
  swing1Motion=Free
  swing2LimitDegrees=45
  swing2Motion=Free
  twistLimitDegrees=45
  twistMotion=Free
 }
 {
  bone1Name=dyn_skirt_bk_05_c
  bone2Name=dyn_skirt_bk_04_c
  swing1LimitDegrees=45
  swing1Motion=Free
  swing2LimitDegrees=45
  swing2Motion=Free
  twistLimitDegrees=45
  twistMotion=Free
 }
 {
  bone1Name=dyn_skirt_bk_04_c
  bone2Name=dyn_skirt_bk_03_c
  swing1LimitDegrees=45
  swing1Motion=Free
  swing2LimitDegrees=45
  swing2Motion=Free
  twistLimitDegrees=45
  twistMotion=Free
 }
 {
  bone1Name=dyn_skirt_bk_03_c
  bone2Name=dyn_skirt_bk_02_c
  swing1LimitDegrees=45
  swing1Motion=Free
  swing2LimitDegrees=45
  swing2Motion=Free
  twistLimitDegrees=45
  twistMotion=Free
 }
 {
  bone1Name=dyn_skirt_bk_02_c
  bone2Name=dyn_skirt_bk_01_c
  swing1LimitDegrees=45
  swing1Motion=Free
  swing2LimitDegrees=45
  swing2Motion=Free
  twistLimitDegrees=45
  twistMotion=Free
 }
 {
  bone1Name=dyn_skirt_bk_01_c
  bone2Name=pelvis
  swing1LimitDegrees=45
  swing1Motion=Free
  swing2LimitDegrees=45
  swing2Motion=Free
  twistLimitDegrees=45
  twistMotion=Free
 }
]
