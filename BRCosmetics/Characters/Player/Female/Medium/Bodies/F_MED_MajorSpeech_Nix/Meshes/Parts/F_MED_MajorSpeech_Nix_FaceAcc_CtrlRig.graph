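# Control Rig for the F_MED_MajorSpeech_Nix face accessory that applies corrective offsets to head pendant and ice_bk bones as the head moves.
# A Delta Rotation function remaps one axis of head rotation relative to head_null into clamped weights for Modify Transforms steps.
# One correction uses head rotation alone, while others also scale by smoothed Crouching or the Pistol value.
# A spherical pose reader on the head, with debug draw enabled, drives another correction gated by smoothed Targeting.
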ControlRigBlueprint /BRCosmetics/Characters/Player/Female/Medium/Bodies/F_MED_MajorSpeech_Nix/Meshes/Parts/F_MED_MajorSpeech_Nix_FaceAcc_CtrlRig
graphs [
 New Function_ContainedGraph {nodes [
   n1 = RigVMFunctionEntryNode()
   RigVMFunction_MathBoolToInteger = &RigVMFunction_MathBoolToInteger(Value=@n1.onlyTwist)
   RigVMFunction_MathQuaternionSwingTwist = &RigVMFunction_MathQuaternionSwingTwist(
    TwistAxis=@n1.twistAxis
    Input=&RigUnit_GetRelativeTransformForItem(
     Child=(Type=Bone,Name="None")
     Parent=(Type=Bone,Name="None")
     bChildInitial=False
     bParentInitial=False
     Child.Name=@n1.child
     Parent.Name=@n1.parent
     Parent.Type=@n1.parentType
    ).RelativeTransform.Rotation
   )
   RigVMFunction_MathQuaternionToEuler = &RigVMFunction_MathQuaternionToEuler(
    Value=RigVMDispatchNode(
     Values=((),())
     Index=@RigVMFunction_MathBoolToInteger.Result
     Values.0=@RigVMFunction_MathQuaternionSwingTwist.Swing
     Values.1=@RigVMFunction_MathQuaternionSwingTwist.Twist
    ) {template="DISPATCH_RigVMDispatch_SelectInt32(in Index,in Values,out Result)"}.Result
    RotationOrder=@n1.Rotation Order
   )
   > DISPATCH_RigVMDispatch_Print = RigVMDispatchNode(ScreenColor=(R=1,G=1,B=1,A=1), ScreenDuration=1.5, Enabled=@n1.debugRotation, Value=@RigVMFunction_MathQuaternionToEuler.Result) {template="DISPATCH_RigVMDispatch_Print(in Prefix,in Value,in Enabled,in ScreenDuration,in ScreenColor)"}
   Clamp = &RigVMFunction_MathIntClamp(Value=@n1.primAxis)
   DISPATCH_RigVMDispatch_SelectInt32 = RigVMDispatchNode(
    Values=(0,0,0)
    Index=@Clamp.Result
    Values.0=@RigVMFunction_MathQuaternionToEuler.Result.X
    Values.1=@RigVMFunction_MathQuaternionToEuler.Result.Y
    Values.2=@RigVMFunction_MathQuaternionToEuler.Result.Z
   ) {template="DISPATCH_RigVMDispatch_SelectInt32(in Index,in Values,out Result)"}
   Remap = &RigVMFunction_MathDoubleRemap(
    TargetMaximum=1
    TargetMinimum=0
    bClamp=False
    Value=@DISPATCH_RigVMDispatch_SelectInt32.Result
    SourceMaximum=@n1.max
    SourceMinimum=@n1.min
   )
   Multiply = &RigVMFunction_MathDoubleMul(A=&RigVMFunction_MathDoubleClamp(Value=@Remap.Result).Result, B=@n1.premult)
   > RigVMFunctionReturnNode(ouputWeight=@Multiply.Result)
  ]}
 RigVMFunctionLibrary {nodes [Delta Rotation = RigVMCollapseNode(ExecuteContext=(), debugRotation=False, onlyTwist=False, twistAxis=(X=0,Y=0,Z=0))]}
 RigVMModel {nodes [
   RigUnit_BeginExecution = &RigUnit_BeginExecution()
   Delta Rotation_3 = RigVMFunctionReferenceNode(
    Rotation Order=ZYX
    child=head
    debugRotation=False
    max=-40
    min=-10
    onlyTwist=False
    parent=head_null
    parentType=Null
    premult=1
    primAxis=2
    twistAxis=(X=1,Y=0,Z=0)
    ExecuteContext=@RigUnit_BeginExecution.ExecutePin
   )
   > &RigUnit_ModifyTransforms(
    ItemToModify="((Item=(Type=Bone,Name=\"base_pendants_sd_b_l\"),Transform=(Rotation=(X=0,Y=0,Z=0.0621,W=-0.9981),Translation=(X=-1.2564,Y=0.2896,Z=-0.6836),Scale3D=(X=0.8671,Y=1,Z=1))),(Item=(Type=Bone,Name=\"base_pendants_sd_b_r\"),Transform=(Rotation=(X=0.0327,Y=-0.0004,Z= ...(truncated)"
    WeightMinimum=0
    Weight=@"Delta Rotation_3".ouputWeight
   )
   > Delta Rotation_1 = RigVMFunctionReferenceNode(
    Rotation Order=ZYX
    child=head
    debugRotation=False
    max=-40
    min=10
    onlyTwist=False
    parent=head_null
    parentType=Null
    premult=1
    primAxis=2
    twistAxis=(X=1,Y=0,Z=0)
   )
   n1 = &RigVMFunction_AlphaInterp(
    InterpSpeedDecreasing=10
    InterpSpeedIncreasing=10
    Scale=1
    bClampResult=true
    Value=&RigVMFunction_MathBoolToFloat(Value=RigVMVariableNode() {variable=Crouching}.Value).Result
   )
   > &RigUnit_ModifyTransforms(
    ItemToModify="((Item=(Type=Bone,Name=\"base_pendants_sd_b_l\"),Transform=(Rotation=(X=0.0018,Y=0.0627,Z=-0.0281,W=0.9976),Translation=(X=-0.013,Y=0.0043,Z=0.4059),Scale3D=(X=0.91,Y=1,Z=1))),(Item=(Type=Bone,Name=\"base_pendants_sd_b_r\"),Transform=(Rotation=(X=-0.0001,Y=-0. ...(truncated)"
    WeightMinimum=0
    Weight=&RigVMFunction_MathDoubleMul(B=@n1.Result, A=@"Delta Rotation_1".ouputWeight).Result
   )
   > n2 = &RigUnit_SphericalPoseReader(
    ActiveRegionScaleFactors=(PositiveWidth=1,NegativeWidth=1,PositiveHeight=1,NegativeHeight=1)
    ActiveRegionSize=0.05
    Debug=(bDrawDebug=true,bDraw2D=False,bDrawLocalAxes=False,DebugScale=25,DebugSegments=20,DebugThickness=0.25)
    DriverAxis=(X=1,Y=0,Z=0)
    DriverItem=(Type=Bone,Name="head")
    FalloffRegionScaleFactors=(PositiveWidth=1,NegativeWidth=1,PositiveHeight=1,NegativeHeight=1)
    FalloffSize=0.2
    FlipHeightScaling=False
    FlipWidthScaling=False
    OptionalParentItem=(Type=Bone,Name="spine_05")
    RotationOffset=(X=-45,Y=0,Z=70)
   )
   n3 = &RigVMFunction_AlphaInterp(
    InterpSpeedDecreasing=20
    InterpSpeedIncreasing=20
    Scale=1
    bClampResult=true
    Value=&RigVMFunction_MathBoolToFloat(Value=RigVMVariableNode() {variable=Targeting}.Value).Result
   )
   > &RigUnit_ModifyTransforms(
    ItemToModify="((Item=(Type=Bone,Name=\"base_pendants_sd_b_r\"),Transform=(Rotation=(X=0,Y=0,Z=0.0209,W=0.9998),Translation=(X=0.1699,Y=0.0104,Z=0),Scale3D=(X=0.5,Y=1,Z=1))),(Item=(Type=Bone,Name=\"base_pendants_sd_a_r\"),Transform=(Rotation=(X=0,Y=0,Z=0.0126,W=0.9999),Trans ...(truncated)"
    WeightMinimum=0
    Weight=&RigVMFunction_MathDoubleMul(B=@n3.Result, A=@n2.OutputParam).Result
   )
   > &RigUnit_ModifyTransforms(
    ItemToModify="((Item=(Type=Bone,Name=\"ice_bk\"),Transform=(Rotation=(X=0,Y=0,Z=0,W=-1),Translation=(X=0,Y=0,Z=0),Scale3D=(X=1,Y=1,Z=1))),(Item=(Type=Bone,Name=\"base_pendants_bk\"),Transform=(Rotation=(X=0,Y=0.0807,Z=0,W=0.9967),Translation=(X=0.5092,Y=0,Z=0.3487),Scale3D= ...(truncated)"
    WeightMinimum=0
    Weight=&RigVMFunction_MathDoubleMul(B=RigVMVariableNode() {variable=Pistol}.Value, A=@"Delta Rotation_1".ouputWeight).Result
   )
   &RigUnit_PrepareForExecution()
  ]}
]
hierarchy [
 root {loc=<0 0 0>, rot=<0 0 0 1>}
 pelvis {parent=root, loc=<0 -0.5301 98.0571>, rot=<0.0127 0.707 -0.0127 -0.707>}
 spine_01 {parent=pelvis, loc=<6.1208 -0.9328 0>, rot=<0 0 0.034 -0.9994>}
 spine_02 {parent=spine_01, loc=<6.5997 -0.2095 0>, rot=<0 0 0.0238 -0.9997>}
 spine_03 {parent=spine_02, loc=<6.4914 -0.1237 0>, rot=<0 0 0.0149 -0.9999>}
 spine_04 {parent=spine_03, loc=<6.9598 0.0794 0>, rot=<0 0 -0.0132 -0.9999>}
 spine_05 {parent=spine_04, loc=<6.9987 0 0>, rot=<0 0 -0.0015 -1>}
 neck_01 {parent=spine_05, loc=<11.3138 0.1529 0>, rot=<0 0 -0.2003 -0.9797>}
 neck_02 {parent=neck_01, loc=<4.5627 0 0>, rot=<0 0 0.0504 -0.9987>}
 head {parent=neck_02, loc=<5.3914 0 0>, rot=<0 0 0.115 -0.9934>}
 base_hair_fr {parent=head, loc=<10.6672 10.265 1.5741>, rot=<0.0092 -0.9992 0.0031 -0.038>}
 dyn_hair_fr_01 {parent=base_hair_fr, loc=<0 0 0>, rot=<0 0 0 1>}
 dyn_hair_fr_02 {parent=dyn_hair_fr_01, loc=<3.6024 0.2666 0>, rot=<0 0 0 1>}
 dyn_hair_fr_03 {parent=dyn_hair_fr_02, loc=<4.6629 -0.5658 1.2963>, rot=<0 0 0 1>}
 dyn_hair_fr_04 {parent=dyn_hair_fr_03, loc=<3.3234 -1.5473 2.4079>, rot=<0 0 0 1>}
 base_pendants_sd_b_r {parent=head, loc=<6.9387 -9.7233 6.5644>, rot=<0.8586 0.0337 -0.0437 -0.5097>}
 dyn_pendants_sd_b_01_r {parent=base_pendants_sd_b_r, loc=<0 0 0>, rot=<0 0 0 1>}
 dyn_pendants_sd_b_02_r {parent=dyn_pendants_sd_b_01_r, loc=<-2.8722 0 0>, rot=<0 0 0 1>}
 base_pendants_sd_a_r {parent=head, loc=<7.5065 -3.6371 11.2015>, rot=<0.9956 0.0391 -0.0757 -0.04>}
 dyn_pendants_sd_a_01_r {parent=base_pendants_sd_a_r, loc=<0 0 0>, rot=<0 0 0 1>}
 dyn_pendants_sd_a_02_r {parent=dyn_pendants_sd_a_01_r, loc=<-3.2886 0 0>, rot=<0 0 0 1>}
 base_pendants_bk {parent=head, loc=<6.7295 -12.2728 -0.0066>, rot=<0.0029 -0.7071 0.7071 -0.0029>}
 dyn_pendants_bk_01 {parent=base_pendants_bk, loc=<0 0 0>, rot=<0 0 0 1>}
 dyn_pendants_bk_02 {parent=dyn_pendants_bk_01, loc=<2.6206 0 0>, rot=<0 0 0 1>}
 base_pendants_sd_b_l {parent=head, loc=<6.9384 -9.7234 -6.5644>, rot=<0.0337 -0.8586 0.5097 -0.0437>}
 dyn_pendants_sd_b_01_l {parent=base_pendants_sd_b_l, loc=<0 0 0>, rot=<0 0 0 1>}
 dyn_pendants_sd_b_02_l {parent=dyn_pendants_sd_b_01_l, loc=<2.8717 0 0>, rot=<0 0 0 1>}
 base_pendants_sd_a_l {parent=head, loc=<7.5065 -3.6371 -11.2015>, rot=<0.0391 -0.9956 0.04 -0.0757>}
 dyn_pendants_sd_a_01_l {parent=base_pendants_sd_a_l, loc=<0 0 0>, rot=<0 0 0 1>}
 dyn_pendants_sd_a_02_l {parent=dyn_pendants_sd_a_01_l, loc=<3.2884 0 0>, rot=<0 0 0 1>}
 ice_bk {parent=head, loc=<10.541 -7.667 0.013>, rot=<0.5021 -0.4979 0.4979 -0.5021>}
 base_hair_bk {parent=head, loc=<4.7328 -11.2859 0>, rot=<0.0029 -0.7071 0.7071 -0.0029>}
 dyn_hair_bk {parent=base_hair_bk, loc=<0 0 0>, rot=<0 0 0 1>}
 base_hair_sd_a_l {parent=head, loc=<2.9223 3.1858 -8.2774>, rot=<-0.0478 -0.9668 -0.1102 0.2257>}
 dyn_hair_sd_a_01_l {parent=base_hair_sd_a_l, loc=<0 0 0>, rot=<0 0 0 1>}
 dyn_hair_sd_a_02_l {parent=dyn_hair_sd_a_01_l, loc=<4.7462 0 0>, rot=<0 0 0 1>}
 base_hair_sd_b_l {parent=head, loc=<0.4192 0.2527 -7.8342>, rot=<-0.1024 -0.9558 -0.0962 0.2582>}
 dyn_hair_sd_b_01_l {parent=base_hair_sd_b_l, loc=<0 0 0>, rot=<0 0 0 1>}
 dyn_hair_sd_b_02_l {parent=dyn_hair_sd_b_01_l, loc=<4.5209 0 0>, rot=<0 0 0 1>}
 base_hair_sd_c_l {parent=head, loc=<-0.8102 -2.031 -7.685>, rot=<-0.1332 -0.9719 -0.0827 0.1757>}
 dyn_hair_sd_c_01_l {parent=base_hair_sd_c_l, loc=<0 0 0>, rot=<0 0 0 1>}
 dyn_hair_sd_c_02_l {parent=dyn_hair_sd_c_01_l, loc=<3.4674 0 0>, rot=<0 0 0 1>}
 base_hair_sd_a_r {parent=head, loc=<3.636 5.8197 7.1016>, rot=<0.9788 -0.078 0.1621 0.0974>}
 dyn_hair_sd_a_01_r {parent=base_hair_sd_a_r, loc=<0 0 0>, rot=<0 0 0 1>}
 dyn_hair_sd_a_02_r {parent=dyn_hair_sd_a_01_r, loc=<-3.2092 0.0001 -0.0003>, rot=<0 0 0 1>}
 base_hair_sd_b_r {parent=head, loc=<-0.2256 4.577 6.2674>, rot=<0.9575 -0.0848 0.2564 0.1009>}
 dyn_hair_sd_b_01_r {parent=base_hair_sd_b_r, loc=<0 0 0>, rot=<0 0 0 -1>}
 dyn_hair_sd_b_02_r {parent=dyn_hair_sd_b_01_r, loc=<-3.0549 -0.0002 0.0003>, rot=<0 0 0 -1>}
 base_hair_sd_c_r {parent=head, loc=<-0.4415 2.6031 6.7535>, rot=<0.926 -0.1625 0.2406 0.2415>}
 dyn_hair_sd_c_01_r {parent=base_hair_sd_c_r, loc=<0 0 0>, rot=<0 0 0 1>}
 dyn_hair_sd_c_02_r {parent=dyn_hair_sd_c_01_r, loc=<-3.4669 -0.0002 0.0002>, rot=<0 0 0 1>}
 base_hair_sd_d_r {parent=head, loc=<-1.0545 0.2174 6.8854>, rot=<0.9528 -0.1861 0.2155 0.1055>}
 dyn_hair_sd_d_01_r {parent=base_hair_sd_d_r, loc=<0 0 0>, rot=<0 0 0 1>}
 dyn_hair_sd_d_02_r {parent=dyn_hair_sd_d_01_r, loc=<-3.4669 -0.0002 0.0002>, rot=<0 0 0 1>}
 base_hair_sd_d_l {parent=head, loc=<2.7003 -4.4124 -10.1348>, rot=<-0.0406 -0.9456 0.285 0.1517>}
 dyn_hair_sd_d_01_l {parent=base_hair_sd_d_l, loc=<0 0 0>, rot=<0 0 0 1>}
 dyn_hair_sd_d_02_l {parent=dyn_hair_sd_d_01_l, loc=<2.9754 0 0>, rot=<0 0 0 1>}
 base_hair_sd_e_r {parent=head, loc=<4.6688 -0.6226 10.6239>, rot=<0.9966 -0.0063 0.0518 -0.0638>}
 dyn_hair_sd_e_01_r {parent=base_hair_sd_e_r, loc=<0 0 0>, rot=<0 0 0 1>}
 dyn_hair_sd_e_02_r {parent=dyn_hair_sd_e_01_r, loc=<-4.0712 -0.0002 0.0002>, rot=<0 0 0 1>}
 base_hair_sd_f_r {parent=head, loc=<-2.1416 -2.0439 6.6562>, rot=<0.9675 -0.1943 0.1348 0.0895>}
 dyn_hair_sd_f_01_r {parent=base_hair_sd_f_r, loc=<0 0 0>, rot=<0 0 0 1>}
 dyn_hair_sd_f_02_r {parent=dyn_hair_sd_f_01_r, loc=<-2.9342 -0.0002 0.0002>, rot=<0 0 0 1>}
 clavicle_l {parent=spine_05, loc=<5.4831 -0.1053 -3.2946>, rot=<0.0465 -0.7289 -0.009 -0.6829>}
 upperarm_l {parent=clavicle_l, loc=<9.8658 -1.2136 -0.4105>, rot=<0.0164 -0.3245 -0.0269 -0.9454>}
 upperarm_twist_01_l {parent=upperarm_l, loc=<7.2135 -0.1134 0>, rot=<0 0 0 1>}
 deform_upperarm_bk_base_l {parent=upperarm_twist_01_l, loc=<-6.4053 -4.3746 0.314>, rot=<0 0 0 -1>}
 deform_upperarm_bk_l {parent=deform_upperarm_bk_base_l, loc=<0 0 0>, rot=<0 0 0 1>}
 deform_upperarm_up_base_l {parent=upperarm_twist_01_l, loc=<-8.891 0.4825 3.9517>, rot=<0 0 0 -1>}
 deform_upperarm_up_l {parent=deform_upperarm_up_base_l, loc=<0 0 0>, rot=<0 0 0 1>}
 deform_upperarm_fr_base_l {parent=upperarm_twist_01_l, loc=<-5.6312 5.092 0.3354>, rot=<0 0 0 -1>}
 deform_upperarm_fr_l {parent=deform_upperarm_fr_base_l, loc=<0 0 0>, rot=<0 0 0 1>}
 base_ShoulderPad_back_l {parent=upperarm_twist_01_l, loc=<1.5069 -5.9925 -0.2578>, rot=<-0.7903 -0.0105 -0.06 -0.6096>}
 dyn_ShoulderPad_back_l {parent=base_ShoulderPad_back_l, loc=<0 0 0>, rot=<0 0 0 1>}
 base_ShoulderPad_front_l {parent=upperarm_twist_01_l, loc=<1.2738 3.5035 -0.6772>, rot=<0.6204 0.0486 0.0592 -0.7806>}
 dyn_ShoulderPad_front_l {parent=base_ShoulderPad_front_l, loc=<0 0 0>, rot=<0 0 0 1>}
 base_ShoulderPad_side_l {parent=upperarm_twist_01_l, loc=<0.5026 -1.15 4.1401>, rot=<0.0195 0.3559 -0.0643 -0.9321>}
 dyn_ShoulderPad_side_l {parent=base_ShoulderPad_side_l, loc=<0 0 0>, rot=<0 0 0 1>}
 lowerarm_l {parent=upperarm_l, loc=<26.465 -0.8345 0.259>, rot=<0.0036 0.0161 -0.2298 -0.9731>}
 hand_l {parent=lowerarm_l, loc=<22.7355 -0.3203 -0.6105>, rot=<-0.6784 -0.0223 0.0735 -0.7307>}
 thumb_01_l {parent=hand_l, loc=<1.3118 -0.3191 -2.459>, rot=<0.6546 -0.4042 0.0664 -0.6354>}
 thumb_02_l {parent=thumb_01_l, loc=<5.9336 -0.0448 0.0401>, rot=<-0.0037 0.0213 0.298 -0.9543>}
 thumb_03_l {parent=thumb_02_l, loc=<2.2223 0.3682 -0.0907>, rot=<0.0118 -0.013 0.0066 -0.9998>}
 index_metacarpal_l {parent=hand_l, loc=<1.6535 0.0183 -2.96>, rot=<0.0158 -0.0124 -0.0173 -0.9996>}
 index_01_l {parent=index_metacarpal_l, loc=<8.3008 0.0357 -0.0524>, rot=<-0.0044 0.0485 0.0924 -0.9945>}
 index_02_l {parent=index_01_l, loc=<3.3587 -0.3004 0.0452>, rot=<-0.002 0.0027 0.3225 -0.9466>}
 index_03_l {parent=index_02_l, loc=<2.5578 -0.123 -0.0252>, rot=<-0.002 0.0067 0.2558 -0.9667>}
 middle_metacarpal_l {parent=hand_l, loc=<2.1856 -0.1725 -0.5984>, rot=<0.006 0.0168 -0.0329 -0.9993>}
 middle_01_l {parent=middle_metacarpal_l, loc=<7.3275 -0.2447 0.0129>, rot=<-0.0208 0.0042 0.0858 -0.9961>}
 middle_02_l {parent=middle_01_l, loc=<4.0123 0.1401 0.1175>, rot=<-0.0218 0.0258 0.3512 -0.9357>}
 middle_03_l {parent=middle_02_l, loc=<2.6077 0.0534 0.1373>, rot=<0.0082 0.0188 0.1708 -0.9851>}
 ring_metacarpal_l {parent=hand_l, loc=<2.2074 -0.3625 1.3638>, rot=<-0.0284 0.0722 -0.0458 -0.9959>}
 ring_01_l {parent=ring_metacarpal_l, loc=<7.0295 -0.6197 -0.1446>, rot=<-0.0489 -0.0252 0.1892 -0.9804>}
 ring_02_l {parent=ring_01_l, loc=<3.6174 0.4647 -0.1027>, rot=<0 0 0.2806 -0.9598>}
 ring_03_l {parent=ring_02_l, loc=<2.4152 0.114 0.0199>, rot=<0 0 0.1189 -0.9929>}
 pinky_metacarpal_l {parent=hand_l, loc=<1.7616 -0.9753 3.3814>, rot=<-0.0397 0.0814 -0.069 -0.9935>}
 pinky_01_l {parent=pinky_metacarpal_l, loc=<7.1709 -0.4831 -0.1094>, rot=<-0.0654 -0.0035 0.2242 -0.9723>}
 pinky_02_l {parent=pinky_01_l, loc=<2.8274 0.0893 -0.1043>, rot=<0.0037 0.0011 0.2854 -0.9584>}
 pinky_03_l {parent=pinky_02_l, loc=<2.2046 0.0307 -0.0078>, rot=<-0.0002 0.0098 0.1445 -0.9895>}
 deform_wrist_dn_base_l {parent=hand_l, loc=<0 -1.5326 0>, rot=<0 0 0 -1>}
 deform_wrist_dn_l {parent=deform_wrist_dn_base_l, loc=<0 0 0>, rot=<0 0 0 1>}
 deform_wrist_up_base_l {parent=hand_l, loc=<0 2.3009 0>, rot=<0 0 0 -1>}
 deform_wrist_up_l {parent=deform_wrist_up_base_l, loc=<0 0 0>, rot=<0 0 0 1>}
 lowerarm_twist_02_l {parent=lowerarm_l, loc=<10.6707 0 0>, rot=<0 0 0 1>}
 lowerarm_twist_01_l {parent=lowerarm_l, loc=<18.316 0 0>, rot=<0 0 0 1>}
 lowerarm_twist_extra_l {parent=lowerarm_twist_01_l, loc=<0 0 0>, rot=<0 0 0 1>}
 deform_elbow_in_base_l {parent=lowerarm_l, loc=<0.1834 3.6062 -0.2572>, rot=<0 0 0 -1>}
 deform_elbow_in_l {parent=deform_elbow_in_base_l, loc=<0 0 0>, rot=<0 0 0 1>}
 deform_elbow_out_base_l {parent=lowerarm_l, loc=<0 -4 0>, rot=<0 0 0 1>}
 deform_elbow_out_l {parent=deform_elbow_out_base_l, loc=<0 0 0>, rot=<0 0 0 1>}
 base_sleeve_l {parent=lowerarm_l, loc=<0 0 0>, rot=<0 0 0 1>}
 base_sleeves_twist_a_l {parent=base_sleeve_l, loc=<10.6707 0 0>, rot=<-0.3794 -0.3794 0.5967 -0.5967>}
 dyn_sleeves_twist_a_01_l {parent=base_sleeves_twist_a_l, loc=<0 0 0>, rot=<0 0 0 1>}
 dyn_sleeves_twist_a_02_l {parent=dyn_sleeves_twist_a_01_l, loc=<4.0997 0 0>, rot=<0 0 0 1>}
 base_sleeves_twist_b_l {parent=base_sleeve_l, loc=<18.316 0 0>, rot=<-0.3758 -0.3758 0.599 -0.599>}
 dyn_sleeves_twist_b_01_l {parent=base_sleeves_twist_b_l, loc=<0 0 0>, rot=<0 0 0 1>}
 dyn_sleeves_twist_b_02_l {parent=dyn_sleeves_twist_b_01_l, loc=<5.9238 0 0>, rot=<0 0 0 1>}
 dyn_sleeves_twist_b_03_l {parent=dyn_sleeves_twist_b_02_l, loc=<4.5539 0 0>, rot=<0 0 0 1>}
 upperarm_twist_02_l {parent=upperarm_l, loc=<19.7998 -0.1134 0>, rot=<0 0 0 1>}
 clavicle_r {parent=spine_05, loc=<5.4831 -0.1053 3.2946>, rot=<0.7289 0.0465 -0.6829 0.009>}
 upperarm_r {parent=clavicle_r, loc=<-9.8658 1.2136 0.4105>, rot=<0.0164 -0.3245 -0.0269 -0.9454>}
 upperarm_twist_01_r {parent=upperarm_r, loc=<-7.2135 0.1134 0>, rot=<0 0 0 1>}
 deform_upperarm_bk_base_r {parent=upperarm_twist_01_r, loc=<6.4053 4.3746 -0.314>, rot=<0 0 0 -1>}
 deform_upperarm_bk_r {parent=deform_upperarm_bk_base_r, loc=<0 0 0>, rot=<0 0 0 1>}
 deform_upperarm_up_base_r {parent=upperarm_twist_01_r, loc=<8.891 -0.4825 -3.9517>, rot=<0 0 0 -1>}
 deform_upperarm_up_r {parent=deform_upperarm_up_base_r, loc=<0 0 0>, rot=<0 0 0 1>}
 deform_upperarm_fr_base_r {parent=upperarm_twist_01_r, loc=<5.6312 -5.092 -0.3354>, rot=<0 0 0 -1>}
 deform_upperarm_fr_r {parent=deform_upperarm_fr_base_r, loc=<0 0 0>, rot=<0 0 0 1>}
 base_ShoulderPad_side_r {parent=upperarm_twist_01_r, loc=<-0.5023 1.15 -4.1404>, rot=<0.0195 0.3559 -0.0643 -0.9321>}
 dyn_ShoulderPad_side_r {parent=base_ShoulderPad_side_r, loc=<0 0 0>, rot=<0 0 0 -1>}
 base_ShoulderPad_front_r {parent=upperarm_twist_01_r, loc=<-1.2734 -3.5035 0.6769>, rot=<0.6204 0.0486 0.0592 -0.7806>}
 dyn_ShoulderPad_front_r {parent=base_ShoulderPad_front_r, loc=<0 0 0>, rot=<0 0 0 -1>}
 base_ShoulderPad_back_r {parent=upperarm_twist_01_r, loc=<-1.5068 5.9925 0.2577>, rot=<-0.7903 -0.0105 -0.06 -0.6096>}
 dyn_ShoulderPad_back_r {parent=base_ShoulderPad_back_r, loc=<0 0 0>, rot=<0 0 0 1>}
 lowerarm_r {parent=upperarm_r, loc=<-26.465 0.8345 -0.259>, rot=<0.0036 0.0161 -0.2298 -0.9731>}
 hand_r {parent=lowerarm_r, loc=<-22.7355 0.3203 0.6106>, rot=<-0.6784 -0.0223 0.0735 -0.7307>}
 thumb_01_r {parent=hand_r, loc=<-1.3118 0.3191 2.459>, rot=<0.6546 -0.4042 0.0664 -0.6354>}
 thumb_02_r {parent=thumb_01_r, loc=<-5.9336 0.0448 -0.0401>, rot=<-0.0037 0.0213 0.298 -0.9543>}
 thumb_03_r {parent=thumb_02_r, loc=<-2.2223 -0.3682 0.0907>, rot=<0.0118 -0.013 0.0066 -0.9998>}
 index_metacarpal_r {parent=hand_r, loc=<-1.6535 -0.0183 2.96>, rot=<0.0158 -0.0124 -0.0173 -0.9996>}
 index_01_r {parent=index_metacarpal_r, loc=<-8.3008 -0.0357 0.0524>, rot=<-0.0044 0.0485 0.0924 -0.9945>}
 index_02_r {parent=index_01_r, loc=<-3.3587 0.3004 -0.0452>, rot=<-0.002 0.0027 0.3225 -0.9466>}
 index_03_r {parent=index_02_r, loc=<-2.5578 0.123 0.0252>, rot=<-0.002 0.0067 0.2558 -0.9667>}
 middle_metacarpal_r {parent=hand_r, loc=<-2.1856 0.1725 0.5984>, rot=<0.006 0.0168 -0.0329 -0.9993>}
 middle_01_r {parent=middle_metacarpal_r, loc=<-7.3275 0.2447 -0.0129>, rot=<-0.0208 0.0042 0.0858 -0.9961>}
 middle_02_r {parent=middle_01_r, loc=<-4.0123 -0.1401 -0.1175>, rot=<-0.0218 0.0258 0.3512 -0.9357>}
 middle_03_r {parent=middle_02_r, loc=<-2.6077 -0.0534 -0.1373>, rot=<0.0082 0.0188 0.1708 -0.9851>}
 ring_metacarpal_r {parent=hand_r, loc=<-2.2074 0.3625 -1.3638>, rot=<-0.0284 0.0722 -0.0458 -0.9959>}
 ring_01_r {parent=ring_metacarpal_r, loc=<-7.0295 0.6197 0.1446>, rot=<-0.0489 -0.0252 0.1892 -0.9804>}
 ring_02_r {parent=ring_01_r, loc=<-3.6174 -0.4647 0.1027>, rot=<0 0 0.2806 -0.9598>}
 ring_03_r {parent=ring_02_r, loc=<-2.4152 -0.114 -0.0199>, rot=<0 0 0.1189 -0.9929>}
 pinky_metacarpal_r {parent=hand_r, loc=<-1.7616 0.9753 -3.3814>, rot=<-0.0397 0.0814 -0.069 -0.9935>}
 pinky_01_r {parent=pinky_metacarpal_r, loc=<-7.1709 0.4831 0.1094>, rot=<-0.0654 -0.0035 0.2242 -0.9723>}
 pinky_02_r {parent=pinky_01_r, loc=<-2.8274 -0.0893 0.1043>, rot=<0.0037 0.0011 0.2854 -0.9584>}
 pinky_03_r {parent=pinky_02_r, loc=<-2.2046 -0.0307 0.0078>, rot=<-0.0002 0.0098 0.1445 -0.9895>}
 deform_wrist_dn_base_r {parent=hand_r, loc=<-0.0001 1.5328 0>, rot=<0 0 0 -1>}
 deform_wrist_dn_r {parent=deform_wrist_dn_base_r, loc=<0 0 0>, rot=<0 0 0 1>}
 deform_wrist_up_base_r {parent=hand_r, loc=<-0.0002 -2.3007 -0.0001>, rot=<0 0 0 -1>}
 deform_wrist_up_r {parent=deform_wrist_up_base_r, loc=<0 0 0>, rot=<0 0 0 1>}
 lowerarm_twist_02_r {parent=lowerarm_r, loc=<-10.6707 0 0>, rot=<0 0 0 1>}
 lowerarm_twist_01_r {parent=lowerarm_r, loc=<-18.316 0 0>, rot=<0 0 0 1>}
 lowerarm_twist_extra_r {parent=lowerarm_twist_01_r, loc=<0 0 0>, rot=<0 0 0 1>}
 deform_elbow_in_base_r {parent=lowerarm_r, loc=<-0.1834 -3.6062 0.2572>, rot=<0 0 0 -1>}
 deform_elbow_in_r {parent=deform_elbow_in_base_r, loc=<0 0 0>, rot=<0 0 0 1>}
 deform_elbow_out_base_r {parent=lowerarm_r, loc=<0 4 -0.0001>, rot=<0 0 0 -1>}
 deform_elbow_out_r {parent=deform_elbow_out_base_r, loc=<0 0 0>, rot=<0 0 0 1>}
 base_sleeve_r {parent=lowerarm_r, loc=<0 0 0.0001>, rot=<0 0 0 -1>}
 base_sleeves_twist_a_r {parent=base_sleeve_r, loc=<-10.6709 0.0001 0.0003>, rot=<-0.3794 -0.3794 0.5967 -0.5967>}
 dyn_sleeves_twist_a_01_r {parent=base_sleeves_twist_a_r, loc=<0 0 0>, rot=<0 0 0 1>}
 dyn_sleeves_twist_a_02_r {parent=dyn_sleeves_twist_a_01_r, loc=<-4.0991 0.0005 -0.0001>, rot=<0 0 0 1>}
 base_sleeves_twist_b_r {parent=base_sleeve_r, loc=<-18.3163 0.0001 0.0002>, rot=<-0.3758 -0.3758 0.599 -0.599>}
 dyn_sleeves_twist_b_01_r {parent=base_sleeves_twist_b_r, loc=<0 0 0>, rot=<0 0 0 1>}
 dyn_sleeves_twist_b_02_r {parent=dyn_sleeves_twist_b_01_r, loc=<-5.9233 0.0003 -0.0001>, rot=<0 0 0 1>}
 dyn_sleeves_twist_b_03_r {parent=dyn_sleeves_twist_b_02_r, loc=<-4.554 -0.0001 0>, rot=<0 0 0 1>}
 upperarm_twist_02_r {parent=upperarm_r, loc=<-19.7998 0.1134 0>, rot=<0 0 0 1>}
 head_null {parent=spine_05, type=Null, loc=<20.6573 3.5489 0>, rot=<0 0 -0.036 -0.9994>}
 base_collar_a_l {parent=spine_05, loc=<8.9997 1.8965 -8.8979>, rot=<0.0001 -0.0788 -0.0784 -0.9938>}
 dyn_collar_a_01_l {parent=base_collar_a_l, loc=<0 0 0>, rot=<0 0 0 1>}
 dyn_collar_a_02_l {parent=dyn_collar_a_01_l, loc=<4.011 0 0>, rot=<0 0 0 1>}
 base_collar_b_l {parent=spine_05, loc=<11.3039 -4.1761 -6.3777>, rot=<0.3354 -0.1311 -0.0258 -0.9326>}
 dyn_collar_b_01_l {parent=base_collar_b_l, loc=<0 0 0>, rot=<0 0 0 1>}
 dyn_collar_b_02_l {parent=dyn_collar_b_01_l, loc=<4.8426 0 0>, rot=<0 0 0 1>}
 base_collar_a_r {parent=spine_05, loc=<9 1.8965 8.8979>, rot=<-0.0788 -0.0001 0.9938 -0.0784>}
 dyn_collar_a_01_r {parent=base_collar_a_r, loc=<0 0 0>, rot=<0 0 0 1>}
 dyn_collar_a_02_r {parent=dyn_collar_a_01_r, loc=<-4.0106 0 0.0001>, rot=<0 0 0 1>}
 base_collar_b_r {parent=spine_05, loc=<11.304 -4.1761 6.3777>, rot=<-0.1311 -0.3354 0.9326 -0.0258>}
 dyn_collar_b_01_r {parent=base_collar_b_r, loc=<0 0 0>, rot=<0 0 0 1>}
 dyn_collar_b_02_r {parent=dyn_collar_b_01_r, loc=<-4.8424 0 0.0001>, rot=<0 0 0 1>}
 base_collar_c {parent=spine_05, loc=<11.8704 -5.6573 0>, rot=<0 0 0.1521 -0.9884>}
 dyn_collar_01_c {parent=base_collar_c, loc=<0 0 0>, rot=<0.7071 0 0 -0.7071>}
 dyn_collar_02_c {parent=dyn_collar_01_c, loc=<7.0945 0 0>, rot=<0 0 0 1>}
 deform_pec_aim_base_l {parent=spine_04, loc=<3.2456 9.3309 -1.2237>, rot=<0.0603 -0.4033 0.3318 -0.8507>}
 deform_pec_aim_l {parent=deform_pec_aim_base_l, loc=<0 0 0>, rot=<0 0 0 1>}
 deform_pec_base_l {parent=deform_pec_aim_l, loc=<9.3665 0 0>, rot=<0 0 0 1>}
 deform_pec_l {parent=deform_pec_base_l, loc=<0 0 0>, rot=<0 0 0 1>}
 deform_pec_aim_base_r {parent=spine_04, loc=<3.2452 9.3309 1.2237>, rot=<-0.4033 -0.0603 0.8507 0.3318>}
 deform_pec_aim_r {parent=deform_pec_aim_base_r, loc=<0 0 0>, rot=<0 0 0 1>}
 deform_pec_base_r {parent=deform_pec_aim_r, loc=<-9.3667 -0.0003 -0.0004>, rot=<0 0 0 1>}
 deform_pec_r {parent=deform_pec_base_r, loc=<0 0 0>, rot=<0 0 0 1>}
 base_spine_pendant {parent=spine_04, loc=<-4.1019 8.9325 5.3992>, rot=<-0.0339 0.5788 0.8139 -0.0387>}
 dyn_spine_pendant_01 {parent=base_spine_pendant, loc=<0 0 0>, rot=<0 0 0 1>}
 dyn_spine_pendant_02 {parent=dyn_spine_pendant_01, loc=<3.2749 0.5336 0.5497>, rot=<0 0 0 1>}
 dyn_spine_pendant_03 {parent=dyn_spine_pendant_02, loc=<2.4154 0 0>, rot=<0 0 0 1>}
 thigh_l {parent=pelvis, loc=<-1.0855 -0.2109 -8.7053>, rot=<0.0009 0.0355 0.0155 -0.9993>}
 calf_l {parent=thigh_l, loc=<-42.5284 -0.0019 -0.7297>, rot=<0.0006 -0.0161 -0.0391 -0.9991>}
 foot_l {parent=calf_l, loc=<-46.7401 -2.5014 -0.592>, rot=<0.0427 -0.0238 0.0353 -0.9982>}
 ball_l {parent=foot_l, loc=<-3.5658 10.8066 -0.1544>, rot=<0.0001 -0.0001 -0.7063 -0.7079>}
 calf_twist_02_l {parent=calf_l, loc=<-19.1546 -1.7943 -0.2638>, rot=<0.0237 0.0095 -0.0035 -0.9997>}
 calf_twist_01_l {parent=calf_l, loc=<-35.8483 -1.9197 -0.5794>, rot=<0.0237 0.0095 -0.0035 -0.9997>}
 deform_knee_in_base_l {parent=calf_l, loc=<0.8031 -5.0231 0.2456>, rot=<0 0 0 -1>}
 deform_knee_in_l {parent=deform_knee_in_base_l, loc=<0 0 0>, rot=<0 0 0 1>}
 deform_knee_out_base_l {parent=calf_l, loc=<0 4 0>, rot=<0 0 0 -1>}
 deform_knee_out_l {parent=deform_knee_out_base_l, loc=<0 0 0>, rot=<0 0 0 1>}
 base_ClothPants_fr_l {parent=calf_l, loc=<-6.5334 10.1037 0.1344>, rot=<-0.0396 -0.7504 -0.6543 -0.0855>}
 dyn_ClothPants_fr_01_l {parent=base_ClothPants_fr_l, loc=<0 0 0>, rot=<0 0 0 1>}
 dyn_ClothPants_fr_02_l {parent=dyn_ClothPants_fr_01_l, loc=<20.1883 0 0>, rot=<0 0 0 1>}
 base_ClothPants_in_l {parent=calf_l, loc=<-7.7183 -1.1251 7.9804>, rot=<-0.0164 0.0059 0.9995 0.0263>}
 dyn_ClothPants_in_01_l {parent=base_ClothPants_in_l, loc=<0 0 0>, rot=<0 0 0 1>}
 dyn_ClothPants_in_02_l {parent=dyn_ClothPants_in_01_l, loc=<20.1883 0 0>, rot=<0 0 0 1>}
 base_ClothPants_bk_l {parent=calf_l, loc=<-8.1296 -12.3971 0.9111>, rot=<-0.0199 -0.7023 0.7115 -0.0117>}
 dyn_ClothPants_bk_01_l {parent=base_ClothPants_bk_l, loc=<0 0 0>, rot=<0 0 0 1>}
 dyn_ClothPants_bk_02_l {parent=dyn_ClothPants_bk_01_l, loc=<20.1883 0 0>, rot=<0 0 0 1>}
 base_ClothPants_out_l {parent=calf_l, loc=<-7.038 -0.873 -6.2917>, rot=<-0.0054 -0.9971 0.0068 -0.0762>}
 dyn_ClothPants_out_01_l {parent=base_ClothPants_out_l, loc=<0 0 0>, rot=<0 0 0 1>}
 dyn_ClothPants_out_02_l {parent=dyn_ClothPants_out_01_l, loc=<20.1883 0 0>, rot=<0 0 0 1>}
 thigh_twist_01_l {parent=thigh_l, loc=<-27.952 0 0>, rot=<0.0013 0 0 -1>}
 deform_groin_base_l {parent=thigh_l, loc=<0.4801 5.6776 -0.2486>, rot=<0 0 0 -1>}
 deform_groin_l {parent=deform_groin_base_l, loc=<0 0 0>, rot=<0 0 0 1>}
 thigh_r {parent=pelvis, loc=<-1.0855 -0.2109 8.7053>, rot=<0.0355 -0.0009 0.9993 0.0155>}
 calf_r {parent=thigh_r, loc=<42.5284 0.0019 0.7297>, rot=<0.0006 -0.0161 -0.0391 -0.9991>}
 foot_r {parent=calf_r, loc=<46.7401 2.5014 0.592>, rot=<0.0427 -0.0238 0.0353 -0.9982>}
 ball_r {parent=foot_r, loc=<3.5658 -10.8066 0.1544>, rot=<0.0001 -0.0001 -0.7063 -0.7079>}
 calf_twist_02_r {parent=calf_r, loc=<19.1546 1.7943 0.2638>, rot=<0.0237 0.0095 -0.0035 -0.9997>}
 calf_twist_01_r {parent=calf_r, loc=<35.8483 1.9197 0.5794>, rot=<0.0237 0.0095 -0.0035 -0.9997>}
 deform_knee_in_base_r {parent=calf_r, loc=<-0.8031 5.0231 -0.2457>, rot=<0 0 0 -1>}
 deform_knee_in_r {parent=deform_knee_in_base_r, loc=<0 0 0>, rot=<0 0 0 1>}
 deform_knee_out_base_r {parent=calf_r, loc=<0 -4 0>, rot=<0 0 0 -1>}
 deform_knee_out_r {parent=deform_knee_out_base_r, loc=<0 0 0>, rot=<0 0 0 1>}
 base_ClothPants_fr_r {parent=calf_r, loc=<6.5334 -10.1037 -0.1344>, rot=<-0.6543 0.0855 0.0396 -0.7504>}
 dyn_ClothPants_fr_01_r {parent=base_ClothPants_fr_r, loc=<0 0 0>, rot=<0 0 0 1>}
 dyn_ClothPants_fr_02_r {parent=dyn_ClothPants_fr_01_r, loc=<20.1883 0 0>, rot=<0 0 0 1>}
 base_ClothPants_in_r {parent=calf_r, loc=<7.7183 1.1251 -7.9804>, rot=<0.9995 -0.0263 0.0164 0.0059>}
 dyn_ClothPants_in_01_r {parent=base_ClothPants_in_r, loc=<0 0 0>, rot=<0 0 0 1>}
 dyn_ClothPants_in_02_r {parent=dyn_ClothPants_in_01_r, loc=<20.1883 0 0>, rot=<0 0 0 1>}
 base_ClothPants_bk_r {parent=calf_r, loc=<8.1296 12.3971 -0.9112>, rot=<0.7115 0.0117 0.0199 -0.7023>}
 dyn_ClothPants_bk_01_r {parent=base_ClothPants_bk_r, loc=<0 0 0>, rot=<0 0 0 1>}
 dyn_ClothPants_bk_02_r {parent=dyn_ClothPants_bk_01_r, loc=<20.1884 -0.0001 0>, rot=<0 0 0 1>}
 base_ClothPants_out_r {parent=calf_r, loc=<7.038 0.873 6.2918>, rot=<0.0068 0.0762 0.0054 -0.9971>}
 dyn_ClothPants_out_01_r {parent=base_ClothPants_out_r, loc=<0 0 0>, rot=<0 0 0 1>}
 dyn_ClothPants_out_02_r {parent=dyn_ClothPants_out_01_r, loc=<20.1883 0 -0.0001>, rot=<0 0 0 -1>}
 thigh_twist_01_r {parent=thigh_r, loc=<27.952 0 0>, rot=<0.0013 0 0 -1>}
 deform_groin_base_r {parent=thigh_r, loc=<-0.4801 -5.6776 0.2486>, rot=<0 0 0 -1>}
 deform_groin_r {parent=deform_groin_base_r, loc=<0 0 0>, rot=<0 0 0 1>}
 deform_glute_base_l {parent=pelvis, loc=<-7.3466 -9.0213 -9.1245>, rot=<0.0172 0.031 -0.4477 -0.8935>}
 deform_glute_l {parent=deform_glute_base_l, loc=<0 0 0>, rot=<0 0 0 1>}
 deform_glute_base_r {parent=pelvis, loc=<-7.3464 -9.0213 9.1245>, rot=<0.031 -0.0172 0.8935 -0.4477>}
 deform_glute_r {parent=deform_glute_base_r, loc=<0 0 0>, rot=<0 0 0 1>}
 base_skirt_sd_l {parent=pelvis, loc=<4.6906 -0.4311 -10.731>, rot=<0.0493 -0.9109 0.3808 -0.1512>}
 dyn_skirt_sd_l_01 {parent=base_skirt_sd_l, loc=<0 0 0>, rot=<0 0 0 1>}
 dyn_skirt_sd_l_02 {parent=dyn_skirt_sd_l_01, loc=<17.4293 0 0>, rot=<0.0176 -0.0329 0.0279 -0.9989>}
 dyn_skirt_sd_l_03 {parent=dyn_skirt_sd_l_02, loc=<17.4269 0 0>, rot=<-0.0537 0.012 0.0189 -0.9983>}
 dyn_skirt_sd_l_04 {parent=dyn_skirt_sd_l_03, loc=<17.4326 0 0>, rot=<0.0105 -0.0083 -0.0032 -0.9999>}
 dyn_skirt_sd_l_05 {parent=dyn_skirt_sd_l_04, loc=<17.4329 0 0>, rot=<0.0062 0.0008 -0.0018 -1>}
 dyn_skirt_sd_l_06 {parent=dyn_skirt_sd_l_05, loc=<17.4332 0 0>, rot=<0.0113 0.0414 -0.0037 -0.9991>}
 base_skirt_bk_c {parent=pelvis, loc=<4.2714 -10.9847 0.0007>, rot=<0.0611 -0.7044 0.7045 -0.0611>}
 dyn_skirt_bk_01_c {parent=base_skirt_bk_c, loc=<0 0 0>, rot=<0 0 0 1>}
 dyn_skirt_bk_02_c {parent=dyn_skirt_bk_01_c, loc=<17.2607 0 0>, rot=<0.0003 -0.0123 -0.0001 -0.9999>}
 dyn_skirt_bk_03_c {parent=dyn_skirt_bk_02_c, loc=<17.2641 0 0>, rot=<-0.0002 0.0222 0.0001 -0.9998>}
 dyn_skirt_bk_04_c {parent=dyn_skirt_bk_03_c, loc=<17.2644 0 0>, rot=<0 0.0151 0 -0.9999>}
 dyn_skirt_bk_05_c {parent=dyn_skirt_bk_04_c, loc=<17.2644 0 0>, rot=<0 0.0243 0 -0.9997>}
 dyn_skirt_bk_06_c {parent=dyn_skirt_bk_05_c, loc=<17.2638 0 0>, rot=<0 0.0333 0 -0.9994>}
 base_skirt_fr_r {parent=pelvis, loc=<5.1484 6.6668 4.1466>, rot=<0.1627 -0.2366 0.9563 0.0549>}
 dyn_skirt_fr_r_01 {parent=base_skirt_fr_r, loc=<0 0 0>, rot=<0 0 0 1>}
 dyn_skirt_fr_r_02 {parent=dyn_skirt_fr_r_01, loc=<13 0 0>, rot=<0 0 0 1>}
 dyn_skirt_fr_r_03 {parent=dyn_skirt_fr_r_02, loc=<13 0 0>, rot=<0 0 0 1>}
 dyn_skirt_fr_r_04 {parent=dyn_skirt_fr_r_03, loc=<13 0 0>, rot=<0 0 -0.0411 -0.9992>}
 dyn_skirt_fr_r_05 {parent=dyn_skirt_fr_r_04, loc=<13 0 0>, rot=<0 0 -0.0275 -0.9996>}
 dyn_skirt_fr_r_06 {parent=dyn_skirt_fr_r_05, loc=<13 0 0>, rot=<0.0204 -0.0002 -0.0288 -0.9994>}
 dyn_skirt_fr_r_07 {parent=dyn_skirt_fr_r_06, loc=<13 0 0>, rot=<0 0 0 1>}
 dyn_skirt_fr_r_08 {parent=dyn_skirt_fr_r_07, loc=<13 0 0>, rot=<0 0 0 1>}
 base_skirt_fr_l {parent=pelvis, loc=<4.3177 5.8613 6.6132>, rot=<-0.1281 -0.963 0.0855 -0.221>}
 dyn_skirt_fr_l_01 {parent=base_skirt_fr_l, loc=<0 0 0>, rot=<0 0 0 1>}
 dyn_skirt_fr_l_02 {parent=dyn_skirt_fr_l_01, loc=<12.717 0.0993 0>, rot=<0 0 0.0463 -0.9989>}
 dyn_skirt_fr_l_03 {parent=dyn_skirt_fr_l_02, loc=<13 0 0>, rot=<0 0 0.0296 -0.9996>}
 dyn_skirt_fr_l_04 {parent=dyn_skirt_fr_l_03, loc=<13 0 0>, rot=<0 0 0.0431 -0.9991>}
 dyn_skirt_fr_l_05 {parent=dyn_skirt_fr_l_04, loc=<13 0 0>, rot=<0 0 0.0368 -0.9993>}
 dyn_skirt_fr_l_06 {parent=dyn_skirt_fr_l_05, loc=<13 0 0>, rot=<0 0 0.0585 -0.9983>}
 dyn_skirt_fr_l_07 {parent=dyn_skirt_fr_l_06, loc=<13 0 0>, rot=<0 0 0.0298 -0.9996>}
 dyn_skirt_fr_l_08 {parent=dyn_skirt_fr_l_07, loc=<13 0 0>, rot=<0 0 0 1>}
 dyn_skirt_fr_l_01_extra {parent=dyn_skirt_fr_l_01, loc=<0 0 0>, rot=<0 0 0 1>}
 skirt_pelvis_fr {parent=pelvis, loc=<1.1169 6.4705 5.3435>, rot=<0.0127 0.707 0.707 0.0127>}
 base_skirt_sd_r {parent=pelvis, loc=<4.6904 -0.4311 10.731>, rot=<0.1512 -0.3808 0.9109 -0.0493>}
 dyn_skirt_sd_r_01 {parent=base_skirt_sd_r, loc=<0 0 0>, rot=<0 0 0 1>}
 dyn_skirt_sd_r_02 {parent=dyn_skirt_sd_r_01, loc=<17.4292 0 0>, rot=<-0.0176 -0.0329 -0.0279 -0.9989>}
 dyn_skirt_sd_r_03 {parent=dyn_skirt_sd_r_02, loc=<17.4268 0 0>, rot=<0.0537 0.012 -0.0189 -0.9983>}
 dyn_skirt_sd_r_04 {parent=dyn_skirt_sd_r_03, loc=<17.4326 0 0>, rot=<-0.0105 -0.0083 0.0032 -0.9999>}
 dyn_skirt_sd_r_05 {parent=dyn_skirt_sd_r_04, loc=<17.4329 0 -0.0001>, rot=<-0.0062 0.0008 0.0018 -1>}
 dyn_skirt_sd_r_06 {parent=dyn_skirt_sd_r_05, loc=<17.4333 0 0.0001>, rot=<-0.0113 0.0414 0.0037 -0.9991>}
]
variables [
 Crouching {type=bool, default=False}
 Pistol {type=double, default=0}
 Targeting {type=bool, default=False}
]
