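# Post-process style secondary-motion AnimBP for the F_MED_MajorSpeech_Nix face accessory, animating the character's hair and pendants.
# It copies the attached parent's pose into a Control Rig fed pistol, crouch and targeting state, then caches the result.
# A serial chain of root-relative AnimDynamics nodes adds sprung, angle-limited motion to front hair, pendants and side hair, some at partial alpha.
# OnWeaponChanged sets the pistol alpha for Pistol weapon animations only.
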AnimBlueprint /BRCosmetics/Characters/Player/Female/Medium/Bodies/F_MED_MajorSpeech_Nix/Meshes/Parts/F_MED_MajorSpeech_Nix_FaceAcc_AnimBP
target_skeleton=&/F_MED_MajorSpeech_Nix_FaceAcc_Skeleton
graphs [
 AnimGraph {nodes [
   n1 = Anim Dynamics - Bone: dyn_hair_fr_01(ComponentPose=Local To Component(LocalPose=Use cached pose 'CtrlRigCache'()), Alpha=1, ExternalForce="0,0,0") {
    angularDampingOverride=0.99
    angularSpringConstant=4000
    bAngularSpring=true
    bOverrideAngularDamping=true
    bUseGravityOverride=true
    gravityScale=0.35
    simulationSpace=RootRelative
    boundBone {boneName=dyn_hair_fr_01}
    componentAppliedLinearAccClamp=<3000 3000 3000>
    componentLinearAccScale=<1 1 1>
    componentLinearVelScale=<1 1 1>
    gravityOverride {z=-980}
    physicsBodyDefinitions [{
      boundBone {boneName=dyn_hair_fr_01}
      boxExtents=<25 25 25>
      constraintSetup {angularLimitsMax=<5 0 15>, angularLimitsMin=<-5 -5 4>, angularTarget=<1 0 0>}
      localJointOffset=<10 0 0>
     }]
    relativeSpaceBone {boneName=spine_05}
    simSpaceSettings {maxAngularAcceleration=100, maxAngularVelocity=100, simSpaceAngularAlpha=0.5}
   }
   n2 = Anim Dynamics - Bone: dyn_hair_fr_02(ComponentPose=@n1, Alpha=1, ExternalForce="0,0,0") {
    angularDampingOverride=0.99
    angularSpringConstant=4000
    bAngularSpring=true
    bOverrideAngularDamping=true
    bUseGravityOverride=true
    gravityScale=0.35
    simulationSpace=RootRelative
    boundBone {boneName=dyn_hair_fr_02}
    componentAppliedLinearAccClamp=<3000 3000 3000>
    componentLinearAccScale=<1 1 1>
    componentLinearVelScale=<1 1 1>
    gravityOverride {z=-980}
    physicsBodyDefinitions [{
      boundBone {boneName=dyn_hair_fr_02}
      boxExtents=<25 25 25>
      constraintSetup {angularLimitsMax=<5 2 30>, angularLimitsMin=<-5 -10 4>, angularTarget=<1 0 0>}
      localJointOffset=<10 0 0>
     }]
    relativeSpaceBone {boneName=spine_05}
    simSpaceSettings {maxAngularAcceleration=100, maxAngularVelocity=100, simSpaceAngularAlpha=0.5}
   }
   n3 = Anim Dynamics - Bone: dyn_hair_fr_03(ComponentPose=@n2, Alpha=1, ExternalForce="0,0,0") {
    angularDampingOverride=0.99
    angularSpringConstant=4000
    bAngularSpring=true
    bOverrideAngularDamping=true
    bUseGravityOverride=true
    gravityScale=0.35
    simulationSpace=RootRelative
    boundBone {boneName=dyn_hair_fr_03}
    componentAppliedLinearAccClamp=<3000 3000 3000>
    componentLinearAccScale=<1 1 1>
    componentLinearVelScale=<1 1 1>
    gravityOverride {z=-980}
    physicsBodyDefinitions [{
      boundBone {boneName=dyn_hair_fr_03}
      boxExtents=<25 25 25>
      constraintSetup {angularLimitsMax=<5 5 35>, angularLimitsMin=<-5 -10 -2>, angularTarget=<1 0 0>}
      localJointOffset=<10 0 0>
     }]
    relativeSpaceBone {boneName=spine_05}
    simSpaceSettings {maxAngularAcceleration=100, maxAngularVelocity=100, simSpaceAngularAlpha=0.5}
   }
   n4 = Anim Dynamics - Bone: dyn_hair_fr_04(ComponentPose=@n3, Alpha=1, ExternalForce="0,0,0") {
    angularDampingOverride=0.99
    angularSpringConstant=4000
    bAngularSpring=true
    bOverrideAngularDamping=true
    bUseGravityOverride=true
    gravityScale=0.35
    simulationSpace=RootRelative
    boundBone {boneName=dyn_hair_fr_04}
    componentAppliedLinearAccClamp=<3000 3000 3000>
    componentLinearAccScale=<1 1 1>
    componentLinearVelScale=<1 1 1>
    gravityOverride {z=-980}
    physicsBodyDefinitions [{
      boundBone {boneName=dyn_hair_fr_04}
      boxExtents=<25 25 25>
      constraintSetup {angularLimitsMax=<5 5 40>, angularLimitsMin=<-5 -10 -2>, angularTarget=<1 0 0>}
      localJointOffset=<10 0 0>
     }]
    relativeSpaceBone {boneName=spine_05}
    simSpaceSettings {maxAngularAcceleration=100, maxAngularVelocity=100, simSpaceAngularAlpha=0.5}
   }
   n5 = Anim Dynamics - Bone: dyn_pendants_sd_a_02_l(ComponentPose=@n4, Alpha=0.5, ExternalForce="0,0,0") {
    angularDampingOverride=1
    angularSpringConstant=800
    bAngularSpring=true
    bOverrideAngularDamping=true
    bUseGravityOverride=true
    gravityScale=0.35
    simulationSpace=RootRelative
    boundBone {boneName=dyn_pendants_sd_a_02_l}
    componentAppliedLinearAccClamp=<2000 2000 2000>
    componentLinearAccScale=<1 1 1>
    componentLinearVelScale=<1 1 1>
    gravityOverride {z=-980}
    physicsBodyDefinitions [{
      boundBone {boneName=dyn_pendants_sd_a_02_l}
      boxExtents=<20 30 30>
      constraintSetup {angularLimitsMax=<0 0 0>, angularLimitsMin=<0 -5 0>, angularTarget=<1 0 0>}
      localJointOffset=<4 0 0>
     }]
   }
   n6 = Anim Dynamics - Bone: dyn_pendants_sd_b_02_l(ComponentPose=@n5, Alpha=1, ExternalForce="0,0,0") {
    angularDampingOverride=1
    angularSpringConstant=800
    bAngularSpring=true
    bOverrideAngularDamping=true
    bUseGravityOverride=true
    gravityScale=0.35
    simulationSpace=RootRelative
    boundBone {boneName=dyn_pendants_sd_b_02_l}
    componentAppliedLinearAccClamp=<2000 2000 2000>
    componentLinearAccScale=<1 1 1>
    componentLinearVelScale=<1 1 1>
    gravityOverride {z=-980}
    physicsBodyDefinitions [{
      boundBone {boneName=dyn_pendants_sd_b_02_l}
      boxExtents=<20 30 30>
      constraintSetup {angularLimitsMax=<0 0 0>, angularLimitsMin=<0 -5 0>, angularTarget=<1 0 0>}
      localJointOffset=<4 0 0>
     }]
   }
   n7 = Anim Dynamics - Bone: dyn_pendants_bk_02(ComponentPose=@n6, Alpha=1, ExternalForce="0,0,0") {
    angularDampingOverride=1
    angularSpringConstant=800
    bAngularSpring=true
    bOverrideAngularDamping=true
    bUseGravityOverride=true
    gravityScale=0.35
    simulationSpace=RootRelative
    boundBone {boneName=dyn_pendants_bk_02}
    componentAppliedLinearAccClamp=<2000 2000 2000>
    componentLinearAccScale=<1 1 1>
    componentLinearVelScale=<1 1 1>
    gravityOverride {z=-980}
    physicsBodyDefinitions [{
      boundBone {boneName=dyn_pendants_bk_02}
      boxExtents=<20 30 30>
      constraintSetup {angularLimitsMax=<0 0 0>, angularLimitsMin=<0 -5 0>, angularTarget=<1 0 0>}
      localJointOffset=<4 0 0>
     }]
   }
   n8 = Anim Dynamics - Bone: dyn_pendants_sd_b_02_r(ComponentPose=@n7, Alpha=1, ExternalForce="0,0,0") {
    angularDampingOverride=1
    angularSpringConstant=800
    bAngularSpring=true
    bOverrideAngularDamping=true
    bUseGravityOverride=true
    gravityScale=0.35
    simulationSpace=RootRelative
    boundBone {boneName=dyn_pendants_sd_b_02_r}
    componentAppliedLinearAccClamp=<2000 2000 2000>
    componentLinearAccScale=<1 1 1>
    componentLinearVelScale=<1 1 1>
    gravityOverride {z=-980}
    physicsBodyDefinitions [{
      boundBone {boneName=dyn_pendants_sd_b_02_r}
      boxExtents=<20 30 30>
      constraintSetup {angularLimitsMax=<0 0 0>, angularLimitsMin=<0 -5 0>, angularTarget=<1 0 0>}
      localJointOffset=<-4 0 0>
     }]
   }
   n9 = Anim Dynamics - Bone: dyn_pendants_sd_a_02_r(ComponentPose=@n8, Alpha=0.5, ExternalForce="0,0,0") {
    angularDampingOverride=1
    angularSpringConstant=800
    bAngularSpring=true
    bOverrideAngularDamping=true
    bUseGravityOverride=true
    gravityScale=0.35
    simulationSpace=RootRelative
    boundBone {boneName=dyn_pendants_sd_a_02_r}
    componentAppliedLinearAccClamp=<2000 2000 2000>
    componentLinearAccScale=<1 1 1>
    componentLinearVelScale=<1 1 1>
    gravityOverride {z=-980}
    physicsBodyDefinitions [{
      boundBone {boneName=dyn_pendants_sd_a_02_r}
      boxExtents=<20 30 30>
      constraintSetup {angularLimitsMax=<0 0 0>, angularLimitsMin=<0 -5 0>, angularTarget=<1 0 0>}
      localJointOffset=<-4 0 0>
     }]
   }
   n10 = Anim Dynamics - Bone: dyn_hair_sd_c_01_l(ComponentPose=@n9, Alpha=1, ExternalForce="0,0,0") {
    angularDampingOverride=1
    angularSpringConstant=10000
    bAngularSpring=true
    bOverrideAngularDamping=true
    bUseGravityOverride=true
    gravityScale=0.35
    simulationSpace=RootRelative
    boundBone {boneName=dyn_hair_sd_c_01_l}
    componentAppliedLinearAccClamp=<3000 3000 3000>
    componentLinearAccScale=<1 1 1>
    componentLinearVelScale=<1 1 1>
    gravityOverride {z=-980}
    physicsBodyDefinitions [{
      boundBone {boneName=dyn_hair_sd_c_01_l}
      boxExtents=<35 35 35>
      constraintSetup {angularLimitsMax=<10 -3 10>, angularLimitsMin=<-10 -15 -10>, angularTarget=<1 0 0>}
      localJointOffset=<8 0 0>
     }]
    simSpaceSettings {maxAngularAcceleration=1000, maxAngularVelocity=1000, simSpaceAngularAlpha=0.5}
   }
   n11 = Anim Dynamics - Bone: dyn_hair_sd_c_02_l(ComponentPose=@n10, Alpha=1, ExternalForce="0,0,0") {
    angularDampingOverride=1
    angularSpringConstant=10000
    bAngularSpring=true
    bOverrideAngularDamping=true
    bUseGravityOverride=true
    gravityScale=0.35
    simulationSpace=RootRelative
    boundBone {boneName=dyn_hair_sd_c_02_l}
    componentAppliedLinearAccClamp=<3000 3000 3000>
    componentLinearAccScale=<1 1 1>
    componentLinearVelScale=<1 1 1>
    gravityOverride {z=-980}
    physicsBodyDefinitions [{
      boundBone {boneName=dyn_hair_sd_c_02_l}
      boxExtents=<35 35 35>
      constraintSetup {angularLimitsMax=<5 -3 10>, angularLimitsMin=<-5 -15 -15>, angularTarget=<1 0 0>}
      localJointOffset=<8 0 0>
     }]
    simSpaceSettings {maxAngularAcceleration=1000, maxAngularVelocity=1000, simSpaceAngularAlpha=0.5}
   }
   n12 = Anim Dynamics - Bone: dyn_hair_sd_b_01_l(ComponentPose=@n11, Alpha=1, ExternalForce="0,0,0") {
    angularDampingOverride=1
    angularSpringConstant=10000
    bAngularSpring=true
    bOverrideAngularDamping=true
    bUseGravityOverride=true
    gravityScale=0.35
    simulationSpace=RootRelative
    boundBone {boneName=dyn_hair_sd_b_01_l}
    componentAppliedLinearAccClamp=<3000 3000 3000>
    componentLinearAccScale=<1 1 1>
    componentLinearVelScale=<1 1 1>
    gravityOverride {z=-980}
    physicsBodyDefinitions [{
      boundBone {boneName=dyn_hair_sd_b_01_l}
      boxExtents=<35 35 35>
      constraintSetup {angularLimitsMax=<10 -3 10>, angularLimitsMin=<-10 -15 -10>, angularTarget=<1 0 0>}
      localJointOffset=<8 0 0>
     }]
    simSpaceSettings {maxAngularAcceleration=1000, maxAngularVelocity=1000, simSpaceAngularAlpha=0.5}
   }
   n13 = Anim Dynamics - Bone: dyn_hair_sd_b_02_l(ComponentPose=@n12, Alpha=1, ExternalForce="0,0,0") {
    angularDampingOverride=1
    angularSpringConstant=10000
    bAngularSpring=true
    bOverrideAngularDamping=true
    bUseGravityOverride=true
    gravityScale=0.35
    simulationSpace=RootRelative
    boundBone {boneName=dyn_hair_sd_b_02_l}
    componentAppliedLinearAccClamp=<3000 3000 3000>
    componentLinearAccScale=<1 1 1>
    componentLinearVelScale=<1 1 1>
    gravityOverride {z=-980}
    physicsBodyDefinitions [{
      boundBone {boneName=dyn_hair_sd_b_02_l}
      boxExtents=<35 35 35>
      constraintSetup {angularLimitsMax=<5 -3 10>, angularLimitsMin=<-5 -15 -15>, angularTarget=<1 0 0>}
      localJointOffset=<8 0 0>
     }]
    simSpaceSettings {maxAngularAcceleration=1000, maxAngularVelocity=1000, simSpaceAngularAlpha=0.5}
   }
   n14 = Anim Dynamics - Bone: dyn_hair_sd_a_01_l(ComponentPose=@n13, Alpha=1, ExternalForce="0,0,0") {
    angularDampingOverride=1
    angularSpringConstant=10000
    bAngularSpring=true
    bOverrideAngularDamping=true
    bUseGravityOverride=true
    gravityScale=0.35
    simulationSpace=RootRelative
    boundBone {boneName=dyn_hair_sd_a_01_l}
    componentAppliedLinearAccClamp=<3000 3000 3000>
    componentLinearAccScale=<1 1 1>
    componentLinearVelScale=<1 1 1>
    gravityOverride {z=-980}
    physicsBodyDefinitions [{
      boundBone {boneName=dyn_hair_sd_a_01_l}
      boxExtents=<35 35 35>
      constraintSetup {angularLimitsMax=<10 -3 10>, angularLimitsMin=<-10 -15 -10>, angularTarget=<1 0 0>}
      localJointOffset=<8 0 0>
     }]
    simSpaceSettings {maxAngularAcceleration=1000, maxAngularVelocity=1000, simSpaceAngularAlpha=0.5}
   }
   n15 = Anim Dynamics - Bone: dyn_hair_sd_a_02_l(ComponentPose=@n14, Alpha=1, ExternalForce="0,0,0") {
    angularDampingOverride=1
    angularSpringConstant=10000
    bAngularSpring=true
    bOverrideAngularDamping=true
    bUseGravityOverride=true
    gravityScale=0.35
    simulationSpace=RootRelative
    boundBone {boneName=dyn_hair_sd_a_02_l}
    componentAppliedLinearAccClamp=<3000 3000 3000>
    componentLinearAccScale=<1 1 1>
    componentLinearVelScale=<1 1 1>
    gravityOverride {z=-980}
    physicsBodyDefinitions [{
      boundBone {boneName=dyn_hair_sd_a_02_l}
      boxExtents=<35 35 35>
      constraintSetup {angularLimitsMax=<5 -3 10>, angularLimitsMin=<-5 -15 -15>, angularTarget=<1 0 0>}
      localJointOffset=<8 0 0>
     }]
    simSpaceSettings {maxAngularAcceleration=1000, maxAngularVelocity=1000, simSpaceAngularAlpha=0.5}
   }
   n16 = Anim Dynamics - Bone: dyn_hair_sd_d_01_l(ComponentPose=@n15, Alpha=1, ExternalForce="0,0,0") {
    angularDampingOverride=1
    angularSpringConstant=10000
    bAngularSpring=true
    bOverrideAngularDamping=true
    bUseGravityOverride=true
    gravityScale=0.35
    simulationSpace=RootRelative
    boundBone {boneName=dyn_hair_sd_d_01_l}
    componentAppliedLinearAccClamp=<3000 3000 3000>
    componentLinearAccScale=<1 1 1>
    componentLinearVelScale=<1 1 1>
    gravityOverride {z=-980}
    physicsBodyDefinitions [{
      boundBone {boneName=dyn_hair_sd_d_01_l}
      boxExtents=<35 35 35>
      constraintSetup {angularLimitsMax=<10 0 5>, angularLimitsMin=<-10 -10 -5>, angularTarget=<1 0 0>}
      localJointOffset=<8 0 0>
     }]
    simSpaceSettings {maxAngularAcceleration=1000, maxAngularVelocity=1000, simSpaceAngularAlpha=0.5}
   }
   n17 = Anim Dynamics - Bone: dyn_hair_sd_d_02_l(ComponentPose=@n16, Alpha=1, ExternalForce="0,0,0") {
    angularDampingOverride=1
    angularSpringConstant=10000
    bAngularSpring=true
    bOverrideAngularDamping=true
    bUseGravityOverride=true
    gravityScale=0.35
    simulationSpace=RootRelative
    boundBone {boneName=dyn_hair_sd_d_02_l}
    componentAppliedLinearAccClamp=<3000 3000 3000>
    componentLinearAccScale=<1 1 1>
    componentLinearVelScale=<1 1 1>
    gravityOverride {z=-980}
    physicsBodyDefinitions [{
      boundBone {boneName=dyn_hair_sd_d_02_l}
      boxExtents=<35 35 35>
      constraintSetup {angularLimitsMax=<5 0 5>, angularLimitsMin=<-5 -10 -5>, angularTarget=<1 0 0>}
      localJointOffset=<8 0 0>
     }]
    simSpaceSettings {maxAngularAcceleration=1000, maxAngularVelocity=1000, simSpaceAngularAlpha=0.5}
   }
   n18 = Anim Dynamics - Bone: dyn_hair_sd_c_01_r(ComponentPose=@n17, Alpha=1, ExternalForce="0,0,0") {
    angularDampingOverride=1
    angularSpringConstant=10000
    bAngularSpring=true
    bOverrideAngularDamping=true
    bUseGravityOverride=true
    gravityScale=0.35
    simulationSpace=RootRelative
    boundBone {boneName=dyn_hair_sd_c_01_r}
    componentAppliedLinearAccClamp=<3000 3000 3000>
    componentLinearAccScale=<1 1 1>
    componentLinearVelScale=<1 1 1>
    gravityOverride {z=-980}
    physicsBodyDefinitions [{
      boundBone {boneName=dyn_hair_sd_c_01_r}
      boxExtents=<35 35 35>
      constraintSetup {angularLimitsMax=<10 2 0>, angularLimitsMin=<-10 -15 -10>, angularTarget=<1 0 0>}
      localJointOffset=<-8 0 0>
     }]
    simSpaceSettings {maxAngularAcceleration=1000, maxAngularVelocity=1000, simSpaceAngularAlpha=0.5}
   }
   n19 = Anim Dynamics - Bone: dyn_hair_sd_c_02_r(ComponentPose=@n18, Alpha=1, ExternalForce="0,0,0") {
    angularDampingOverride=1
    angularSpringConstant=10000
    bAngularSpring=true
    bOverrideAngularDamping=true
    bUseGravityOverride=true
    gravityScale=0.35
    simulationSpace=RootRelative
    boundBone {boneName=dyn_hair_sd_c_02_r}
    componentAppliedLinearAccClamp=<3000 3000 3000>
    componentLinearAccScale=<1 1 1>
    componentLinearVelScale=<1 1 1>
    gravityOverride {z=-980}
    physicsBodyDefinitions [{
      boundBone {boneName=dyn_hair_sd_c_02_r}
      boxExtents=<35 35 35>
      constraintSetup {angularLimitsMax=<5 2 0>, angularLimitsMin=<-5 -15 -15>, angularTarget=<1 0 0>}
      localJointOffset=<-8 0 0>
     }]
    simSpaceSettings {maxAngularAcceleration=1000, maxAngularVelocity=1000, simSpaceAngularAlpha=0.5}
   }
   n20 = Anim Dynamics - Bone: dyn_hair_sd_b_01_r(ComponentPose=@n19, Alpha=1, ExternalForce="0,0,0") {
    angularDampingOverride=1
    angularSpringConstant=10000
    bAngularSpring=true
    bOverrideAngularDamping=true
    bUseGravityOverride=true
    gravityScale=0.35
    simulationSpace=RootRelative
    boundBone {boneName=dyn_hair_sd_b_01_r}
    componentAppliedLinearAccClamp=<3000 3000 3000>
    componentLinearAccScale=<1 1 1>
    componentLinearVelScale=<1 1 1>
    gravityOverride {z=-980}
    physicsBodyDefinitions [{
      boundBone {boneName=dyn_hair_sd_b_01_r}
      boxExtents=<35 35 35>
      constraintSetup {angularLimitsMax=<10 2 0>, angularLimitsMin=<-10 -15 -10>, angularTarget=<1 0 0>}
      localJointOffset=<-8 0 0>
     }]
    simSpaceSettings {maxAngularAcceleration=1000, maxAngularVelocity=1000, simSpaceAngularAlpha=0.5}
   }
   n21 = Anim Dynamics - Bone: dyn_hair_sd_b_02_r(ComponentPose=@n20, Alpha=1, ExternalForce="0,0,0") {
    angularDampingOverride=1
    angularSpringConstant=10000
    bAngularSpring=true
    bOverrideAngularDamping=true
    bUseGravityOverride=true
    gravityScale=0.35
    simulationSpace=RootRelative
    boundBone {boneName=dyn_hair_sd_b_02_r}
    componentAppliedLinearAccClamp=<3000 3000 3000>
    componentLinearAccScale=<1 1 1>
    componentLinearVelScale=<1 1 1>
    gravityOverride {z=-980}
    physicsBodyDefinitions [{
      boundBone {boneName=dyn_hair_sd_b_02_r}
      boxExtents=<35 35 35>
      constraintSetup {angularLimitsMax=<5 2 0>, angularLimitsMin=<-5 -15 -15>, angularTarget=<1 0 0>}
      localJointOffset=<-8 0 0>
     }]
    simSpaceSettings {maxAngularAcceleration=1000, maxAngularVelocity=1000, simSpaceAngularAlpha=0.5}
   }
   n22 = Anim Dynamics - Bone: dyn_hair_sd_a_01_r(ComponentPose=@n21, Alpha=0.3, ExternalForce="0,0,0") {
    angularDampingOverride=1
    angularSpringConstant=10000
    bAngularSpring=true
    bOverrideAngularDamping=true
    bUseGravityOverride=true
    gravityScale=0.35
    simulationSpace=RootRelative
    boundBone {boneName=dyn_hair_sd_a_01_r}
    componentAppliedLinearAccClamp=<3000 3000 3000>
    componentLinearAccScale=<1 1 1>
    componentLinearVelScale=<1 1 1>
    gravityOverride {z=-980}
    physicsBodyDefinitions [{
      boundBone {boneName=dyn_hair_sd_a_01_r}
      boxExtents=<35 35 35>
      constraintSetup {angularLimitsMax=<0 0 0>, angularLimitsMin=<0 0 0>, angularTarget=<1 0 0>}
      localJointOffset=<-8 0 0>
     }]
    simSpaceSettings {maxAngularAcceleration=1000, maxAngularVelocity=1000, simSpaceAngularAlpha=0.5}
   }
   n23 = Anim Dynamics - Bone: dyn_hair_sd_a_02_r(ComponentPose=@n22, Alpha=0.3, ExternalForce="0,0,0") {
    angularDampingOverride=1
    angularSpringConstant=10000
    bAngularSpring=true
    bOverrideAngularDamping=true
    bUseGravityOverride=true
    gravityScale=0.35
    simulationSpace=RootRelative
    boundBone {boneName=dyn_hair_sd_a_02_r}
    componentAppliedLinearAccClamp=<3000 3000 3000>
    componentLinearAccScale=<1 1 1>
    componentLinearVelScale=<1 1 1>
    gravityOverride {z=-980}
    physicsBodyDefinitions [{
      boundBone {boneName=dyn_hair_sd_a_02_r}
      boxExtents=<35 35 35>
      constraintSetup {angularLimitsMax=<0 0 0>, angularLimitsMin=<0 0 0>, angularTarget=<1 0 0>}
      localJointOffset=<-8 0 0>
     }]
    simSpaceSettings {maxAngularAcceleration=1000, maxAngularVelocity=1000, simSpaceAngularAlpha=0.5}
   }
   n24 = Anim Dynamics - Bone: dyn_hair_sd_d_01_r(ComponentPose=@n23, Alpha=1, ExternalForce="0,0,0") {
    angularDampingOverride=1
    angularSpringConstant=10000
    bAngularSpring=true
    bOverrideAngularDamping=true
    bUseGravityOverride=true
    gravityScale=0.35
    simulationSpace=RootRelative
    boundBone {boneName=dyn_hair_sd_d_01_r}
    componentAppliedLinearAccClamp=<3000 3000 3000>
    componentLinearAccScale=<1 1 1>
    componentLinearVelScale=<1 1 1>
    gravityOverride {z=-980}
    physicsBodyDefinitions [{
      boundBone {boneName=dyn_hair_sd_d_01_r}
      boxExtents=<35 35 35>
      constraintSetup {angularLimitsMax=<10 2 0>, angularLimitsMin=<-10 -15 -10>, angularTarget=<1 0 0>}
      localJointOffset=<-8 0 0>
     }]
    simSpaceSettings {maxAngularAcceleration=1000, maxAngularVelocity=1000, simSpaceAngularAlpha=0.5}
   }
   n25 = Anim Dynamics - Bone: dyn_hair_sd_d_02_r(ComponentPose=@n24, Alpha=1, ExternalForce="0,0,0") {
    angularDampingOverride=1
    angularSpringConstant=10000
    bAngularSpring=true
    bOverrideAngularDamping=true
    bUseGravityOverride=true
    gravityScale=0.35
    simulationSpace=RootRelative
    boundBone {boneName=dyn_hair_sd_d_02_r}
    componentAppliedLinearAccClamp=<3000 3000 3000>
    componentLinearAccScale=<1 1 1>
    componentLinearVelScale=<1 1 1>
    gravityOverride {z=-980}
    physicsBodyDefinitions [{
      boundBone {boneName=dyn_hair_sd_d_02_r}
      boxExtents=<35 35 35>
      constraintSetup {angularLimitsMax=<0 2 0>, angularLimitsMin=<0 -15 -15>, angularTarget=<1 0 0>}
      localJointOffset=<-8 0 0>
     }]
    simSpaceSettings {maxAngularAcceleration=1000, maxAngularVelocity=1000, simSpaceAngularAlpha=0.5}
   }
   n26 = Anim Dynamics - Bone: dyn_hair_sd_e_01_r(ComponentPose=@n25, Alpha=1, ExternalForce="0,0,0") {
    angularDampingOverride=1
    angularSpringConstant=10000
    bAngularSpring=true
    bOverrideAngularDamping=true
    bUseGravityOverride=true
    gravityScale=0.35
    simulationSpace=RootRelative
    boundBone {boneName=dyn_hair_sd_e_01_r}
    componentAppliedLinearAccClamp=<3000 3000 3000>
    componentLinearAccScale=<1 1 1>
    componentLinearVelScale=<1 1 1>
    gravityOverride {z=-980}
    physicsBodyDefinitions [{
      boundBone {boneName=dyn_hair_sd_e_01_r}
      boxExtents=<35 35 35>
      constraintSetup {angularLimitsMax=<10 0 0>, angularLimitsMin=<-10 -7 -5>, angularTarget=<1 0 0>}
      localJointOffset=<-8 0 0>
     }]
    simSpaceSettings {maxAngularAcceleration=1000, maxAngularVelocity=1000, simSpaceAngularAlpha=0.5}
   }
   n27 = Anim Dynamics - Bone: dyn_hair_sd_e_02_r(ComponentPose=@n26, Alpha=1, ExternalForce="0,0,0") {
    angularDampingOverride=1
    angularSpringConstant=10000
    bAngularSpring=true
    bOverrideAngularDamping=true
    bUseGravityOverride=true
    gravityScale=0.35
    simulationSpace=RootRelative
    boundBone {boneName=dyn_hair_sd_e_02_r}
    componentAppliedLinearAccClamp=<3000 3000 3000>
    componentLinearAccScale=<1 1 1>
    componentLinearVelScale=<1 1 1>
    gravityOverride {z=-980}
    physicsBodyDefinitions [{
      boundBone {boneName=dyn_hair_sd_e_02_r}
      boxExtents=<35 35 35>
      constraintSetup {angularLimitsMax=<0 0 0>, angularLimitsMin=<0 -7 -5>, angularTarget=<1 0 0>}
      localJointOffset=<-8 0 0>
     }]
    simSpaceSettings {maxAngularAcceleration=1000, maxAngularVelocity=1000, simSpaceAngularAlpha=0.5}
   }
   n28 = Anim Dynamics - Bone: dyn_hair_sd_f_01_r(ComponentPose=@n27, Alpha=1, ExternalForce="0,0,0") {
    angularDampingOverride=1
    angularSpringConstant=10000
    bAngularSpring=true
    bOverrideAngularDamping=true
    bUseGravityOverride=true
    gravityScale=0.35
    simulationSpace=RootRelative
    boundBone {boneName=dyn_hair_sd_f_01_r}
    componentAppliedLinearAccClamp=<3000 3000 3000>
    componentLinearAccScale=<1 1 1>
    componentLinearVelScale=<1 1 1>
    gravityOverride {z=-980}
    physicsBodyDefinitions [{
      boundBone {boneName=dyn_hair_sd_f_01_r}
      boxExtents=<35 35 35>
      constraintSetup {angularLimitsMax=<10 2 0>, angularLimitsMin=<-10 -15 -10>, angularTarget=<1 0 0>}
      localJointOffset=<-8 0 0>
     }]
    simSpaceSettings {maxAngularAcceleration=1000, maxAngularVelocity=1000, simSpaceAngularAlpha=0.5}
   }
   n29 = Anim Dynamics - Bone: dyn_hair_sd_f_02_r(ComponentPose=@n28, Alpha=1, ExternalForce="0,0,0") {
    angularDampingOverride=1
    angularSpringConstant=10000
    bAngularSpring=true
    bOverrideAngularDamping=true
    bUseGravityOverride=true
    gravityScale=0.35
    simulationSpace=RootRelative
    boundBone {boneName=dyn_hair_sd_f_02_r}
    componentAppliedLinearAccClamp=<3000 3000 3000>
    componentLinearAccScale=<1 1 1>
    componentLinearVelScale=<1 1 1>
    gravityOverride {z=-980}
    physicsBodyDefinitions [{
      boundBone {boneName=dyn_hair_sd_f_02_r}
      boxExtents=<35 35 35>
      constraintSetup {angularLimitsMax=<0 2 0>, angularLimitsMin=<0 -15 -15>, angularTarget=<1 0 0>}
      localJointOffset=<-8 0 0>
     }]
    simSpaceSettings {maxAngularAcceleration=1000, maxAngularVelocity=1000, simSpaceAngularAlpha=0.5}
   }
   Output Pose(Result=Component To Local(ComponentPose=@n29)) {layerGroup=None}
   n30 = Control Rig!(
    Source=Copy Pose From Mesh() {bUseAttachedParent=true}
    Alpha
    Pistol:double=Get Alpha_DetectedPistol() -> (Alpha_DetectedPistol:double)
    Crouching:bool=Get bIsCrouching()
    Targeting:bool=Get bIsTargeting()
   ) -> (Pose) {controlRigAssetReference {blueprintRigClass=&/F_MED_MajorSpeech_Nix_FaceAcc_CtrlRig.F_MED_MajorSpeech_Nix_FaceAcc_CtrlRig_C}}
   Save cached pose 'CtrlRigCache'(Pose=@n30) {cacheName=CtrlRigCache}
  ]}
 EventGraph {nodes [
   Event OnWeaponChanged()
   > n1 = Cast To FortPlayerAnimInstance(Object=GetAnimInstance(self=GetOwnerSkeletalMeshComponent()))
   > Set Alpha_DetectedPistol(Alpha_DetectedPistol:double=0) -> (Output_Get:double)
   > n2 = Switch on EFortWeaponCoreAnimation!(execute, Selection:EFortWeaponCoreAnimation Enum=Get WeaponCoreAnim(self=@n1)) -> (
    Melee:Exec
    Pistol:Exec
    Shotgun:Exec
    PaperBlueprint:Exec
    Rifle:Exec
    MeleeOneHand:Exec
    MachinePistol:Exec
    RocketLauncher:Exec
    GrenadeLauncher:Exec
    GoingCommando:Exec
    AssaultRifle:Exec
    TacticalShotgun:Exec
    SniperRifle:Exec
    TrapPlacement:Exec
    ShoulderLauncher:Exec
    AbilityDecoTool:Exec
    Crossbow:Exec
    C4:Exec
    RemoteControl:Exec
    DualWield:Exec
    AR_BullPup:Exec
    AR_ForwardGrip:Exec
    MedPackPaddles:Exec
    SMG_P90:Exec
    AR_DrumGun:Exec
    Consumable_Small:Exec
    Consumable_Large:Exec
    Balloon:Exec
    MountedTurret:Exec
    CreativeTool:Exec
    ExplosiveBow:Exec
    AshtonIndigo:Exec
    AshtonChicago:Exec
    MeleeDualWield:Exec
    Unarmed:Exec
   )
   Set Alpha_DetectedPistol(execute=@n2.Pistol, Alpha_DetectedPistol=1)
  ]}
]
variables [Alpha_DetectedPistol {type=number}]
